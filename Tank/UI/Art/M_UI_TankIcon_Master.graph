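# Master additive UI material for tank icons, with a normal and a warning look blended by the WarningOn parameter.
# The texture's red channel blends outer and inner colors within each state, and WarningOn blends between the two color pairs.
# A saturated intensity mask sums red, blue scaled by a glow that rises from Base Glow under warning, and green scaled by normal or warning fill opacity.
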Material /Tank/UI/Art/M_UI_TankIcon_Master
graph {
 properties {blendMode=BLEND_Additive, materialDomain=MD_UI, pixelDepthOffsetMode=PDOM_Legacy, usedShadingModels=Default Lit}
 nodes [
  n1 = TextureSampleParameter2D() {parameterName=Texture, texture=&/T_UI_TankIcon_Canon}
  LinearInterpolate(
   A=LinearInterpolate(
    A=VectorParameter() {parameterName=RGB_Out_Off, defaultValue=<rgba 1 0.25 0 1>}.RGB
    B=VectorParameter() {parameterName=RGB_In_Off, defaultValue=<rgba 1 0.7558 0.2674 1>}.RGB
    Alpha=@n1.R
   )
   B=LinearInterpolate(
    A=VectorParameter() {parameterName=RGB_Out_On, defaultValue=<rgba 1 0 0 1>}.RGB
    B=VectorParameter() {parameterName=RGB_In_On, defaultValue=<rgba 1 0.05 0 1>}.RGB
    Alpha=@n1.R
   )
   Alpha=ScalarParameter() {parameterName=WarningOn, sliderMax=1}
  )
  n2 = LinearInterpolate(
   A=ScalarParameter() {defaultValue=0.0025, parameterName=Fill Opacity}
   B=ScalarParameter() {defaultValue=0.2, parameterName=Fill Opacity Warning}
   Alpha=ScalarParameter() {parameterName=WarningOn, sliderMax=1}
  )
  n3 = ScalarParameter() {defaultValue=0.5, parameterName=Base Glow, sliderMax=1}
  n4 = ScalarParameter() {parameterName=WarningOn, sliderMax=1}
  n5 = Add(A=@n1.R, B=Multiply(A=@n1.B, B=LinearInterpolate(A=@n3, Alpha=@n4)))
  Saturate(None=Add(A=Multiply(A=@n1.G, B=@n2), B=Saturate(None=@n5)))
 ]
}
