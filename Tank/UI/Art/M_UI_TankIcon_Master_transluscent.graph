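# Translucent UI master material for tank icons, with a swappable texture parameter and a warning color state.
# The texture's R channel blends between outer and inner colors in separate off and on pairs, and the WarningOn scalar interpolates between those pairs.
# An alpha mask adds the doubled, clamped R channel to the G channel scaled by Fill Opacity, giving a faint fill by default.
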Material /Tank/UI/Art/M_UI_TankIcon_Master_transluscent
graph {
 properties {blendMode=BLEND_Translucent, materialDomain=MD_UI, pixelDepthOffsetMode=PDOM_Legacy, usedShadingModels=Default Lit}
 nodes [
  n1 = TextureSampleParameter2D() {parameterName=Texture, texture=&/T_UI_TankIcon_Canon}
  LinearInterpolate(
   A=LinearInterpolate(
    A=VectorParameter() {parameterName=RGB_Out_Off, defaultValue=<rgba 1 0.112 0 1>}.RGB
    B=VectorParameter() {parameterName=RGB_In_Off, defaultValue=<rgba 1 0.6524 0.0521 1>}.RGB
    Alpha=@n1.R
   )
   B=LinearInterpolate(
    A=VectorParameter() {parameterName=RGB_Out_On, defaultValue=<rgba 1 0 0 1>}.RGB
    B=VectorParameter() {parameterName=RGB_In_On, defaultValue=<rgba 1 0.05 0 1>}.RGB
    Alpha=@n1.R
   )
   Alpha=ScalarParameter() {parameterName=WarningOn, sliderMax=1}
  )
  n2 = ScalarParameter() {defaultValue=0.05, parameterName=Fill Opacity}
  n3 = Add(A=@n1.R, B=@n1.R)
  Saturate(None=Add(A=Multiply(A=@n1.G, B=@n2), B=Saturate(None=@n3)))
 ]
}
