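# Translucent UI material, named for the tank icon's speed lines, that scales the blue channel of the tank-icon dot texture by an intensity value.
# A LinearSine pulse is built from Time, PulseSpeed and PulseTimeOffset to drive that intensity between a dim and a brighter level.
# Both ends of the PulseConstraints blend read the red channel, so the pulse cancels out and intensity stays constant, dim by default.
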Material /Tank/UI/Art/M_UI_TankIcon_SpeedLines
graph {
 properties {blendMode=BLEND_Translucent, materialDomain=MD_UI, pixelDepthOffsetMode=PDOM_Legacy, usedShadingModels=Default Lit}
 nodes [
  Constant() {r=1}
  n1 = VectorParameter() {group=Pulse, parameterName=PulseConstraints, defaultValue=<rgba 0 1 0 1>}
  n2 = Multiply(A=Time(), B=ScalarParameter() {defaultValue=0.66, group=Pulse, parameterName=PulseSpeed})
  n3 = Constant() {r=1}
  n4 = MaterialFunctionCall(Value=Add(A=Fmod(A=@n2, B=@n3), B=ScalarParameter() {parameterName=PulseTimeOffset})) {materialFunction=&LinearSine}
  Multiply(
   A=TextureSample() {samplerSource=SSM_Clamp_WorldGroupSettings, texture=&/T_UI_TankIcon_Dot}.B
   B=LinearInterpolate(A=Constant() {r=0.15}, B=Constant() {r=0.75}, Alpha=LinearInterpolate(A=@n1.R, B=@n1.R, Alpha=@n4.Linear Sine))
  )
 ]
}
