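# UI material for a tank corner graphic, drawn with additive blending.
# It uses the red channel of the T_UI_Tank_Corner texture to interpolate between two constant colors, a deep red-orange and a lighter amber-orange.
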Material /Tank/UI/Art/M_UI_Tank_Corner
graph {
 properties {blendMode=BLEND_Additive, materialDomain=MD_UI, pixelDepthOffsetMode=PDOM_Legacy, usedShadingModels=Default Lit}
 nodes [LinearInterpolate(
   A=Constant4Vector() {constant=<rgba 1 0.112 0 1>}.RGBA
   B=Constant4Vector() {constant=<rgba 1 0.5029 0 1>}.RGBA
   Alpha=TextureSample() {texture=&/T_UI_Tank_Corner}.R
  )]
}
