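# Translucent UI material for the Tank side decoration, procedurally drawing a mask of thin horizontal lines in UV space.
# Six line groups each evaluate a box signed-distance Custom node twice, at mirrored vertical offsets, with a horizontal shift that varies per group.
# SmoothStep with a Sharpness parameter softens the edges, and named reroutes carry the shared BoxSize and per-group results.
# The groups are multiplied together and inverted to form the line mask.
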Material /Tank/UI/Art/M_UI_Tank_SideDeco
graph {
 properties {blendMode=BLEND_Translucent, materialDomain=MD_UI, pixelDepthOffsetMode=PDOM_Legacy, usedShadingModels=Default Lit}
 nodes [
  Constant() {r=1}
  n1 = ScalarParameter() {defaultValue=0.01, parameterName=Sharpness}
  n2 = TextureCoordinate()
  n3 = Constant() {r=0.1}
  n4 = Constant() {r=0.045}
  MaterialExpressionConstantBiasScale_0 = ConstantBiasScale(None=Add(A=@n2, B=AppendVector(A=@n3, B=@n4))) {bias=-0.5, scale=2}
  MaterialExpressionNamedRerouteUsage_11 = NamedRerouteUsage()
  n5 = Custom!(p=@MaterialExpressionConstantBiasScale_0, b=@MaterialExpressionNamedRerouteUsage_11) {
   code="float2 d = abs(p)-b;\r\n    return length(max(d,0.0)) + min(max(d.x,d.y),0.0);"
   outputType=CMOT_Float1
   inputs [
    {inputName=p, input {expression=&self:MaterialExpressionConstantBiasScale_0}}
    {inputName=b, input {expression=&self:MaterialExpressionNamedRerouteUsage_11}}
   ]
  }
  n6 = Multiply(A=@n4) {constB=-1}
  MaterialExpressionConstantBiasScale_7 = ConstantBiasScale(None=Add(A=@n2, B=AppendVector(A=@n3, B=@n6))) {bias=-0.5, scale=2}
  n7 = Custom(p=@MaterialExpressionConstantBiasScale_7, b=@MaterialExpressionNamedRerouteUsage_11) {
   code="float2 d = abs(p)-b;\r\n    return length(max(d,0.0)) + min(max(d.x,d.y),0.0);"
   outputType=CMOT_Float1
   inputs [
    {inputName=p, input {expression=&self:MaterialExpressionConstantBiasScale_7}}
    {inputName=b, input {expression=&self:MaterialExpressionNamedRerouteUsage_11}}
   ]
  }
  NamedRerouteDeclaration(None=Multiply(A=SmoothStep(Max=@n1, Value=@n5), B=SmoothStep(Max=@n1, Value=@n7))) {name=line 6, nodeColor=<rgba 0.4766 0 1 1>}
  n8 = ScalarParameter() {defaultValue=0.01, parameterName=Sharpness}
  n9 = TextureCoordinate()
  n10 = Constant() {r=0.07}
  n11 = Constant() {r=0.13}
  MaterialExpressionConstantBiasScale_1 = ConstantBiasScale(None=Add(A=@n9, B=AppendVector(A=@n10, B=@n11))) {bias=-0.5, scale=2}
  MaterialExpressionNamedRerouteUsage_13 = NamedRerouteUsage()
  n12 = Custom(p=@MaterialExpressionConstantBiasScale_1, b=@MaterialExpressionNamedRerouteUsage_13) {
   code="float2 d = abs(p)-b;\r\n    return length(max(d,0.0)) + min(max(d.x,d.y),0.0);"
   outputType=CMOT_Float1
   inputs [
    {inputName=p, input {expression=&self:MaterialExpressionConstantBiasScale_1}}
    {inputName=b, input {expression=&self:MaterialExpressionNamedRerouteUsage_13}}
   ]
  }
  n13 = Multiply(A=@n11) {constB=-1}
  MaterialExpressionConstantBiasScale_8 = ConstantBiasScale(None=Add(A=@n9, B=AppendVector(A=@n10, B=@n13))) {bias=-0.5, scale=2}
  n14 = Custom(p=@MaterialExpressionConstantBiasScale_8, b=@MaterialExpressionNamedRerouteUsage_13) {
   code="float2 d = abs(p)-b;\r\n    return length(max(d,0.0)) + min(max(d.x,d.y),0.0);"
   outputType=CMOT_Float1
   inputs [
    {inputName=p, input {expression=&self:MaterialExpressionConstantBiasScale_8}}
    {inputName=b, input {expression=&self:MaterialExpressionNamedRerouteUsage_13}}
   ]
  }
  NamedRerouteDeclaration(None=Multiply(A=SmoothStep(Max=@n8, Value=@n12), B=SmoothStep(Max=@n8, Value=@n14))) {name=line 5, nodeColor=<rgba 0.4766 0 1 1>}
  n15 = ScalarParameter() {defaultValue=0.01, parameterName=Sharpness}
  n16 = TextureCoordinate()
  n17 = Constant() {r=0.06}
  n18 = Constant() {r=0.22}
  MaterialExpressionConstantBiasScale_2 = ConstantBiasScale(None=Add(A=@n16, B=AppendVector(A=@n17, B=@n18))) {bias=-0.5, scale=2}
  MaterialExpressionNamedRerouteUsage_17 = NamedRerouteUsage()
  n19 = Custom(p=@MaterialExpressionConstantBiasScale_2, b=@MaterialExpressionNamedRerouteUsage_17) {
   code="float2 d = abs(p)-b;\r\n    return length(max(d,0.0)) + min(max(d.x,d.y),0.0);"
   outputType=CMOT_Float1
   inputs [
    {inputName=p, input {expression=&self:MaterialExpressionConstantBiasScale_2}}
    {inputName=b, input {expression=&self:MaterialExpressionNamedRerouteUsage_17}}
   ]
  }
  n20 = Multiply(A=@n18) {constB=-1}
  MaterialExpressionConstantBiasScale_12 = ConstantBiasScale(None=Add(A=@n16, B=AppendVector(A=@n17, B=@n20))) {bias=-0.5, scale=2}
  n21 = Custom(p=@MaterialExpressionConstantBiasScale_12, b=@MaterialExpressionNamedRerouteUsage_17) {
   code="float2 d = abs(p)-b;\r\n    return length(max(d,0.0)) + min(max(d.x,d.y),0.0);"
   outputType=CMOT_Float1
   inputs [
    {inputName=p, input {expression=&self:MaterialExpressionConstantBiasScale_12}}
    {inputName=b, input {expression=&self:MaterialExpressionNamedRerouteUsage_17}}
   ]
  }
  NamedRerouteDeclaration(None=Multiply(A=SmoothStep(Max=@n15, Value=@n19), B=SmoothStep(Max=@n15, Value=@n21) {constMax=0.01})) {name=line 4, nodeColor=<rgba 0.4766 0 1 1>}
  n22 = ScalarParameter() {defaultValue=0.01, parameterName=Sharpness}
  n23 = TextureCoordinate()
  n24 = Constant()
  n25 = Constant() {r=0.31}
  MaterialExpressionConstantBiasScale_3 = ConstantBiasScale(None=Add(A=@n23, B=AppendVector(A=@n24, B=@n25))) {bias=-0.5, scale=2}
  MaterialExpressionNamedRerouteUsage_16 = NamedRerouteUsage()
  n26 = Custom(p=@MaterialExpressionConstantBiasScale_3, b=@MaterialExpressionNamedRerouteUsage_16) {
   code="float2 d = abs(p)-b;\r\n    return length(max(d,0.0)) + min(max(d.x,d.y),0.0);"
   outputType=CMOT_Float1
   inputs [
    {inputName=p, input {expression=&self:MaterialExpressionConstantBiasScale_3}}
    {inputName=b, input {expression=&self:MaterialExpressionNamedRerouteUsage_16}}
   ]
  }
  n27 = Multiply(A=@n25) {constB=-1}
  MaterialExpressionConstantBiasScale_11 = ConstantBiasScale(None=Add(A=@n23, B=AppendVector(A=@n24, B=@n27))) {bias=-0.5, scale=2}
  n28 = Custom(p=@MaterialExpressionConstantBiasScale_11, b=@MaterialExpressionNamedRerouteUsage_16) {
   code="float2 d = abs(p)-b;\r\n    return length(max(d,0.0)) + min(max(d.x,d.y),0.0);"
   outputType=CMOT_Float1
   inputs [
    {inputName=p, input {expression=&self:MaterialExpressionConstantBiasScale_11}}
    {inputName=b, input {expression=&self:MaterialExpressionNamedRerouteUsage_16}}
   ]
  }
  NamedRerouteDeclaration(None=Multiply(A=SmoothStep(Max=@n22, Value=@n26), B=SmoothStep(Max=@n22, Value=@n28) {constMax=0.01})) {name=line 3, nodeColor=<rgba 0.4766 0 1 1>}
  n29 = ScalarParameter() {defaultValue=0.01, parameterName=Sharpness}
  n30 = TextureCoordinate()
  n31 = Constant() {r=-0.04}
  n32 = Constant() {r=0.3975}
  MaterialExpressionConstantBiasScale_4 = ConstantBiasScale(None=Add(A=@n30, B=AppendVector(A=@n31, B=@n32))) {bias=-0.5, scale=2}
  MaterialExpressionNamedRerouteUsage_15 = NamedRerouteUsage()
  n33 = Custom(p=@MaterialExpressionConstantBiasScale_4, b=@MaterialExpressionNamedRerouteUsage_15) {
   code="float2 d = abs(p)-b;\r\n    return length(max(d,0.0)) + min(max(d.x,d.y),0.0);"
   outputType=CMOT_Float1
   inputs [
    {inputName=p, input {expression=&self:MaterialExpressionConstantBiasScale_4}}
    {inputName=b, input {expression=&self:MaterialExpressionNamedRerouteUsage_15}}
   ]
  }
  n34 = Multiply(A=@n32) {constB=-1}
  MaterialExpressionConstantBiasScale_10 = ConstantBiasScale(None=Add(A=@n30, B=AppendVector(A=@n31, B=@n34))) {bias=-0.5, scale=2}
  n35 = Custom(p=@MaterialExpressionConstantBiasScale_10, b=@MaterialExpressionNamedRerouteUsage_15) {
   code="float2 d = abs(p)-b;\r\n    return length(max(d,0.0)) + min(max(d.x,d.y),0.0);"
   outputType=CMOT_Float1
   inputs [
    {inputName=p, input {expression=&self:MaterialExpressionConstantBiasScale_10}}
    {inputName=b, input {expression=&self:MaterialExpressionNamedRerouteUsage_15}}
   ]
  }
  NamedRerouteDeclaration(None=Multiply(A=SmoothStep(Max=@n29, Value=@n33), B=SmoothStep(Max=@n29, Value=@n35) {constMax=0.01})) {name=line 2, nodeColor=<rgba 0.4766 0 1 1>}
  n36 = ScalarParameter() {defaultValue=0.01, parameterName=Sharpness}
  n37 = TextureCoordinate()
  n38 = Constant() {r=-0.09}
  n39 = Constant() {r=0.485}
  MaterialExpressionConstantBiasScale_5 = ConstantBiasScale(None=Add(A=@n37, B=AppendVector(A=@n38, B=@n39))) {bias=-0.5, scale=2}
  MaterialExpressionNamedRerouteUsage_14 = NamedRerouteUsage()
  n40 = Custom(p=@MaterialExpressionConstantBiasScale_5, b=@MaterialExpressionNamedRerouteUsage_14) {
   code="float2 d = abs(p)-b;\r\n    return length(max(d,0.0)) + min(max(d.x,d.y),0.0);"
   outputType=CMOT_Float1
   inputs [
    {inputName=p, input {expression=&self:MaterialExpressionConstantBiasScale_5}}
    {inputName=b, input {expression=&self:MaterialExpressionNamedRerouteUsage_14}}
   ]
  }
  n41 = Multiply(A=@n39) {constB=-1}
  MaterialExpressionConstantBiasScale_9 = ConstantBiasScale(None=Add(A=@n37, B=AppendVector(A=@n38, B=@n41))) {bias=-0.5, scale=2}
  n42 = Custom(p=@MaterialExpressionConstantBiasScale_9, b=@MaterialExpressionNamedRerouteUsage_14) {
   code="float2 d = abs(p)-b;\r\n    return length(max(d,0.0)) + min(max(d.x,d.y),0.0);"
   outputType=CMOT_Float1
   inputs [
    {inputName=p, input {expression=&self:MaterialExpressionConstantBiasScale_9}}
    {inputName=b, input {expression=&self:MaterialExpressionNamedRerouteUsage_14}}
   ]
  }
  NamedRerouteDeclaration(None=Multiply(A=SmoothStep(Max=@n36, Value=@n40) {constMax=0.01}, B=SmoothStep(Max=@n36, Value=@n42) {constMax=0.01})) {name=line 1, nodeColor=<rgba 0.4766 0 1 1>}
  NamedRerouteDeclaration(None=AppendVector(A=Constant() {r=0.66}, B=Constant() {r=0.0005})) {name=BoxSize, nodeColor=<rgba 0 0.2734 1 1>}
  n43 = NamedRerouteUsage()
  n44 = NamedRerouteUsage()
  n45 = Multiply(A=Multiply(A=Multiply(A=@n43, B=@n44), B=NamedRerouteUsage()), B=NamedRerouteUsage())
  n46 = NamedRerouteUsage()
  OneMinus(None=Multiply(A=Multiply(A=@n45, B=@n46), B=NamedRerouteUsage()))
 ]
}
