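# Translucent UI material for a center decoration, drawing a procedural stack of thin horizontal separator lines.
# A thresholded absolute sine of the V coordinate forms repeating lines, with parameters for count, thickness and sharpness.
# The product of two sharpened offsets around V = 0.5 forms a mask that blanks a narrow middle band.
# Multiplying lines by mask leaves separators only above and below that gap, without any texture.
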Material /Tank/UI/Art/M_Ui_centerDeco
graph {
 properties {blendMode=BLEND_Translucent, materialDomain=MD_UI, pixelDepthOffsetMode=PDOM_Legacy, usedShadingModels=Default Lit}
 nodes [
  Constant() {r=1}
  n1 = TextureCoordinate()
  n2 = Multiply(A=ScalarParameter() {defaultValue=11, group=Separator, parameterName=Separator Number}) {constB=0.5}
  n3 = Abs(None=Sine(None=Multiply(A=@n1, B=@n2) {constB=10}))
  n4 = Saturate(None=Subtract(A=ComponentMask(None=@n3) {g=true}, B=ScalarParameter() {defaultValue=0.995, group=Separator, parameterName=Separators_Thickness}))
  n5 = ScalarParameter() {defaultValue=1024, group=Separator, parameterName=Separator_Sharpness}
  n6 = ScalarParameter() {defaultValue=1024, group=Separator, parameterName=Separator_Sharpness}
  n7 = Multiply(A=ComponentMask(None=Add(A=@n1) {constB=-0.52}) {g=true}, B=@n6) {constB=256}
  n8 = Multiply(A=ComponentMask(None=Add(A=@n1) {constB=-0.48}) {g=true}, B=@n6) {constB=256}
  Multiply(A=Saturate(None=DotProduct(A=@n7, B=@n8)), B=Saturate(None=Multiply(A=@n4, B=@n5) {constB=256}))
 ]
}
