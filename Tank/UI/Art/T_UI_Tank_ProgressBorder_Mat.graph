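# Additive UI material for a tank progress border, tinted from the channels of a single border texture.
# The red channel blends between two orange shades, and the blue channel lerps that colour toward white.
# A separate mask saturates the sum of red, green and blue, with blue scaled by the Stroke_Opacity parameter, which keeps the stroke faint by default.
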Material /Tank/UI/Art/T_UI_Tank_ProgressBorder_Mat
graph {
 properties {blendMode=BLEND_Additive, materialDomain=MD_UI, pixelDepthOffsetMode=PDOM_Legacy, usedShadingModels=Default Lit}
 nodes [
  n1 = TextureSample() {texture=&/T_UI_Tank_ProgressBorder}
  LinearInterpolate(
   A=LinearInterpolate(A=Constant4Vector() {constant=<rgba 1 0.112 0 1>}.RGBA, B=Constant4Vector() {constant=<rgba 1 0.5029 0 1>}.RGBA, Alpha=@n1.R)
   B=Constant() {r=1}
   Alpha=@n1.B
  )
  n2 = Multiply(A=@n1.B, B=ScalarParameter() {defaultValue=0.05, parameterName=Stroke_Opacity, sliderMax=1})
  Saturate(None=Add(A=@n1.R, B=Add(A=@n1.G, B=@n2)))
 ]
}
