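# Tank_HUD is the reticle HUD widget for the driveable tank, showing speed, fuel, turret pitch, distance, hull and turret rotation, and damage warnings.
# On bind it subscribes to tank engine, health and hatch-cover changes, while vehicle events update meters, silhouette and warnings through dynamic material parameters.
# UMG animations drive the intro, warning pop-ins, ADS and boost effects.
# Touch devices get adjusted layout, and gameplay-tag waits track the disabled state.
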WidgetBlueprint /Tank/UI/Tank_HUD
animations [
 Critical {
  binding_count=1
  length_seconds=0.4803
  object=&self:Critical
  sequence {bindings [CommonBorder_0 {
     id=#1
     possessed_class=CommonBorder
     tracks [{
       type=MovieSceneColorTrack
       display_name=Content Color and Opacity
       sections [{
         type=MovieSceneColorSection
         end_frame=9
         start_frame=0
         channels [
          Color.R {type=MovieSceneScriptingFloatChannel, keys [[0,0.8229],[9.6057,0.5365]]}
          Color.G {type=MovieSceneScriptingFloatChannel, keys [[0,0.0312],[9.6057,0.0203]]}
          Color.B {type=MovieSceneScriptingFloatChannel, const [0,2]}
          Color.A {type=MovieSceneScriptingFloatChannel, keys [[0,0.9],[9.6057,0.5]]}
         ]
        }]
      }]
    }]}
 }
 Intro {
  binding_count=16
  length_seconds=0.4167
  object=&self:Intro
  sequence {
   bindings [
    HBox_DecoSide {
     id=#2
     possessed_class=HorizontalBox
     tracks [{
       type=MovieSceneFloatTrack
       display_name=Render Opacity
       sections [{type=MovieSceneFloatSection, channels [None {type=MovieSceneScriptingFloatChannel, keys [[0,0],[15,1]]}]}]
      }]
    }
    CommonBorder_Distance {
     id=#3
     display_name=CommonBorder_FuelMeterContainer
     possessed_class=CommonBorder
     tracks [
      {
       type=MovieSceneColorTrack
       display_name=Content Color and Opacity
       sections [{
         type=MovieSceneColorSection
         channels [
          Color.R {type=MovieSceneScriptingFloatChannel}
          Color.G {type=MovieSceneScriptingFloatChannel}
          Color.B {type=MovieSceneScriptingFloatChannel}
          Color.A {type=MovieSceneScriptingFloatChannel, keys [[15,0],[22,1]]}
         ]
        }]
      }
      {
       type=MovieSceneColorTrack
       display_name=Brush Color
       sections [{
         type=MovieSceneColorSection
         channels [
          Color.R {type=MovieSceneScriptingFloatChannel}
          Color.G {type=MovieSceneScriptingFloatChannel}
          Color.B {type=MovieSceneScriptingFloatChannel}
          Color.A {type=MovieSceneScriptingFloatChannel, keys [[6,0],[11,1]]}
         ]
        }]
      }
     ]
    }
    SizeBox_AltMeter {
     id=#4
     possessed_class=SizeBox
     tracks [{
       type=MovieSceneFloatTrack
       display_name=Height Override
       sections [{type=MovieSceneFloatSection, channels [None {type=MovieSceneScriptingFloatChannel, keys [[11,14],[21,270]]}]}]
      }]
    }
    SizeBox_Speedometer {
     id=#5
     possessed_class=SizeBox
     tracks [{
       type=MovieSceneFloatTrack
       display_name=Height Override
       sections [{type=MovieSceneFloatSection, channels [None {type=MovieSceneScriptingFloatChannel, keys [[5,14],[15,370]]}]}]
      }]
    }
    CRichText_Pitch {
     id=#6
     display_name=CRichText_Speed
     possessed_class=CommonRichTextBlock
     tracks [
      {
       type=MovieSceneWidgetMaterialTrack
       display_name=Default Text Style Override.Font.FontMaterial
       sections [{
         type=MovieSceneParameterSection
         channels [
          RGBAR {type=MovieSceneScriptingFloatChannel, const [1,2]}
          RGBAG {type=MovieSceneScriptingFloatChannel, const [1,2]}
          RGBAB {type=MovieSceneScriptingFloatChannel, const [1,2]}
          RGBAA {type=MovieSceneScriptingFloatChannel, keys [[7,0],[15,1]]}
         ]
        }]
      }
      {
       type=MovieScene2DTransformTrack
       sections [{
         type=MovieScene2DTransformSection
         channels [
          Translation.X {type=MovieSceneScriptingFloatChannel, keys [[7,-16],[15,0]]}
          Translation.Y {type=MovieSceneScriptingFloatChannel}
         ]
        }]
      }
     ]
    }
    CommonBorder_Distance {
     id=#7
     possessed_class=CommonBorder
     tracks [
      {
       type=MovieSceneColorTrack
       display_name=Content Color and Opacity
       sections [{
         type=MovieSceneColorSection
         channels [
          Color.R {type=MovieSceneScriptingFloatChannel}
          Color.G {type=MovieSceneScriptingFloatChannel}
          Color.B {type=MovieSceneScriptingFloatChannel}
          Color.A {type=MovieSceneScriptingFloatChannel, keys [[9,0],[16,1]]}
         ]
        }]
      }
      {
       type=MovieSceneColorTrack
       display_name=Brush Color
       sections [{
         type=MovieSceneColorSection
         channels [
          Color.R {type=MovieSceneScriptingFloatChannel}
          Color.G {type=MovieSceneScriptingFloatChannel}
          Color.B {type=MovieSceneScriptingFloatChannel}
          Color.A {type=MovieSceneScriptingFloatChannel, keys [[0,0],[5,1]]}
         ]
        }]
      }
     ]
    }
    Border_Speedometer {
     id=#8
     possessed_class=CommonBorder
     tracks [
      {
       type=MovieSceneColorTrack
       display_name=Content Color and Opacity
       sections [{
         type=MovieSceneColorSection
         channels [
          Color.R {type=MovieSceneScriptingFloatChannel}
          Color.G {type=MovieSceneScriptingFloatChannel}
          Color.B {type=MovieSceneScriptingFloatChannel}
          Color.A {type=MovieSceneScriptingFloatChannel, keys [[9,0],[16,1]]}
         ]
        }]
      }
      {
       type=MovieSceneColorTrack
       display_name=Brush Color
       sections [{
         type=MovieSceneColorSection
         channels [
          Color.R {type=MovieSceneScriptingFloatChannel}
          Color.G {type=MovieSceneScriptingFloatChannel}
          Color.B {type=MovieSceneScriptingFloatChannel}
          Color.A {type=MovieSceneScriptingFloatChannel, keys [[0,0],[5,1]]}
         ]
        }]
      }
     ]
    }
    CommonBorder_Distance {
     id=#9
     display_name=CommonBorder_Pitch
     possessed_class=CommonBorder
     tracks [
      {
       type=MovieSceneColorTrack
       display_name=Content Color and Opacity
       sections [{
         type=MovieSceneColorSection
         channels [
          Color.R {type=MovieSceneScriptingFloatChannel}
          Color.G {type=MovieSceneScriptingFloatChannel}
          Color.B {type=MovieSceneScriptingFloatChannel}
          Color.A {type=MovieSceneScriptingFloatChannel, keys [[15,0],[22,1]]}
         ]
        }]
      }
      {
       type=MovieSceneColorTrack
       display_name=Brush Color
       sections [{
         type=MovieSceneColorSection
         channels [
          Color.R {type=MovieSceneScriptingFloatChannel}
          Color.G {type=MovieSceneScriptingFloatChannel}
          Color.B {type=MovieSceneScriptingFloatChannel}
          Color.A {type=MovieSceneScriptingFloatChannel, keys [[6,0],[11,1]]}
         ]
        }]
      }
     ]
    }
    CRichText_Pitch {
     id=#10
     display_name=CRichText_Distance
     possessed_class=CommonRichTextBlock
     tracks [
      {
       type=MovieSceneWidgetMaterialTrack
       display_name=Default Text Style Override.Font.FontMaterial
       sections [{
         type=MovieSceneParameterSection
         channels [
          RGBAR {type=MovieSceneScriptingFloatChannel, const [1,2]}
          RGBAG {type=MovieSceneScriptingFloatChannel, const [1,2]}
          RGBAB {type=MovieSceneScriptingFloatChannel, const [1,2]}
          RGBAA {type=MovieSceneScriptingFloatChannel, keys [[7,0],[15,1]]}
         ]
        }]
      }
      {
       type=MovieScene2DTransformTrack
       sections [{
         type=MovieScene2DTransformSection
         channels [
          Translation.X {type=MovieSceneScriptingFloatChannel, keys [[7,16],[15,0]]}
          Translation.Y {type=MovieSceneScriptingFloatChannel}
         ]
        }]
      }
     ]
    }
    SizeBox_LeftArc {
     id=#11
     display_name=SizeBox_RightArc
     possessed_class=SizeBox
     tracks [{
       type=MovieScene2DTransformTrack
       sections [{
         type=MovieScene2DTransformSection
         channels [
          Translation.X {type=MovieSceneScriptingFloatChannel, keys [[10,32],[25,0]]}
          Translation.Y {type=MovieSceneScriptingFloatChannel}
         ]
        }]
      }]
    }
    Overlay_TankSilhouette {
     id=#12
     possessed_class=Overlay
     tracks [{
       type=MovieSceneFloatTrack
       display_name=Render Opacity
       sections [{
         type=MovieSceneFloatSection
         end_frame=16
         start_frame=0
         channels [None {type=MovieSceneScriptingFloatChannel, keys [[0,0],[2,0.25],[3,0],[6,0.75],[9,0.1],[11,0.5],[12,0],[13,0],[16,1]]}]
        }]
      }]
    }
    CRichText_Pitch {
     id=#13
     possessed_class=CommonRichTextBlock
     tracks [
      {
       type=MovieSceneWidgetMaterialTrack
       display_name=Default Text Style Override.Font.FontMaterial
       sections [{
         type=MovieSceneParameterSection
         channels [
          RGBAR {type=MovieSceneScriptingFloatChannel, const [1,2]}
          RGBAG {type=MovieSceneScriptingFloatChannel, const [1,2]}
          RGBAB {type=MovieSceneScriptingFloatChannel, const [1,2]}
          RGBAA {type=MovieSceneScriptingFloatChannel, keys [[13,0],[21,1]]}
         ]
        }]
      }
      {
       type=MovieScene2DTransformTrack
       sections [{
         type=MovieScene2DTransformSection
         channels [
          Translation.X {type=MovieSceneScriptingFloatChannel, keys [[13,-16],[21,0]]}
          Translation.Y {type=MovieSceneScriptingFloatChannel}
         ]
        }]
      }
     ]
    }
    Image_Center_Crosshair {
     id=#14
     possessed_class=Image
     tracks [{
       type=MovieSceneWidgetMaterialTrack
       sections [{
         type=MovieSceneParameterSection
         end_frame=25
         start_frame=0
         channels [
          Intro {type=MovieSceneScriptingFloatChannel, keys [[5,0],[20,1]]}
          XOffset {type=MovieSceneScriptingFloatChannel, keys [[5,-0.25],[20,0]]}
         ]
        }]
      }]
    }
    SizeBox_LeftArc {
     id=#15
     possessed_class=SizeBox
     tracks [{
       type=MovieScene2DTransformTrack
       sections [{
         type=MovieScene2DTransformSection
         channels [
          Translation.X {type=MovieSceneScriptingFloatChannel, keys [[10,-32],[25,0]]}
          Translation.Y {type=MovieSceneScriptingFloatChannel}
         ]
        }]
      }]
    }
    SizeBox_AltMeter {
     id=#16
     display_name=SizeBox_Fuel
     possessed_class=SizeBox
     tracks [{
       type=MovieSceneFloatTrack
       display_name=Height Override
       sections [{type=MovieSceneFloatSection, channels [None {type=MovieSceneScriptingFloatChannel, keys [[11,14],[21,270]]}]}]
      }]
    }
    SizeBox_DistanceMeter {
     id=#17
     possessed_class=SizeBox
     tracks [{
       type=MovieSceneFloatTrack
       display_name=Height Override
       sections [{type=MovieSceneFloatSection, channels [None {type=MovieSceneScriptingFloatChannel, keys [[5,14],[15,370]]}]}]
      }]
    }
   ]
   display_rate {numerator=60}
  }
 }
 IntroLowFuel {
  binding_count=5
  length_seconds=0.75
  object=&self:IntroLowFuel
  sequence {
   bindings [
    Overlay_Warning {
     id=#18
     display_name=VerticalBox_LowFuel
     possessed_class=Overlay
     tracks [{
       type=MovieScene2DTransformTrack
       sections [{
         type=MovieScene2DTransformSection
         end_frame=45
         start_frame=0
         channels [
          Translation.X {type=MovieSceneScriptingFloatChannel, keys [[25,-200],[40,0]]}
          Translation.Y {type=MovieSceneScriptingFloatChannel}
          Angle {type=MovieSceneScriptingFloatChannel}
          Scale.X {type=MovieSceneScriptingFloatChannel}
          Scale.Y {type=MovieSceneScriptingFloatChannel}
          Shear.X {type=MovieSceneScriptingFloatChannel}
          Shear.Y {type=MovieSceneScriptingFloatChannel}
         ]
        }]
      }]
    }
    SizeBox_3 {
     id=#19
     possessed_class=SizeBox
     tracks [{
       type=MovieSceneFloatTrack
       display_name=Height Override
       sections [{type=MovieSceneFloatSection, channels [None {type=MovieSceneScriptingFloatChannel, keys [[15,0],[27,64]]}]}]
      }]
    }
    Image_Dot_Bottom {
     id=#20
     possessed_class=Image
     tracks [
      {
       type=MovieScene2DTransformTrack
       sections [{type=MovieScene2DTransformSection, channels [Translation.Y {type=MovieSceneScriptingFloatChannel, keys [[9,-4],[15,0]]}]}]
      }
      {
       type=MovieSceneFloatTrack
       display_name=Render Opacity
       sections [{type=MovieSceneFloatSection, channels [None {type=MovieSceneScriptingFloatChannel, keys [[0,0],[3,1],[5,0],[6,0],[9,1]]}]}]
      }
     ]
    }
    CommonTextBlock_Warning {
     id=#21
     possessed_class=CommonTextBlock
     tracks [{
       type=MovieSceneWidgetMaterialTrack
       display_name=Font.FontMaterial
       sections [{
         type=MovieSceneParameterSection
         channels [
          RGBAR {type=MovieSceneScriptingFloatChannel, const [1,1]}
          RGBAG {type=MovieSceneScriptingFloatChannel, const [1,1]}
          RGBAB {type=MovieSceneScriptingFloatChannel, const [1,1]}
          RGBAA {type=MovieSceneScriptingFloatChannel, keys [[25,0],[40,1]]}
         ]
        }]
      }]
    }
    Image_Dot_Top {
     id=#22
     possessed_class=Image
     tracks [
      {
       type=MovieScene2DTransformTrack
       sections [{type=MovieScene2DTransformSection, channels [Translation.Y {type=MovieSceneScriptingFloatChannel, keys [[9,4],[15,0]]}]}]
      }
      {
       type=MovieSceneFloatTrack
       display_name=Render Opacity
       sections [{type=MovieSceneFloatSection, channels [None {type=MovieSceneScriptingFloatChannel, keys [[9,0],[15,1]]}]}]
      }
     ]
    }
   ]
   display_rate {numerator=60}
  }
 }
 ToADS {
  binding_count=4
  length_seconds=0.25
  object=&self:ToADS
  sequence {bindings [
    Border_DecoSide_Lines_Left {
     id=#23
     possessed_class=Border
     tracks [{
       type=MovieSceneColorTrack
       display_name=Content Color and Opacity
       sections [{
         type=MovieSceneColorSection
         end_frame=5
         start_frame=0
         channels [
          Color.R {type=MovieSceneScriptingFloatChannel}
          Color.G {type=MovieSceneScriptingFloatChannel}
          Color.B {type=MovieSceneScriptingFloatChannel}
          Color.A {type=MovieSceneScriptingFloatChannel, keys [[0,0.075],[5,0.15]]}
         ]
        }]
      }]
    }
    Image_Center_Scale {
     id=#24
     possessed_class=Image
     tracks [{
       type=MovieSceneColorTrack
       display_name=Color and Opacity
       sections [{
         type=MovieSceneColorSection
         channels [
          Color.R {type=MovieSceneScriptingFloatChannel}
          Color.G {type=MovieSceneScriptingFloatChannel}
          Color.B {type=MovieSceneScriptingFloatChannel}
          Color.A {type=MovieSceneScriptingFloatChannel, keys [[0,0.05],[5,0.1]]}
         ]
        }]
      }]
    }
    Image_Center_Crosshair {
     id=#25
     possessed_class=Image
     tracks [
      {
       type=MovieSceneColorTrack
       display_name=Color and Opacity
       sections [{
         type=MovieSceneColorSection
         channels [
          Color.R {type=MovieSceneScriptingFloatChannel}
          Color.G {type=MovieSceneScriptingFloatChannel}
          Color.B {type=MovieSceneScriptingFloatChannel}
          Color.A {type=MovieSceneScriptingFloatChannel, keys [[0,0.075],[5,0.2]]}
         ]
        }]
      }
      {
       type=MovieSceneWidgetMaterialTrack
       sections [{
         type=MovieSceneParameterSection
         channels [
          Line SIxe X {type=MovieSceneScriptingFloatChannel, keys [[0,0.63],[5,0.66]]}
          CenterLine_X {type=MovieSceneScriptingFloatChannel, keys [[0,0.5213],[5,0.6413]]}
          Top_Bot_X {type=MovieSceneScriptingFloatChannel, keys [[0,-0.1547],[5,-0.0947]]}
         ]
        }]
      }
     ]
    }
    Border_DecoSide_Lines_Right {
     id=#26
     possessed_class=Border
     tracks [{
       type=MovieSceneColorTrack
       display_name=Content Color and Opacity
       sections [{
         type=MovieSceneColorSection
         end_frame=5
         start_frame=0
         channels [
          Color.R {type=MovieSceneScriptingFloatChannel}
          Color.G {type=MovieSceneScriptingFloatChannel}
          Color.B {type=MovieSceneScriptingFloatChannel}
          Color.A {type=MovieSceneScriptingFloatChannel, keys [[0,0.075],[5,0.15]]}
         ]
        }]
      }]
    }
   ]}
 }
 Boost {
  binding_count=2
  length_seconds=0.25
  object=&self:Boost
  sequence {bindings [
    HBox_SpeedLines {
     id=#27
     possessed_class=HorizontalBox
     tracks [
      {
       type=MovieSceneFloatTrack
       display_name=Render Opacity
       sections [{type=MovieSceneFloatSection, channels [None {type=MovieSceneScriptingFloatChannel, keys [[0,0],[5,1]]}]}]
      }
      {
       type=MovieScene2DTransformTrack
       sections [{
         type=MovieScene2DTransformSection
         channels [
          Translation.X {type=MovieSceneScriptingFloatChannel}
          Translation.Y {type=MovieSceneScriptingFloatChannel}
          Angle {type=MovieSceneScriptingFloatChannel}
          Scale.X {type=MovieSceneScriptingFloatChannel}
          Scale.Y {type=MovieSceneScriptingFloatChannel, keys [[0,0],[5,1]]}
          Shear.X {type=MovieSceneScriptingFloatChannel}
          Shear.Y {type=MovieSceneScriptingFloatChannel}
         ]
        }]
      }
     ]
    }
    Overlay_BoostSparks {
     id=#28
     possessed_class=Overlay
     tracks [{
       type=MovieSceneFloatTrack
       display_name=Render Opacity
       sections [{type=MovieSceneFloatSection, channels [None {type=MovieSceneScriptingFloatChannel, keys [[0,0],[5,0.25]]}]}]
      }]
    }
   ]}
 }
 IntroWarning {
  binding_count=4
  length_seconds=4.0834
  object=&self:IntroWarning
  sequence {
   bindings [
    Overlay_Warning {
     id=#18
     display_name=VerticalBox_8
     possessed_class=Overlay
     tracks [{
       type=MovieScene2DTransformTrack
       sections [{
         type=MovieScene2DTransformSection
         channels [
          Translation.X {type=MovieSceneScriptingFloatChannel, keys [[22,512],[52.002,0],[202.002,0],[230.001,512]]}
          Translation.Y {type=MovieSceneScriptingFloatChannel}
          Angle {type=MovieSceneScriptingFloatChannel}
          Scale.X {type=MovieSceneScriptingFloatChannel}
          Scale.Y {type=MovieSceneScriptingFloatChannel}
          Shear.X {type=MovieSceneScriptingFloatChannel}
          Shear.Y {type=MovieSceneScriptingFloatChannel}
         ]
        }]
      }]
    }
    SizeBox_3 {
     id=#19
     display_name=SizeBox
     possessed_class=SizeBox
     tracks [{
       type=MovieSceneFloatTrack
       display_name=Height Override
       sections [{type=MovieSceneFloatSection, channels [None {type=MovieSceneScriptingFloatChannel, keys [[15,0],[27,64],[227.001,64],[238.001,0]]}]}]
      }]
    }
    Image_Dot_Bottom {
     id=#20
     display_name=Image_Dot_Bottom_1
     possessed_class=Image
     tracks [
      {
       type=MovieScene2DTransformTrack
       sections [{
         type=MovieScene2DTransformSection
         channels [Translation.Y {type=MovieSceneScriptingFloatChannel, keys [[9,-4],[15,0],[238.002,0],[240.001,-4]]}]
        }]
      }
      {
       type=MovieSceneFloatTrack
       display_name=Render Opacity
       sections [{
         type=MovieSceneFloatSection
         channels [None {type=MovieSceneScriptingFloatChannel, keys [[0,0],[3,1],[5,0],[6,0],[9,1],[238.002,1],[240.001,0]]}]
        }]
      }
     ]
    }
    Image_Dot_Top {
     id=#22
     display_name=Image_Dot_Top_1
     possessed_class=Image
     tracks [
      {
       type=MovieScene2DTransformTrack
       sections [{
         type=MovieScene2DTransformSection
         channels [Translation.Y {type=MovieSceneScriptingFloatChannel, keys [[9,4],[15,0],[238.001,0],[240.001,4]]}]
        }]
      }
      {
       type=MovieSceneFloatTrack
       display_name=Render Opacity
       sections [{type=MovieSceneFloatSection, channels [None {type=MovieSceneScriptingFloatChannel, keys [[9,0],[15,1],[243.001,1],[245.001,0]]}]}]
      }
     ]
    }
   ]
   display_rate {numerator=60}
  }
 }
 IntroDanger {
  binding_count=8
  length_seconds=0.65
  object=&self:IntroDanger
  sequence {
   bindings [
    SizeBox_3 {
     id=#19
     display_name=SizeBox_4
     possessed_class=SizeBox
     tracks [{
       type=MovieSceneFloatTrack
       display_name=Width Override
       sections [{type=MovieSceneFloatSection, channels [None {type=MovieSceneScriptingFloatChannel, keys [[15,0],[27,100]]}]}]
      }]
    }
    SizeBox_3 {
     id=#29
     display_name=SizeBox_5
     possessed_class=SizeBox
     tracks [{
       type=MovieSceneFloatTrack
       display_name=Width Override
       sections [{type=MovieSceneFloatSection, channels [None {type=MovieSceneScriptingFloatChannel, keys [[15,0],[27,100]]}]}]
      }]
    }
    SizeBox_6 {
     id=#30
     possessed_class=SizeBox
     tracks [{
       type=MovieSceneFloatTrack
       display_name=Height Override
       sections [{type=MovieSceneFloatSection, channels [None {type=MovieSceneScriptingFloatChannel, keys [[27,0],[27.001,0],[39.001,80]]}]}]
      }]
    }
    VBox_LowFuel_3 {
     id=#31
     possessed_class=HorizontalBox
     tracks [{
       type=MovieScene2DTransformTrack
       sections [{
         type=MovieScene2DTransformSection
         channels [
          Translation.X {type=MovieSceneScriptingFloatChannel}
          Translation.Y {type=MovieSceneScriptingFloatChannel, keys [[25,2],[27,0]]}
         ]
        }]
      }]
    }
    Image_Dot_Bottom {
     id=#20
     display_name=Image_Dot_Bottom_2
     possessed_class=Image
     tracks [
      {
       type=MovieSceneFloatTrack
       display_name=Render Opacity
       sections [{type=MovieSceneFloatSection, channels [None {type=MovieSceneScriptingFloatChannel, keys [[0,0],[3,1],[5,0],[6,0],[9,1]]}]}]
      }
      {
       type=MovieScene2DTransformTrack
       sections [{
         type=MovieScene2DTransformSection
         channels [
          Translation.X {type=MovieSceneScriptingFloatChannel, keys [[9,-4],[15,0]]}
          Translation.Y {type=MovieSceneScriptingFloatChannel}
         ]
        }]
      }
     ]
    }
    Image_Dot_Bottom {
     id=#32
     display_name=Image_Dot_Bottom_3
     possessed_class=Image
     tracks [
      {
       type=MovieSceneFloatTrack
       display_name=Render Opacity
       sections [{type=MovieSceneFloatSection, channels [None {type=MovieSceneScriptingFloatChannel, keys [[0,0],[3,1],[5,0],[6,0],[9,1]]}]}]
      }
      {
       type=MovieScene2DTransformTrack
       sections [{
         type=MovieScene2DTransformSection
         channels [
          Translation.X {type=MovieSceneScriptingFloatChannel, keys [[9,-4],[15,0]]}
          Translation.Y {type=MovieSceneScriptingFloatChannel}
         ]
        }]
      }
     ]
    }
    Image_Dot_Top {
     id=#33
     display_name=Image_Dot_Top_3
     possessed_class=Image
     tracks [
      {
       type=MovieSceneFloatTrack
       display_name=Render Opacity
       sections [{type=MovieSceneFloatSection, channels [None {type=MovieSceneScriptingFloatChannel, keys [[9,0],[15,1]]}]}]
      }
      {
       type=MovieScene2DTransformTrack
       sections [{
         type=MovieScene2DTransformSection
         channels [
          Translation.X {type=MovieSceneScriptingFloatChannel, keys [[9,3.3321],[15,0]]}
          Translation.Y {type=MovieSceneScriptingFloatChannel}
         ]
        }]
      }
     ]
    }
    Image_Dot_Top {
     id=#22
     display_name=Image_Dot_Top_2
     possessed_class=Image
     tracks [
      {
       type=MovieSceneFloatTrack
       display_name=Render Opacity
       sections [{type=MovieSceneFloatSection, channels [None {type=MovieSceneScriptingFloatChannel, keys [[9,0],[15,1]]}]}]
      }
      {
       type=MovieScene2DTransformTrack
       sections [{
         type=MovieScene2DTransformSection
         channels [
          Translation.X {type=MovieSceneScriptingFloatChannel, keys [[9,3.3321],[15,0]]}
          Translation.Y {type=MovieSceneScriptingFloatChannel}
         ]
        }]
      }
     ]
    }
   ]
   display_rate {numerator=60}
  }
 }
 IntroNoFuel {
  binding_count=1
  length_seconds=0.4167
  object=&self:IntroNoFuel
  sequence {
   bindings [Overlay_Warning {
     id=#18
     display_name=VerticalBox_NoFuel
     possessed_class=Overlay
     tracks [{
       type=MovieScene2DTransformTrack
       sections [{
         type=MovieScene2DTransformSection
         channels [
          Translation.X {type=MovieSceneScriptingFloatChannel, keys [[0,-200],[10,-200],[25,0]]}
          Translation.Y {type=MovieSceneScriptingFloatChannel}
          Angle {type=MovieSceneScriptingFloatChannel}
          Scale.X {type=MovieSceneScriptingFloatChannel}
          Scale.Y {type=MovieSceneScriptingFloatChannel}
          Shear.X {type=MovieSceneScriptingFloatChannel}
          Shear.Y {type=MovieSceneScriptingFloatChannel}
         ]
        }]
      }]
    }]
   display_rate {numerator=60}
  }
 }
]
graphs [
 DriverCoverEnabledChanged {nodes [
   n1 = DriverCoverEnabledChanged!() -> (then, Active:bool=~) {is=title K2Node_FunctionEntry}
   > n2 = GetDynamicMaterial(self=Get Image_Hatch() -> (Image_Hatch:Image Object Reference))
   > SetScalarParameterValue(self=@n2, ParameterName=WarningOn, Value=Select(Option 0:double=1, Option 1:double=0, Index=@n1) -> (ReturnValue:double))
   > n3 = Branch(Condition=@n1)
   SetVisibility(execute=@n3.else, self=Get HorizontalBox_Warning() -> (HorizontalBox_Warning:Horizontal Box Object Reference), InVisibility=HitTestInvisible)
   > PlayAnimationForward(InAnimation=Get IntroWarning() -> (IntroWarning:Widget Animation Object Reference))
   > "SetText (Text)"(
    self=[
     Get Text_WarningText_Backing() -> (Text_WarningText_Backing:Common Text Object Reference)
     Get Text_Warningtext() -> (Text_Warningtext:Common Text Object Reference)
    ]
    InText="LOCTEXT(\"Hatch Destroyed\")"
   )
  ]}
 EngineEnabledChanged {nodes [
   n1 = EngineEnabledChanged!() -> (then, Active:bool=~) {is=title K2Node_FunctionEntry}
   > n2 = GetDynamicMaterial(self=Get Image_LeftEngine() -> (Image_LeftEngine:Image Object Reference))
   n3 = Select(Option 0=1, Option 1=0, Index=@n1)
   > SetScalarParameterValue(self=@n2, ParameterName=WarningOn, Value=@n3)
   > n4 = GetDynamicMaterial(self=Get Image_RightEngine() -> (Image_RightEngine:Image Object Reference))
   > SetScalarParameterValue(self=@n4, ParameterName=WarningOn, Value=@n3)
  ]}
 EventGraph {nodes [
   n1 = Event BindTo!() -> (OutputDelegate, then, AssociatedObject:Object Reference) {is=title K2Node_Event}
   > n2 = Cast To Actor!(execute, Object=@n1.AssociatedObject) -> (then, CastFailed, AsActor:Actor Object Reference) {is=title K2Node_DynamicCast}
   > SetupVehicleFromBP!(execute, self:Tank Vehicle Reticle Object Reference, NewVehicle:Actor Object Reference=@n2) -> (then)
   > n3 = Sequence()
   > n4 = Cast To Tank_Vehicle!(execute, Object=@n2) -> (then, CastFailed, AsTank Vehicle:Tank Vehicle Object Reference) {is=title K2Node_DynamicCast}
   > Bind Event to Engine Enabled Changed Broadcast(self:Tank Vehicle Object Reference=@n4, Delegate=Create Event()) {is=title K2Node_AddDelegate}
   > EngineEnabledChanged!(execute, self, Active:bool=NOT Boolean(A=Get EngineOverheated!(self:Tank Vehicle Object Reference=@n4) -> (EngineOverheated:bool))) -> (then) {is=title K2Node_CallFunction}
   > SetVisibility(self=Get HorizontalBox_Danger() -> (HorizontalBox_Danger:Vertical Box Object Reference), InVisibility=Collapsed)
   > Bind Event to On Health Changed(self:Fort Athena Vehicle Object Reference=@n4, Delegate=Create Event()) {is=title K2Node_AddDelegate}
   > UpdateCriticalHealth()
   > Bind Event to Driver Cover Enabled Changed Broadcast(self:Tank Vehicle Object Reference=@n4, Delegate=Create Event()) {is=title K2Node_AddDelegate}
   > DriverCoverEnabledChanged!(execute, self, Active:bool=NOT Boolean(A=Get DriverCoverOnCooldown!(self:Tank Vehicle Object Reference=@n4) -> (DriverCoverOnCooldown:bool))) -> (then) {is=title K2Node_CallFunction}
   > UpdateFuelMeter!(execute, self, InFuel:double=1) -> (then)
   n5 = Get OwningTankVehicle() -> (OwningTankVehicle:Fort Tank Vehicle Object Reference)
   > Is Valid(InputObject=@n5)
   > n6 = Branch(Condition=HasInfiniteFuel!(self:Fort Vehicle Interface Interface=@n5) -> (ReturnValue:bool))
   n7 = Get CRichText_Infinite() -> (CRichText_Infinite:Common Rich Text Block Object Reference)
   > SetVisibility(self=@n7, InVisibility=HitTestInvisible)
   > UpdateNoFuel!(execute, self, InNoFuel:bool=false) -> (then)
   > UpdateLowFuel!(execute, self, InLowFuel:bool=false) -> (then)
   SetVisibility(execute=@n6.else, self=@n7, InVisibility=Collapsed)
   n8 = Get OwningTankVehicle()
   n9 = IsVehicleFuelEmpty!(self:Fort Vehicle Interface Interface=@n8) -> (ReturnValue:bool)
   > UpdateNoFuel(InNoFuel=@n9)
   > n10 = Branch(Condition=@n9)
   UpdateLowFuel(execute=@n10.else, InLowFuel=IsVehicleFuelLow!(self:Fort Vehicle Interface Interface=@n8) -> (ReturnValue:bool))
   "SetText (Text)"(
    execute=@n3.then_1
    self=Get Text_MinPitch() -> (Text_MinPitch:Common Text Object Reference)
    InText="To Text (Float)"(
     Value=Get MinTurretPitch() -> (MinTurretPitch:float)
     MinimumIntegralDigits=1
     MaximumIntegralDigits=324
     MinimumFractionalDigits=0
     MaximumFractionalDigits=0
    )
   )
   > "SetText (Text)"(
    self=Get Text_MaxPitch() -> (Text_MaxPitch:Common Text Object Reference)
    InText="To Text (Float)"(
     Value=Get MaxTurretPitch() -> (MaxTurretPitch:float)
     MinimumIntegralDigits=1
     MaximumIntegralDigits=324
     MinimumFractionalDigits=0
     MaximumFractionalDigits=0
    )
   )
   Event OnTankTargetingPressed() {is=title K2Node_Event}
   n11 = Get ToADS() -> (ToADS:Widget Animation Object Reference)
   > PlayAnimationForward(InAnimation=@n11)
   > UpdateADS!(execute, self, IsADS:bool=true) -> (then)
   Event OnTankTargetingReleased() {is=title K2Node_Event}
   > PlayAnimationReverse(InAnimation=@n11, PlaybackSpeed=4)
   > UpdateADS(IsADS=false)
   Event OnBoostStarted() {is=title K2Node_Event}
   > UpdateBoostState!(execute, self, State:EHoagieBoostState Enum=Started) -> (then)
   n12 = Get Boost() -> (Boost:Widget Animation Object Reference)
   > PlayAnimationForward(InAnimation=@n12)
   Event OnBoostFinished() {is=title K2Node_Event}
   > UpdateBoostState(State=Finished)
   > PlayAnimationReverse(InAnimation=@n12, PlaybackSpeed=2)
   n13 = Event UnbindFrom!() -> (OutputDelegate, then, AssociatedObject:Object Reference) {is=title K2Node_Event}
   > Parent: UnbindFrom!(execute, AssociatedObject:Object Reference=@n13.AssociatedObject) -> (then) {is=title K2Node_CallParentFunction}
   > UpdateADS(IsADS=false)
   n14 = Event OnTankDistanceUpdate!() -> (OutputDelegate, then, Distance:float, DistancePercent:float) {is=title K2Node_Event}
   > n15 = GetDynamicMaterial(self=Get Image_DistanceMeter() -> (Image_DistanceMeter:Image Object Reference))
   > SetScalarParameterValue(self=@n15, ParameterName=Progress, Value=@n14.DistancePercent)
   > n16 = GetDynamicMaterial(self=Get Image_DistanceMeter_Backing() -> (Image_DistanceMeter_Backing:Image Object Reference))
   > SetScalarParameterValue(self=@n16, ParameterName=Progress, Value=@n14.DistancePercent)
   n17 = Truncate to Integer64!(A:double=@n14.Distance) -> (ReturnValue:Integer64=0)
   > SetText(
    self=Get CRichText_Distance() -> (CRichText_Distance:Common Rich Text Block Object Reference)
    InText=Select(
     Option 0:"Text (by ref)"="To Text (Integer64)"!(Value:Integer64=@n17, bAlwaysSign:bool=false, bUseGrouping:bool=true, MinimumIntegralDigits:int=1, MaximumIntegralDigits:int=324) -> (ReturnValue:Text)
     Option 1:"Text (by ref)"=Get MaxDistanceText() -> (MaxDistanceText:Text)
     Index="Equal (Float)"(A=@n14.DistancePercent, B=1)
    ) -> (ReturnValue:"Text (by ref)")
   )
   Event OnOutOfFuel() {is=title K2Node_Event}
   > UpdateNoFuel(InNoFuel=true)
   n18 = Event OnTankEngineHealthChanged!() -> (OutputDelegate, then, EngineHealth:float, EngineHealthMax:float, bNewlyDisabled:bool, bNewlyEnabled:bool) {is=title K2Node_Event}
   n19 = Get EngineWarningThresholdPercent() -> (EngineWarningThresholdPercent:double)
   > n20 = Branch(Condition="float > float"(A=@n18.EngineHealth, B=float * float(A=@n18.EngineHealthMax, B=@n19) {is=title K2Node_PromotableOperator}))
   > Branch(Condition=Get AlreadyEngineCriticalHealth() -> (AlreadyEngineCriticalHealth:bool))
   n21 = Get HorizontalBox_Danger()
   > SetVisibility(self=@n21, InVisibility=Collapsed)
   > Set AlreadyEngineCriticalHealth(AlreadyEngineCriticalHealth:bool=false) -> (Output_Get:bool)
   n22 = Branch(execute=@n20.else, Condition=Get AlreadyEngineCriticalHealth())
   SetVisibility(execute=@n22.else, self=@n21, InVisibility=HitTestInvisible)
   > PlayAnimationForward(InAnimation=Get IntroDanger() -> (IntroDanger:Widget Animation Object Reference))
   > Set AlreadyEngineCriticalHealth(AlreadyEngineCriticalHealth=true)
   Event OnLowFuel() {is=title K2Node_Event}
   > UpdateLowFuel(InLowFuel=true)
   n23 = Event OnVehicleUpdated!() -> (OutputDelegate, then, NewVehicle:Fort Athena Vehicle Object Reference) {is=title K2Node_Event}
   n24 = Get DisabledTag() -> (DisabledTag:Gameplay Tag Structure)
   > n25 = WaitGameplayTagAddToActor!(execute:Exec, TargetActor:Actor Object Reference=@n23.NewVehicle, Tag:Gameplay Tag Structure=@n24, OnlyTriggerOnce:bool=false) -> (then:Exec, AsyncTaskProxy:Ability Async Wait Gameplay Tag Added Object Reference, Added:Exec) {is=title K2Node_AsyncAction}
   > n26 = WaitGameplayTagRemoveFromActor!(execute:Exec, TargetActor:Actor Object Reference=@n23.NewVehicle, Tag:Gameplay Tag Structure=@n24, OnlyTriggerOnce:bool=false) -> (then:Exec, AsyncTaskProxy:Ability Async Wait Gameplay Tag Removed Object Reference, Removed:Exec) {is=title K2Node_AsyncAction}
   OnVehicleEnabled!(execute=@n26.Removed, self, Enabled:bool=true) -> (then)
   OnVehicleEnabled(execute=@n25.Added, Enabled=false)
   n27 = Event OnTreadEnabledChanged!() -> (OutputDelegate, then, bEnabled:bool, Side:ETreadSide Enum=Invalid) {is=title K2Node_Event}
   > n28 = GetDynamicMaterial(self=Select!(
     Invalid:Image Object Reference
     Left:Image Object Reference=Get Image_LeftTread() -> (Image_LeftTread:Image Object Reference)
     Right:Image Object Reference=Get Image_RightTread() -> (Image_RightTread:Image Object Reference)
     Index:ETreadSide Enum=@n27.Side
    ) -> (ReturnValue:Image Object Reference))
   > SetScalarParameterValue(self=@n28, ParameterName=WarningOn, Value=Select!(Option 0:double=1, Option 1:double=0, Index:bool=@n27.bEnabled) -> (ReturnValue:double))
   > n29 = Branch(Condition=@n27.bEnabled)
   n30 = Get HorizontalBox_Warning()
   > SetVisibility(self=@n30, InVisibility=Collapsed)
   SetVisibility(execute=@n29.else, self=@n30, InVisibility=HitTestInvisible)
   > PlayAnimationForward(InAnimation=Get IntroWarning())
   > "SetText (Text)"(self=[Get Text_WarningText_Backing(), Get Text_Warningtext()], InText="LOCTEXT(\"Tread Destroyed\")")
   n31 = Event Construct()
   n32 = Get Intro()
   StopAnimation(execute=[@n31.then, @n45.then], InAnimation=@n32)
   > PlayAnimation(InAnimation=@n32, PlaybackSpeed=0.75)
   > UpdateSpedometer!(execute, self, NormalizedSpeed:double=0, RawSpeed:double=0) -> (then)
   > SetRenderOpacity(self=Get HBox_SpeedLines() -> (HBox_SpeedLines:Horizontal Box Object Reference), InOpacity=0)
   > n33 = Branch(Condition="float > float"(A=GetNumericCVar!(CVar:String=Fort.TankDistanceMeterEnabled) -> (OutValue:float, ReturnValue:bool)))
   n34 = Get CRichText_Distance()
   > SetVisibility(self=@n34, InVisibility=HitTestInvisible)
   SetVisibility(execute=@n33.else, self=@n34, InVisibility=Collapsed)
   n35 = Event OnTankRotationValueChanged!() -> (OutputDelegate, then, TankValue:float, TurretValue:float) {is=title K2Node_Event}
   > SetRenderTransformAngle!(
    execute
    self:Widget Object Reference=Get Overlay_TankSilhouette() -> (Overlay_TankSilhouette:Overlay Object Reference)
    Angle:float=@n35.TankValue
   ) -> (then)
   > SetRenderTransformAngle(self=Get Overlay_Tank_Turret() -> (Overlay_Tank_Turret:Overlay Object Reference), Angle=@n35.TurretValue)
   n36 = Event OnTankTurretAimPitchValueChanged!() -> (OutputDelegate, then, Pitch:float, PitchPercent:float) {is=title K2Node_Event}
   > n37 = GetDynamicMaterial(self=Get Image_AltMeter() -> (Image_AltMeter:Image Object Reference))
   > SetScalarParameterValue(self=@n37, ParameterName=Progress, Value=@n36.PitchPercent)
   > n38 = GetDynamicMaterial(self=Get Image_AltMeter_Backing() -> (Image_AltMeter_Backing:Image Object Reference))
   > SetScalarParameterValue(self=@n38, ParameterName=Progress, Value=@n36.PitchPercent)
   n39 = Floor to Integer64!(A:double=@n36.Pitch) -> (ReturnValue:Integer64=0)
   > SetText(
    self=Get CRichText_Pitch() -> (CRichText_Pitch:Common Rich Text Block Object Reference)
    InText=Format Text!(
     Format="INVTEXT(\"{angle}°\")"
     angle:Text="To Text (Integer64)"(Value=@n39, bAlwaysSign=false, bUseGrouping=true, MinimumIntegralDigits=1, MaximumIntegralDigits=324)
    ) -> (Result)
   )
   Event OnInitialized()
   > Branch(Condition=IsUsingTouch(PlayerController=GetOwningPlayer()))
   > SetPadding!(
    execute
    self:Overlay Slot Object Reference=SlotAsOverlaySlot(Widget=Get Overlay_TankSilhouette())
    InPadding_Left:float=0
    InPadding_Top:float=0
    InPadding_Right:float=320
    InPadding_Bottom:float=0
   ) -> (then)
   > SetPadding(self=SlotAsOverlaySlot(Widget=Get HorizontalBox_Warning()), InPadding_Left=0, InPadding_Top=0, InPadding_Right=320, InPadding_Bottom=0)
   n40 = Get HorizontalBox_Danger()
   > SetRenderTranslation!(execute, self=@n40, Translation_X:double=0, Translation_Y:double=350) -> (then)
   > SetVerticalAlignment(self=SlotAsOverlaySlot(Widget=@n40), InVerticalAlignment=VAlign_Center)
   n41 = GetOwningPlayer()
   n42 = BreakVector2D(InVec=GetViewportSize!() -> (ReturnValue:Vector 2D Structure))
   n43 = float / float(A=@n42, B=@n42.Y)
   > Branch(Condition=AND Boolean(
     A=IsUsingTouch(PlayerController=@n41)
     B="In Range (Float)"!(Value:double=@n43, Min:double=1.3333, Max:double=2, InclusiveMin:bool=true, InclusiveMax:bool=true) -> (ReturnValue:bool)
    ))
   > SetSize!(
    execute
    self:Spacer Object Reference=Get Spacer_DecoSide() -> (Spacer_DecoSide:Spacer Object Reference)
    InSize:Vector 2D Structure=(X=900,Y=0)
   ) -> (then)
   > SetPadding(
    self=SlotAsOverlaySlot(Widget=Get Image_Horizon() -> (Image_Horizon:Image Object Reference))
    InPadding_Left=130
    InPadding_Top=0
    InPadding_Right=130
    InPadding_Bottom=0
   )
   > SetRenderTranslation(self=Get HBox_LowFuel() -> (HBox_LowFuel:Horizontal Box Object Reference), Translation_X=490, Translation_Y=0)
   n44 = Event OnWidgetVisibilityChanged!() -> (OutputDelegate, then, bNewVisibility:bool) {is=title K2Node_Event}
   > Update HUDElement Visibility!(execute, self, IsTankHUDVisible:bool=@n44.bNewVisibility) -> (then)
   > n45 = Branch(Condition=@n44.bNewVisibility)
   n46 = Event OnFuelValueChanged!() -> (OutputDelegate, then, FuelPercentage:float, bIsLowOnFuel:bool) {is=title K2Node_Event}
   > UpdateFuelMeter(InFuel=@n46.FuelPercentage)
   n47 = Get OwningTankVehicle()
   > Is Valid(InputObject=@n47)
   n48 = IsVehicleFuelLow(self=@n47)
   n49 = IsVehicleFuelEmpty(self=@n47)
   > Branch(Condition=AND Boolean(A=NOT Boolean(A=@n48), B=NOT Boolean(A=@n49)))
   > SetVisibility(self=[Get HBox_LowFuel(), Get Fuel_Low() -> (Fuel_Low:Image Object Reference)], InVisibility=Collapsed)
   n50 = Event UpdateSpeedometer!() -> (OutputDelegate, then, NormalizedSpeed:float, RawSpeed:float) {is=title K2Node_Event}
   > UpdateSpedometer(NormalizedSpeed=@n50.NormalizedSpeed, RawSpeed=@n50.RawSpeed)
  ]}
 OnVehicleEnabled {nodes [
   n1 = OnVehicleEnabled!() -> (then, Enabled:bool=~) {is=title K2Node_FunctionEntry}
   > n2 = Branch(Condition=@n1)
   n3 = Get WBP_Vehicle_Warning() -> (WBP_Vehicle_Warning:WBP Vehicle Warning Object Reference)
   > RemoveWarning!(execute, self:WBP Vehicle Warning Object Reference=@n3, Type:ENUM_WarningType Enum=NewEnumerator2) -> (then)
   > SetVisibility(self=@n3, InVisibility=Collapsed)
   ShowWarning!(execute=@n2.else, self:WBP Vehicle Warning Object Reference=@n3, Type:ENUM_WarningType Enum=NewEnumerator2) -> (then)
   > SetVisibility(self=@n3, InVisibility=HitTestInvisible)
  ]}
 Update HUDElement Visibility {nodes [
   n1 = Update HUDElement Visibility!() -> (then, IsTankHUDVisible:bool=~) {is=title K2Node_FunctionEntry}
   > SetHUDElementVisibility!(
    execute
    self:Fort HUDContext Object Reference=Get FortHUDContext!() -> (ReturnValue:Fort HUDContext Object Reference) {is=title K2Node_GetBlueprintContextEx}
    HUDElementTags:"Gameplay Tag Container Structure (by ref)"=MakeLiteralGameplayTagContainer!(Value:Gameplay Tag Container Structure=(GameplayTags=((TagName="HUD.Athena.KillFeed")))) -> (ReturnValue:Gameplay Tag Container Structure)
    bHideElements:bool=@n1
    ReasonName:"Name (by ref)"=None
   ) -> (then)
  ]}
 UpdateADS {nodes [
   n1 = UpdateADS!() -> (then, IsADS:bool=~) {is=title K2Node_FunctionEntry}
   > SetHUDElementVisibility(
    self=Get FortHUDContext()
    HUDElementTags=MakeLiteralGameplayTagContainer(Value=(GameplayTags=((TagName="HUD.TopRight"),(TagName="HUD.TopLeft.VoteUpdateNotification"),(TagName="HUD.TopLeft.TeamPlayers"),(TagName="HUD.TopLeft.PlayerInfo"),(TagName="HUD.TopLeft.PlayerHitPoints"),(TagName="HUD.TopLeft.PickupManager"),(TagName="HUD.Athena.MapPackKeys"))))
    bHideElements=@n1
    ReasonName=None
   )
  ]}
 UpdateBoostState {nodes [
   n1 = UpdateBoostState!() -> (then, State:EHoagieBoostState Enum=Unknown) {is=title K2Node_FunctionEntry}
   > n2 = GetDynamicMaterial(self=Get Image_FuelMeter() -> (Image_FuelMeter:Image Object Reference))
   > SetScalarParameterValue(
    self=@n2
    ParameterName=BoostOn
    Value=Select!(Unknown:double=0, Ready:double=0, Started:double=1, Finished:double=0, Failed:double=0, Index:EHoagieBoostState Enum=@n1) -> (ReturnValue)
   )
   SetVisibility(
    self=Get Overlay_BoostSparks() -> (Overlay_BoostSparks:Overlay Object Reference)
    InVisibility=Select(
     Unknown:ESlateVisibility Enum=Collapsed
     Ready:ESlateVisibility Enum=Collapsed
     Started:ESlateVisibility Enum=HitTestInvisible
     Finished:ESlateVisibility Enum=Collapsed
     Failed:ESlateVisibility Enum=Collapsed
     Index=@n1
    ) -> (ReturnValue:ESlateVisibility Enum)
   )
  ]}
 UpdateCriticalHealth {nodes [
   UpdateCriticalHealth() {is=title K2Node_FunctionEntry}
   n1 = Get OwningTankVehicle()
   > Is Valid(InputObject=@n1)
   > n2 = Branch(Condition=IsCriticalHealth!(self:Fort Athena Vehicle Object Reference=@n1) -> (ReturnValue:bool))
   n3 = Get Critical() -> (Critical:Widget Animation Object Reference)
   > n4 = Branch(Condition=IsAnimationPlaying(InAnimation=@n3))
   PlayAnimation(execute=@n4.else, InAnimation=@n3)
   Branch(execute=@n2.else, Condition=IsAnimationPlaying(InAnimation=Get Critical()))
   > StopAnimation(InAnimation=@n3)
  ]}
 UpdateFuelMeter {nodes [
   n1 = UpdateFuelMeter!() -> (then, InFuel:double=~) {is=title K2Node_FunctionEntry}
   > n2 = GetDynamicMaterial(self=Get Image_FuelMeter())
   > SetScalarParameterValue(self=@n2, ParameterName=Progress, Value=@n1)
   > n3 = GetDynamicMaterial(self=Get Image_FuelMeter_Backing() -> (Image_FuelMeter_Backing:Image Object Reference))
   > SetScalarParameterValue(self=@n3, ParameterName=Progress, Value=@n1)
  ]}
 UpdateLowFuel {nodes [
   n1 = UpdateLowFuel!() -> (then, InLowFuel:bool=~) {is=title K2Node_FunctionEntry}
   > n2 = Branch(Condition=@n1)
   > SetVisibility(self=[Get HBox_LowFuel(), Get Fuel_Low()], InVisibility=Visible)
   > SetActiveWidgetIndex(self=Get Switcher_Fuel() -> (Switcher_Fuel:Common Widget Switcher Legacy Object Reference), Index=0)
   > PlayAnimationForward(InAnimation=Get IntroLowFuel() -> (IntroLowFuel:Widget Animation Object Reference))
   > SetVisibility(self=Get Overlay_BoostSparks(), InVisibility=Collapsed)
   SetVisibility(execute=@n2.else, self=[Get HBox_LowFuel(), Get Fuel_Low()], InVisibility=Collapsed)
  ]}
 UpdateNoFuel {nodes [
   n1 = UpdateNoFuel!() -> (then, InNoFuel:bool=~) {is=title K2Node_FunctionEntry}
   > n2 = Branch(Condition=@n1)
   > SetVisibility(self=[Get HBox_LowFuel(), Get Fuel_Low()], InVisibility=Visible)
   > SetActiveWidgetIndex(self=Get Switcher_Fuel(), Index=1)
   > PlayAnimationForward(InAnimation=Get IntroNoFuel() -> (IntroNoFuel:Widget Animation Object Reference))
   > SetVisibility(self=Get Overlay_BoostSparks(), InVisibility=HitTestInvisible)
   SetVisibility(execute=@n2.else, self=[Get HBox_LowFuel(), Get Fuel_Low()], InVisibility=Collapsed)
  ]}
 UpdateSpedometer {nodes [
   n1 = UpdateSpedometer!() -> (then, NormalizedSpeed:double=~, RawSpeed:double=~) {is=title K2Node_FunctionEntry}
   > n2 = GetDynamicMaterial(self=Get Image_Speedometer() -> (Image_Speedometer:Image Object Reference))
   > SetScalarParameterValue(self=@n2, ParameterName=Progress, Value=@n1)
   > n3 = GetDynamicMaterial(self=Get Image_Speedometer_Backing() -> (Image_Speedometer_Backing:Image Object Reference))
   > SetScalarParameterValue(self=@n3, ParameterName=Progress, Value=@n1)
   > SetText(
    self=Get CRichText_Speed() -> (CRichText_Speed:Common Rich Text Block Object Reference)
    InText="To Text (Integer64)"(
     Value=Floor to Integer64(A=@n1.RawSpeed) -> (ReturnValue=0)
     bAlwaysSign=false
     bUseGrouping=true
     MinimumIntegralDigits=1
     MaximumIntegralDigits=324
    )
   )
  ]}
]
variables [
 AlreadyEngineCriticalHealth {type=boolean}
 AltitudeMeterMID {type=MaterialInstanceDynamic}
 DisabledTag {type=GameplayTag, default {tagName=Vehicle.Disabled}}
 EngineWarningThresholdPercent {type=number, default=0.3}
 HorizonMID {type=MaterialInstanceDynamic}
 MaxDistanceText {type=string, default=150+}
 TankVehicle {type=Tank_Vehicle_C, default=null}
]
widgets {
 info {inheritedWidgetCount=10, parentClass=/Script/TankUI.TankVehicleReticle, rootWidgetClass=&Overlay, widgetCount=175}
 widgets [
  Overlay_8 {widgetClassPath=&Overlay, widget=Overlay_8}
  HBox_DecoSide {
   widgetClassPath=&HorizontalBox
   parent=Overlay_8
   widget=HBox_DecoSide
   slot {class=OverlaySlot, properties {horizontalAlignment=HAlign_Center, verticalAlignment=VAlign_Center}}
  }
  SizeBox_LeftArc {
   widgetClassPath=&SizeBox
   parent=HBox_DecoSide
   widget=SizeBox_LeftArc
   properties {bOverride_HeightOverride=true, heightOverride=600}
   slot {class=HorizontalBoxSlot, properties {verticalAlignment=VAlign_Center}}
  }
  VerticalBox_5 {widgetClassPath=&VerticalBox, parent=SizeBox_LeftArc, widget=VerticalBox_5, slot {class=SizeBoxSlot}}
  Image_294 {
   widgetClassPath=&Image
   parent=VerticalBox_5
   widget=Image_294
   properties {
    brush {
     bIsDynamicallyLoaded=false
     drawAs=Image
     imageType=NoImage
     mirroring=NoMirror
     resourceName=None
     resourceObject=/Tank/UI/Art/M_UI_Tank_Corner
     tiling=NoTile
     imageSize=<32 32>
     margin=<ltrb 0 0 0 0>
     outlineSettings {
      bUseBrushTransparency=false
      roundingType=HalfHeightRadius
      width=0
      color {colorUseRule=UseColor_Specified, specifiedColor=<rgba 0 0 0 0>}
      cornerRadii=<0 0 0 0>
     }
     tintColor {colorUseRule=UseColor_Specified, specifiedColor=<rgba 1 1 1 1>}
     uVRegion {bIsValid=false, max=<0 0>, min=<0 0>}
    }
    renderTransform {angle=0, scale=<2 2>, shear=<0 0>, translation=<24 8>}
   }
   slot {class=VerticalBoxSlot, properties {horizontalAlignment=HAlign_Left, verticalAlignment=VAlign_Top}}
  }
  Border_DecoSide_Lines_Left {
   widgetClassPath=&Border
   parent=VerticalBox_5
   widget=Border_DecoSide_Lines_Left
   properties {
    background {
     bIsDynamicallyLoaded=false
     drawAs=NoDrawType
     imageType=NoImage
     mirroring=NoMirror
     resourceName=None
     resourceObject=null
     tiling=NoTile
     imageSize=<32 32>
     margin=<ltrb 0 0 0 0>
     outlineSettings {
      bUseBrushTransparency=false
      roundingType=HalfHeightRadius
      width=0
      color {colorUseRule=UseColor_Specified, specifiedColor=<rgba 0 0 0 0>}
      cornerRadii=<0 0 0 0>
     }
     tintColor {colorUseRule=UseColor_Specified, specifiedColor=<rgba 1 1 1 1>}
     uVRegion {bIsValid=false, max=<0 0>, min=<0 0>}
    }
    contentColorAndOpacity=<rgba 1 1 1 0.075>
    padding=<ltrb 0 0 0 0>
   }
   slot {class=VerticalBoxSlot, properties {verticalAlignment=VAlign_Center, size {sizeRule=Fill, value=1}}}
  }
  DecoSide_Lines_Left {
   widgetClassPath=&Image
   bIsVariable=true
   parent=Border_DecoSide_Lines_Left
   widget=DecoSide_Lines_Left
   properties {brush {
     bIsDynamicallyLoaded=false
     drawAs=Image
     imageType=NoImage
     mirroring=NoMirror
     resourceName=None
     resourceObject=/Tank/UI/Art/M_UI_Tank_SideDeco
     tiling=NoTile
     imageSize=<32 340>
     margin=<ltrb 0 0 0 0>
     outlineSettings {
      bUseBrushTransparency=false
      roundingType=HalfHeightRadius
      width=0
      color {colorUseRule=UseColor_Specified, specifiedColor=<rgba 0 0 0 0>}
      cornerRadii=<0 0 0 0>
     }
     tintColor {colorUseRule=UseColor_Specified, specifiedColor=<rgba 1 1 1 1>}
     uVRegion {bIsValid=false, max=<0 0>, min=<0 0>}
    }}
   slot {class=BorderSlot, properties {padding=<ltrb 0 0 0 0>}}
  }
  Image_30 {
   widgetClassPath=&Image
   parent=VerticalBox_5
   widget=Image_30
   properties {
    brush {
     bIsDynamicallyLoaded=false
     drawAs=Image
     imageType=NoImage
     mirroring=NoMirror
     resourceName=None
     resourceObject=&M_UI_Tank_Corner
     tiling=NoTile
     imageSize=<32 32>
     margin=<ltrb 0 0 0 0>
     outlineSettings {
      bUseBrushTransparency=false
      roundingType=HalfHeightRadius
      width=0
      color {colorUseRule=UseColor_Specified, specifiedColor=<rgba 0 0 0 0>}
      cornerRadii=<0 0 0 0>
     }
     tintColor {colorUseRule=UseColor_Specified, specifiedColor=<rgba 1 1 1 1>}
     uVRegion {bIsValid=false, max=<0 0>, min=<0 0>}
    }
    renderTransform {angle=0, scale=<2 -2>, shear=<0 0>, translation=<24 -8>}
   }
   slot {class=VerticalBoxSlot, properties {horizontalAlignment=HAlign_Left, verticalAlignment=VAlign_Bottom}}
  }
  Spacer_DecoSide {
   widgetClassPath=&Spacer
   bIsVariable=true
   parent=HBox_DecoSide
   widget=Spacer_DecoSide
   properties {size=<1172 1>}
   slot {class=HorizontalBoxSlot}
  }
  SizeBox_RightArc {
   widgetClassPath=&SizeBox
   parent=HBox_DecoSide
   widget=SizeBox_RightArc
   properties {bOverride_HeightOverride=true, heightOverride=600, renderTransform {angle=0, scale=<-1 1>, shear=<0 0>, translation=<0 0>}}
   slot {class=HorizontalBoxSlot}
  }
  VerticalBox_6 {widgetClassPath=&VerticalBox, parent=SizeBox_RightArc, widget=VerticalBox_6, slot {class=SizeBoxSlot}}
  Image_32 {
   widgetClassPath=&Image
   parent=VerticalBox_6
   widget=Image_32
   properties {
    brush {
     bIsDynamicallyLoaded=false
     drawAs=Image
     imageType=NoImage
     mirroring=NoMirror
     resourceName=None
     resourceObject=&M_UI_Tank_Corner
     tiling=NoTile
     imageSize=<32 32>
     margin=<ltrb 0 0 0 0>
     outlineSettings {
      bUseBrushTransparency=false
      roundingType=HalfHeightRadius
      width=0
      color {colorUseRule=UseColor_Specified, specifiedColor=<rgba 0 0 0 0>}
      cornerRadii=<0 0 0 0>
     }
     tintColor {colorUseRule=UseColor_Specified, specifiedColor=<rgba 1 1 1 1>}
     uVRegion {bIsValid=false, max=<0 0>, min=<0 0>}
    }
    renderTransform {angle=0, scale=<2 2>, shear=<0 0>, translation=<24 8>}
   }
   slot {class=VerticalBoxSlot, properties {horizontalAlignment=HAlign_Left, verticalAlignment=VAlign_Top}}
  }
  Border_DecoSide_Lines_Right {
   widgetClassPath=&Border
   parent=VerticalBox_6
   widget=Border_DecoSide_Lines_Right
   properties {
    background {
     bIsDynamicallyLoaded=false
     drawAs=NoDrawType
     imageType=NoImage
     mirroring=NoMirror
     resourceName=None
     resourceObject=null
     tiling=NoTile
     imageSize=<32 32>
     margin=<ltrb 0 0 0 0>
     outlineSettings {
      bUseBrushTransparency=false
      roundingType=HalfHeightRadius
      width=0
      color {colorUseRule=UseColor_Specified, specifiedColor=<rgba 0 0 0 0>}
      cornerRadii=<0 0 0 0>
     }
     tintColor {colorUseRule=UseColor_Specified, specifiedColor=<rgba 1 1 1 1>}
     uVRegion {bIsValid=false, max=<0 0>, min=<0 0>}
    }
    contentColorAndOpacity=<rgba 1 1 1 0.075>
    padding=<ltrb 0 0 0 0>
   }
   slot {class=VerticalBoxSlot, properties {horizontalAlignment=HAlign_Center, verticalAlignment=VAlign_Center, size {sizeRule=Fill, value=1}}}
  }
  DecoSide_Lines_Right {
   widgetClassPath=&Image
   bIsVariable=true
   parent=Border_DecoSide_Lines_Right
   widget=DecoSide_Lines_Right
   properties {brush {
     bIsDynamicallyLoaded=false
     drawAs=Image
     imageType=NoImage
     mirroring=NoMirror
     resourceName=None
     resourceObject=&M_UI_Tank_SideDeco
     tiling=NoTile
     imageSize=<32 340>
     margin=<ltrb 0 0 0 0>
     outlineSettings {
      bUseBrushTransparency=false
      roundingType=HalfHeightRadius
      width=0
      color {colorUseRule=UseColor_Specified, specifiedColor=<rgba 0 0 0 0>}
      cornerRadii=<0 0 0 0>
     }
     tintColor {colorUseRule=UseColor_Specified, specifiedColor=<rgba 1 1 1 1>}
     uVRegion {bIsValid=false, max=<0 0>, min=<0 0>}
    }}
   slot {class=BorderSlot, properties {padding=<ltrb 0 0 0 0>}}
  }
  Image_31 {
   widgetClassPath=&Image
   parent=VerticalBox_6
   widget=Image_31
   properties {
    brush {
     bIsDynamicallyLoaded=false
     drawAs=Image
     imageType=NoImage
     mirroring=NoMirror
     resourceName=None
     resourceObject=&M_UI_Tank_Corner
     tiling=NoTile
     imageSize=<32 32>
     margin=<ltrb 0 0 0 0>
     outlineSettings {
      bUseBrushTransparency=false
      roundingType=HalfHeightRadius
      width=0
      color {colorUseRule=UseColor_Specified, specifiedColor=<rgba 0 0 0 0>}
      cornerRadii=<0 0 0 0>
     }
     tintColor {colorUseRule=UseColor_Specified, specifiedColor=<rgba 1 1 1 1>}
     uVRegion {bIsValid=false, max=<0 0>, min=<0 0>}
    }
    renderTransform {angle=0, scale=<2 -2>, shear=<0 0>, translation=<24 -8>}
   }
   slot {class=VerticalBoxSlot, properties {horizontalAlignment=HAlign_Left, verticalAlignment=VAlign_Bottom}}
  }
  Overlay_3 {
   widgetClassPath=&Overlay
   parent=Overlay_8
   widget=Overlay_3
   slot {class=OverlaySlot, properties {horizontalAlignment=HAlign_Center, verticalAlignment=VAlign_Center}}
  }
  Image_Center_Scale {
   widgetClassPath=&Image
   bIsVariable=true
   parent=Overlay_3
   widget=Image_Center_Scale
   properties {
    brush {
     bIsDynamicallyLoaded=false
     drawAs=Image
     imageType=NoImage
     mirroring=NoMirror
     resourceName=None
     resourceObject=/Tank/UI/Art/M_Ui_centerDeco
     tiling=NoTile
     imageSize=<20 335>
     margin=<ltrb 0 0 0 0>
     outlineSettings {
      bUseBrushTransparency=false
      roundingType=HalfHeightRadius
      width=0
      color {colorUseRule=UseColor_Specified, specifiedColor=<rgba 0 0 0 0>}
      cornerRadii=<0 0 0 0>
     }
     tintColor {colorUseRule=UseColor_Specified, specifiedColor=<rgba 1 1 1 1>}
     uVRegion {bIsValid=false, max=<0 0>, min=<0 0>}
    }
    colorAndOpacity=<rgba 1 1 1 0.05>
   }
   slot {class=OverlaySlot, properties {horizontalAlignment=HAlign_Center, verticalAlignment=VAlign_Center}}
  }
  Image_Center_Crosshair {
   widgetClassPath=&Image
   bIsVariable=true
   parent=Overlay_3
   widget=Image_Center_Crosshair
   properties {
    brush {
     bIsDynamicallyLoaded=false
     drawAs=Image
     imageType=NoImage
     mirroring=NoMirror
     resourceName=None
     resourceObject=/Tank/UI/Art/M_UI_CenterReticle_Tank_Inst
     tiling=NoTile
     imageSize=<320 320>
     margin=<ltrb 0 0 0 0>
     outlineSettings {
      bUseBrushTransparency=false
      roundingType=HalfHeightRadius
      width=0
      color {colorUseRule=UseColor_Specified, specifiedColor=<rgba 0 0 0 0>}
      cornerRadii=<0 0 0 0>
     }
     tintColor {colorUseRule=UseColor_Specified, specifiedColor=<rgba 1 1 1 1>}
     uVRegion {bIsValid=false, max=<0 0>, min=<0 0>}
    }
    colorAndOpacity=<rgba 1 1 1 0.075>
   }
   slot {class=OverlaySlot, properties {horizontalAlignment=HAlign_Center, verticalAlignment=VAlign_Center}}
  }
  CommonBorder_0 {
   widgetClassPath=&CommonBorder
   parent=Overlay_8
   widget=CommonBorder_0
   properties {padding=<ltrb 0 0 0 0>}
   slot {class=OverlaySlot, properties {horizontalAlignment=HAlign_Center, verticalAlignment=VAlign_Center}}
  }
  HorizontalBox_0 {
   widgetClassPath=&HorizontalBox
   parent=CommonBorder_0
   widget=HorizontalBox_0
   slot {class=BorderSlot, properties {padding=<ltrb 0 0 0 0>}}
  }
  Border_1 {
   widgetClassPath=&Border
   parent=HorizontalBox_0
   widget=Border_1
   properties {
    background {
     bIsDynamicallyLoaded=false
     drawAs=NoDrawType
     imageType=NoImage
     mirroring=NoMirror
     resourceName=None
     resourceObject=null
     tiling=NoTile
     imageSize=<32 32>
     margin=<ltrb 0 0 0 0>
     outlineSettings {
      bUseBrushTransparency=false
      roundingType=HalfHeightRadius
      width=0
      color {colorUseRule=UseColor_Specified, specifiedColor=<rgba 0 0 0 0>}
      cornerRadii=<0 0 0 0>
     }
     tintColor {colorUseRule=UseColor_Specified, specifiedColor=<rgba 1 1 1 1>}
     uVRegion {bIsValid=false, max=<0 0>, min=<0 0>}
    }
    padding=<ltrb 0 0 0 0>
   }
   slot {class=HorizontalBoxSlot, properties {horizontalAlignment=HAlign_Center, verticalAlignment=VAlign_Center}}
  }
  SizeBox_Pitch {
   widgetClassPath=&SizeBox
   parent=Border_1
   widget=SizeBox_Pitch
   properties {bOverride_WidthOverride=true}
   slot {class=BorderSlot, properties {padding=<ltrb 0 0 0 0>}}
  }
  CanvasPanel_1 {widgetClassPath=&CanvasPanel, parent=SizeBox_Pitch, widget=CanvasPanel_1, slot {class=SizeBoxSlot}}
  VerticalBox_9 {
   widgetClassPath=&VerticalBox
   parent=CanvasPanel_1
   widget=VerticalBox_9
   slot {
    class=CanvasPanelSlot
    properties {bAutoSize=true, layoutData {alignment=<1 0.5>, anchors {maximum=<0 0.5>, minimum=<0 0.5>}, offsets=<ltrb 0 0 0 0>}}
   }
  }
  CRichText_Pitch {
   widgetClassPath=&CommonRichTextBlock
   bIsVariable=true
   parent=VerticalBox_9
   widget=CRichText_Pitch
   properties {
    bOverrideDefaultStyle=true
    justification=Right
    lineHeightPercentage=0.85
    text="XX°"
    textTransformPolicy=ToUpper
    defaultTextStyleOverride {
     overflowPolicy=Clip
     transformPolicy=None
     colorAndOpacity {colorUseRule=UseColor_Specified, specifiedColor=<rgba 1 1 1 1>}
     font {
      bForceMonospaced=false
      bMaterialIsStencil=false
      fontMaterial=/Tank/UI/Art/M_Ui_Additive_White
      fontObject=&BurbankSmall
      letterSpacing=0
      monospacedWidth=1
      size=16
      skewAmount=0
      typefaceFontName=Bold
      outlineSettings {
       bApplyOutlineToDropShadows=false
       bMiteredCorners=false
       bSeparateFillAlpha=false
       outlineMaterial=null
       outlineSize=0
       outlineColor=<rgba 0 0 0 1>
      }
     }
     highlightColor {colorUseRule=UseColor_Specified, specifiedColor=<rgba 0 0 0 1>}
     highlightShape {
      bIsDynamicallyLoaded=false
      drawAs=Image
      imageType=NoImage
      mirroring=NoMirror
      resourceName=None
      resourceObject=null
      tiling=NoTile
      imageSize=<32 32>
      margin=<ltrb 0 0 0 0>
      outlineSettings {
       bUseBrushTransparency=false
       roundingType=HalfHeightRadius
       width=0
       color {colorUseRule=UseColor_Specified, specifiedColor=<rgba 0 0 0 0>}
       cornerRadii=<0 0 0 0>
      }
      tintColor {colorUseRule=UseColor_Specified, specifiedColor=<rgba 1 1 1 1>}
      uVRegion {bIsValid=false, max=<0 0>, min=<0 0>}
     }
     selectedBackgroundColor {colorUseRule=UseColor_Foreground, specifiedColor=<rgba 1 0 1 1>}
     shadowColorAndOpacity=<rgba 0 0 0 1>
     shadowOffset=<0 0>
     strikeBrush {
      bIsDynamicallyLoaded=false
      drawAs=Image
      imageType=NoImage
      mirroring=NoMirror
      resourceName=None
      resourceObject=null
      tiling=NoTile
      imageSize=<32 32>
      margin=<ltrb 0 0 0 0>
      outlineSettings {
       bUseBrushTransparency=false
       roundingType=HalfHeightRadius
       width=0
       color {colorUseRule=UseColor_Specified, specifiedColor=<rgba 0 0 0 0>}
       cornerRadii=<0 0 0 0>
      }
      tintColor {colorUseRule=UseColor_Specified, specifiedColor=<rgba 1 1 1 1>}
      uVRegion {bIsValid=false, max=<0 0>, min=<0 0>}
     }
     underlineBrush {
      bIsDynamicallyLoaded=false
      drawAs=Image
      imageType=NoImage
      mirroring=NoMirror
      resourceName=None
      resourceObject=null
      tiling=NoTile
      imageSize=<32 32>
      margin=<ltrb 0 0 0 0>
      outlineSettings {
       bUseBrushTransparency=false
       roundingType=HalfHeightRadius
       width=0
       color {colorUseRule=UseColor_Specified, specifiedColor=<rgba 0 0 0 0>}
       cornerRadii=<0 0 0 0>
      }
      tintColor {colorUseRule=UseColor_Specified, specifiedColor=<rgba 1 1 1 1>}
      uVRegion {bIsValid=false, max=<0 0>, min=<0 0>}
     }
    }
   }
   slot {class=VerticalBoxSlot, properties {horizontalAlignment=HAlign_Right}}
  }
  CRichText_PitchAngle {
   widgetClassPath=&CommonRichTextBlock
   bIsVariable=true
   parent=VerticalBox_9
   widget=CRichText_PitchAngle
   properties {
    bOverrideDefaultStyle=true
    justification=Right
    lineHeightPercentage=0.85
    text=Angle
    textTransformPolicy=ToUpper
    defaultTextStyleOverride {
     overflowPolicy=Clip
     transformPolicy=None
     colorAndOpacity {colorUseRule=UseColor_Specified, specifiedColor=<rgba 1 0.2542 0 1>}
     font {
      bForceMonospaced=false
      bMaterialIsStencil=false
      fontMaterial=&M_Ui_Additive_White
      fontObject=&BurbankSmall
      letterSpacing=0
      monospacedWidth=1
      size=10
      skewAmount=0
      typefaceFontName=Bold
      outlineSettings {
       bApplyOutlineToDropShadows=false
       bMiteredCorners=false
       bSeparateFillAlpha=false
       outlineMaterial=null
       outlineSize=0
       outlineColor=<rgba 0 0 0 1>
      }
     }
     highlightColor {colorUseRule=UseColor_Specified, specifiedColor=<rgba 0 0 0 1>}
     highlightShape {
      bIsDynamicallyLoaded=false
      drawAs=Image
      imageType=NoImage
      mirroring=NoMirror
      resourceName=None
      resourceObject=null
      tiling=NoTile
      imageSize=<32 32>
      margin=<ltrb 0 0 0 0>
      outlineSettings {
       bUseBrushTransparency=false
       roundingType=HalfHeightRadius
       width=0
       color {colorUseRule=UseColor_Specified, specifiedColor=<rgba 0 0 0 0>}
       cornerRadii=<0 0 0 0>
      }
      tintColor {colorUseRule=UseColor_Specified, specifiedColor=<rgba 1 1 1 1>}
      uVRegion {bIsValid=false, max=<0 0>, min=<0 0>}
     }
     selectedBackgroundColor {colorUseRule=UseColor_Foreground, specifiedColor=<rgba 1 0 1 1>}
     shadowColorAndOpacity=<rgba 0 0 0 1>
     shadowOffset=<0 0>
     strikeBrush {
      bIsDynamicallyLoaded=false
      drawAs=Image
      imageType=NoImage
      mirroring=NoMirror
      resourceName=None
      resourceObject=null
      tiling=NoTile
      imageSize=<32 32>
      margin=<ltrb 0 0 0 0>
      outlineSettings {
       bUseBrushTransparency=false
       roundingType=HalfHeightRadius
       width=0
       color {colorUseRule=UseColor_Specified, specifiedColor=<rgba 0 0 0 0>}
       cornerRadii=<0 0 0 0>
      }
      tintColor {colorUseRule=UseColor_Specified, specifiedColor=<rgba 1 1 1 1>}
      uVRegion {bIsValid=false, max=<0 0>, min=<0 0>}
     }
     underlineBrush {
      bIsDynamicallyLoaded=false
      drawAs=Image
      imageType=NoImage
      mirroring=NoMirror
      resourceName=None
      resourceObject=null
      tiling=NoTile
      imageSize=<32 32>
      margin=<ltrb 0 0 0 0>
      outlineSettings {
       bUseBrushTransparency=false
       roundingType=HalfHeightRadius
       width=0
       color {colorUseRule=UseColor_Specified, specifiedColor=<rgba 0 0 0 0>}
       cornerRadii=<0 0 0 0>
      }
      tintColor {colorUseRule=UseColor_Specified, specifiedColor=<rgba 1 1 1 1>}
      uVRegion {bIsValid=false, max=<0 0>, min=<0 0>}
     }
    }
   }
   slot {class=VerticalBoxSlot, properties {horizontalAlignment=HAlign_Right}}
  }
  VerticalBox_10 {
   widgetClassPath=&VerticalBox
   parent=HorizontalBox_0
   widget=VerticalBox_10
   slot {class=HorizontalBoxSlot, properties {horizontalAlignment=HAlign_Right, verticalAlignment=VAlign_Center, padding=<ltrb 0 0 4 0>}}
  }
  Text_MaxPitch {
   widgetClassPath=&CommonTextBlock
   bIsVariable=true
   parent=VerticalBox_10
   widget=Text_MaxPitch
   properties {
    text=40
    colorAndOpacity {colorUseRule=UseColor_Specified, specifiedColor=<rgba 1 0.3328 0 1>}
    font {
     bForceMonospaced=false
     bMaterialIsStencil=false
     fontMaterial=/Tank/UI/Art/M_Ui_Additive_White_025
     fontObject=&BurbankSmall
     letterSpacing=0
     monospacedWidth=1
     size=12
     skewAmount=0
     typefaceFontName=Bold
     outlineSettings {
      bApplyOutlineToDropShadows=false
      bMiteredCorners=false
      bSeparateFillAlpha=false
      outlineMaterial=null
      outlineSize=0
      outlineColor=<rgba 0 0 0 1>
     }
    }
    shadowOffset=<0 0>
   }
   slot {class=VerticalBoxSlot, properties {horizontalAlignment=HAlign_Center, verticalAlignment=VAlign_Center, padding=<ltrb -32 0 -32 0>}}
  }
  CommonBorder_Pitch {
   widgetClassPath=&CommonBorder
   parent=VerticalBox_10
   widget=CommonBorder_Pitch
   properties {
    background {
     bIsDynamicallyLoaded=false
     drawAs=Box
     imageType=NoImage
     mirroring=NoMirror
     resourceName=None
     resourceObject=/Tank/UI/Art/T_UI_Tank_ProgressBorder_Mat
     tiling=NoTile
     imageSize=<32 32>
     margin=<ltrb 0.33 0.33 0.33 0.33>
     outlineSettings {
      bUseBrushTransparency=false
      roundingType=HalfHeightRadius
      width=0
      color {colorUseRule=UseColor_Specified, specifiedColor=<rgba 0 0 0 0>}
      cornerRadii=<0 0 0 0>
     }
     tintColor {colorUseRule=UseColor_Specified, specifiedColor=<rgba 1 1 1 1>}
     uVRegion {bIsValid=false, max=<0 0>, min=<0 0>}
    }
    padding=<ltrb 10 10 10 10>
   }
   slot {class=VerticalBoxSlot}
  }
  SizeBox_AltMeter {
   widgetClassPath=&SizeBox
   parent=CommonBorder_Pitch
   widget=SizeBox_AltMeter
   properties {bOverride_HeightOverride=true, heightOverride=270}
   slot {class=BorderSlot, properties {padding=<ltrb 10 10 10 10>}}
  }
  Overlay_9 {widgetClassPath=&Overlay, parent=SizeBox_AltMeter, widget=Overlay_9, slot {class=SizeBoxSlot}}
  Image_AltMeter_Backing {
   widgetClassPath=&Image
   bIsVariable=true
   parent=Overlay_9
   widget=Image_AltMeter_Backing
   properties {
    bIsEnabled=false
    visibility=HitTestInvisible
    brush {
     bIsDynamicallyLoaded=false
     drawAs=Image
     imageType=NoImage
     mirroring=NoMirror
     resourceName=None
     resourceObject=/Tank/UI/Art/M_UI_TankProgressBar_Backing_Gas
     tiling=NoTile
     imageSize=<14 14>
     margin=<ltrb 0 0 0 0>
     outlineSettings {
      bUseBrushTransparency=false
      roundingType=HalfHeightRadius
      width=0
      color {colorUseRule=UseColor_Specified, specifiedColor=<rgba 0 0 0 0>}
      cornerRadii=<0 0 0 0>
     }
     tintColor {colorUseRule=UseColor_Specified, specifiedColor=<rgba 1 1 1 1>}
     uVRegion {bIsValid=false, max=<0 0>, min=<0 0>}
    }
   }
   slot {class=OverlaySlot, properties {verticalAlignment=VAlign_Fill}}
  }
  Image_AltMeter {
   widgetClassPath=&Image
   bIsVariable=true
   parent=Overlay_9
   widget=Image_AltMeter
   properties {
    bIsEnabled=false
    visibility=HitTestInvisible
    brush {
     bIsDynamicallyLoaded=false
     drawAs=Image
     imageType=NoImage
     mirroring=NoMirror
     resourceName=None
     resourceObject=/Tank/UI/Art/M_UI_TankProgressBar_Angle
     tiling=NoTile
     imageSize=<14 14>
     margin=<ltrb 0 0 0 0>
     outlineSettings {
      bUseBrushTransparency=false
      roundingType=HalfHeightRadius
      width=0
      color {colorUseRule=UseColor_Specified, specifiedColor=<rgba 0 0 0 0>}
      cornerRadii=<0 0 0 0>
     }
     tintColor {colorUseRule=UseColor_Specified, specifiedColor=<rgba 1 1 1 1>}
     uVRegion {bIsValid=false, max=<0 0>, min=<0 0>}
    }
   }
   slot {class=OverlaySlot, properties {verticalAlignment=VAlign_Fill}}
  }
  Text_MinPitch {
   widgetClassPath=&CommonTextBlock
   bIsVariable=true
   parent=VerticalBox_10
   widget=Text_MinPitch
   properties {
    text=-20
    colorAndOpacity {colorUseRule=UseColor_Specified, specifiedColor=<rgba 1 0.3328 0 1>}
    font {
     bForceMonospaced=false
     bMaterialIsStencil=false
     fontMaterial=&M_Ui_Additive_White_025
     fontObject=&BurbankSmall
     letterSpacing=0
     monospacedWidth=1
     size=12
     skewAmount=0
     typefaceFontName=Bold
     outlineSettings {
      bApplyOutlineToDropShadows=false
      bMiteredCorners=false
      bSeparateFillAlpha=false
      outlineMaterial=null
      outlineSize=0
      outlineColor=<rgba 0 0 0 1>
     }
    }
    shadowOffset=<0 0>
   }
   slot {class=VerticalBoxSlot, properties {horizontalAlignment=HAlign_Center}}
  }
  VerticalBox_Distance {
   widgetClassPath=&VerticalBox
   parent=HorizontalBox_0
   widget=VerticalBox_Distance
   slot {class=HorizontalBoxSlot, properties {horizontalAlignment=HAlign_Left, verticalAlignment=VAlign_Center}}
  }
  CommonTextBlock_3 {
   widgetClassPath=&CommonTextBlock
   parent=VerticalBox_Distance
   widget=CommonTextBlock_3
   properties {
    text=150
    colorAndOpacity {colorUseRule=UseColor_Specified, specifiedColor=<rgba 1 0.3328 0 1>}
    font {
     bForceMonospaced=false
     bMaterialIsStencil=false
     fontMaterial=&M_Ui_Additive_White_025
     fontObject=&BurbankSmall
     letterSpacing=0
     monospacedWidth=1
     size=12
     skewAmount=0
     typefaceFontName=Bold
     outlineSettings {
      bApplyOutlineToDropShadows=false
      bMiteredCorners=false
      bSeparateFillAlpha=false
      outlineMaterial=null
      outlineSize=0
      outlineColor=<rgba 0 0 0 1>
     }
    }
    shadowOffset=<0 0>
   }
   slot {class=VerticalBoxSlot, properties {horizontalAlignment=HAlign_Center, verticalAlignment=VAlign_Center, padding=<ltrb -32 0 -32 0>}}
  }
  CommonBorder_Distance {
   widgetClassPath=&CommonBorder
   parent=VerticalBox_Distance
   widget=CommonBorder_Distance
   properties {
    background {
     bIsDynamicallyLoaded=false
     drawAs=Box
     imageType=NoImage
     mirroring=NoMirror
     resourceName=None
     resourceObject=&T_UI_Tank_ProgressBorder_Mat
     tiling=NoTile
     imageSize=<32 32>
     margin=<ltrb 0.33 0.33 0.33 0.33>
     outlineSettings {
      bUseBrushTransparency=false
      roundingType=HalfHeightRadius
      width=0
      color {colorUseRule=UseColor_Specified, specifiedColor=<rgba 0 0 0 0>}
      cornerRadii=<0 0 0 0>
     }
     tintColor {colorUseRule=UseColor_Specified, specifiedColor=<rgba 1 1 1 1>}
     uVRegion {bIsValid=false, max=<0 0>, min=<0 0>}
    }
    padding=<ltrb 10 10 10 10>
   }
   slot {class=VerticalBoxSlot}
  }
  SizeBox_DistanceMeter {
   widgetClassPath=&SizeBox
   parent=CommonBorder_Distance
   widget=SizeBox_DistanceMeter
   properties {bOverride_HeightOverride=true, bOverride_WidthOverride=true, heightOverride=370, widthOverride=14}
   slot {class=BorderSlot, properties {padding=<ltrb 10 10 10 10>}}
  }
  Overlay_7 {widgetClassPath=&Overlay, parent=SizeBox_DistanceMeter, widget=Overlay_7, slot {class=SizeBoxSlot}}
  Image_DistanceMeter_Backing {
   widgetClassPath=&Image
   bIsVariable=true
   parent=Overlay_7
   widget=Image_DistanceMeter_Backing
   properties {
    visibility=HitTestInvisible
    brush {
     bIsDynamicallyLoaded=false
     drawAs=Image
     imageType=NoImage
     mirroring=NoMirror
     resourceName=None
     resourceObject=/Tank/UI/Art/M_UI_TankProgressBar_Backing_Distance
     tiling=NoTile
     imageSize=<14 185.3434>
     margin=<ltrb 0 0 0 0>
     outlineSettings {
      bUseBrushTransparency=false
      roundingType=HalfHeightRadius
      width=0
      color {colorUseRule=UseColor_Specified, specifiedColor=<rgba 0 0 0 0>}
      cornerRadii=<0 0 0 0>
     }
     tintColor {colorUseRule=UseColor_Specified, specifiedColor=<rgba 1 1 1 1>}
     uVRegion {bIsValid=false, max=<0 0>, min=<0 0>}
    }
   }
   slot {class=OverlaySlot, properties {verticalAlignment=VAlign_Fill}}
  }
  Image_DistanceMeter {
   widgetClassPath=&Image
   bIsVariable=true
   parent=Overlay_7
   widget=Image_DistanceMeter
   properties {
    visibility=HitTestInvisible
    brush {
     bIsDynamicallyLoaded=false
     drawAs=Image
     imageType=NoImage
     mirroring=NoMirror
     resourceName=None
     resourceObject=/Tank/UI/Art/M_UI_TankProgressBar_Distance
     tiling=NoTile
     imageSize=<14 185.3434>
     margin=<ltrb 0 0 0 0>
     outlineSettings {
      bUseBrushTransparency=false
      roundingType=HalfHeightRadius
      width=0
      color {colorUseRule=UseColor_Specified, specifiedColor=<rgba 0 0 0 0>}
      cornerRadii=<0 0 0 0>
     }
     tintColor {colorUseRule=UseColor_Specified, specifiedColor=<rgba 1 1 1 1>}
     uVRegion {bIsValid=false, max=<0 0>, min=<0 0>}
    }
   }
   slot {class=OverlaySlot, properties {verticalAlignment=VAlign_Fill}}
  }
  CommonTextBlock_2 {
   widgetClassPath=&CommonTextBlock
   parent=VerticalBox_Distance
   widget=CommonTextBlock_2
   properties {
    text=0
    colorAndOpacity {colorUseRule=UseColor_Specified, specifiedColor=<rgba 1 0.3328 0 1>}
    font {
     bForceMonospaced=false
     bMaterialIsStencil=false
     fontMaterial=&M_Ui_Additive_White_025
     fontObject=&BurbankSmall
     letterSpacing=0
     monospacedWidth=1
     size=12
     skewAmount=0
     typefaceFontName=Bold
     outlineSettings {
      bApplyOutlineToDropShadows=false
      bMiteredCorners=false
      bSeparateFillAlpha=false
      outlineMaterial=null
      outlineSize=0
      outlineColor=<rgba 0 0 0 1>
     }
    }
    shadowOffset=<0 0>
   }
   slot {class=VerticalBoxSlot, properties {horizontalAlignment=HAlign_Center}}
  }
  SizeBox_Distance {
   widgetClassPath=&SizeBox
   parent=HorizontalBox_0
   widget=SizeBox_Distance
   properties {bOverride_WidthOverride=true}
   slot {class=HorizontalBoxSlot, properties {verticalAlignment=VAlign_Center}}
  }
  CanvasPanel_2 {widgetClassPath=&CanvasPanel, parent=SizeBox_Distance, widget=CanvasPanel_2, slot {class=SizeBoxSlot}}
  VerticalBox_11 {
   widgetClassPath=&VerticalBox
   parent=CanvasPanel_2
   widget=VerticalBox_11
   slot {
    class=CanvasPanelSlot
    properties {bAutoSize=true, layoutData {alignment=<0 0.5>, anchors {maximum=<0 0.5>, minimum=<0 0.5>}, offsets=<ltrb 0 0 0 0>}}
   }
  }
  CRichText_Distance {
   widgetClassPath=&CommonRichTextBlock
   bIsVariable=true
   parent=VerticalBox_11
   widget=CRichText_Distance
   properties {
    bOverrideDefaultStyle=true
    lineHeightPercentage=0.85
    text=88
    textTransformPolicy=ToUpper
    defaultTextStyleOverride {
     overflowPolicy=Clip
     transformPolicy=None
     colorAndOpacity {colorUseRule=UseColor_Specified, specifiedColor=<rgba 1 1 1 1>}
     font {
      bForceMonospaced=false
      bMaterialIsStencil=false
      fontMaterial=&M_Ui_Additive_White
      fontObject=&BurbankSmall
      letterSpacing=0
      monospacedWidth=1
      size=16
      skewAmount=0
      typefaceFontName=Bold
      outlineSettings {
       bApplyOutlineToDropShadows=false
       bMiteredCorners=false
       bSeparateFillAlpha=false
       outlineMaterial=null
       outlineSize=0
       outlineColor=<rgba 0 0 0 1>
      }
     }
     highlightColor {colorUseRule=UseColor_Specified, specifiedColor=<rgba 0 0 0 1>}
     highlightShape {
      bIsDynamicallyLoaded=false
      drawAs=Image
      imageType=NoImage
      mirroring=NoMirror
      resourceName=None
      resourceObject=null
      tiling=NoTile
      imageSize=<32 32>
      margin=<ltrb 0 0 0 0>
      outlineSettings {
       bUseBrushTransparency=false
       roundingType=HalfHeightRadius
       width=0
       color {colorUseRule=UseColor_Specified, specifiedColor=<rgba 0 0 0 0>}
       cornerRadii=<0 0 0 0>
      }
      tintColor {colorUseRule=UseColor_Specified, specifiedColor=<rgba 1 1 1 1>}
      uVRegion {bIsValid=false, max=<0 0>, min=<0 0>}
     }
     selectedBackgroundColor {colorUseRule=UseColor_Foreground, specifiedColor=<rgba 1 0 1 1>}
     shadowColorAndOpacity=<rgba 0 0 0 1>
     shadowOffset=<0 0>
     strikeBrush {
      bIsDynamicallyLoaded=false
      drawAs=Image
      imageType=NoImage
      mirroring=NoMirror
      resourceName=None
      resourceObject=null
      tiling=NoTile
      imageSize=<32 32>
      margin=<ltrb 0 0 0 0>
      outlineSettings {
       bUseBrushTransparency=false
       roundingType=HalfHeightRadius
       width=0
       color {colorUseRule=UseColor_Specified, specifiedColor=<rgba 0 0 0 0>}
       cornerRadii=<0 0 0 0>
      }
      tintColor {colorUseRule=UseColor_Specified, specifiedColor=<rgba 1 1 1 1>}
      uVRegion {bIsValid=false, max=<0 0>, min=<0 0>}
     }
     underlineBrush {
      bIsDynamicallyLoaded=false
      drawAs=Image
      imageType=NoImage
      mirroring=NoMirror
      resourceName=None
      resourceObject=null
      tiling=NoTile
      imageSize=<32 32>
      margin=<ltrb 0 0 0 0>
      outlineSettings {
       bUseBrushTransparency=false
       roundingType=HalfHeightRadius
       width=0
       color {colorUseRule=UseColor_Specified, specifiedColor=<rgba 0 0 0 0>}
       cornerRadii=<0 0 0 0>
      }
      tintColor {colorUseRule=UseColor_Specified, specifiedColor=<rgba 1 1 1 1>}
      uVRegion {bIsValid=false, max=<0 0>, min=<0 0>}
     }
    }
   }
   slot {class=VerticalBoxSlot, properties {horizontalAlignment=HAlign_Left}}
  }
  CRichText_DistanceMeters {
   widgetClassPath=&CommonRichTextBlock
   bIsVariable=true
   parent=VerticalBox_11
   widget=CRichText_DistanceMeters
   properties {
    bOverrideDefaultStyle=true
    lineHeightPercentage=0.85
    text=meters
    textTransformPolicy=ToUpper
    defaultTextStyleOverride {
     overflowPolicy=Clip
     transformPolicy=None
     colorAndOpacity {colorUseRule=UseColor_Specified, specifiedColor=<rgba 1 0.2542 0 1>}
     font {
      bForceMonospaced=false
      bMaterialIsStencil=false
      fontMaterial=&M_Ui_Additive_White
      fontObject=&BurbankSmall
      letterSpacing=0
      monospacedWidth=1
      size=10
      skewAmount=0
      typefaceFontName=Bold
      outlineSettings {
       bApplyOutlineToDropShadows=false
       bMiteredCorners=false
       bSeparateFillAlpha=false
       outlineMaterial=null
       outlineSize=0
       outlineColor=<rgba 0 0 0 1>
      }
     }
     highlightColor {colorUseRule=UseColor_Specified, specifiedColor=<rgba 0 0 0 1>}
     highlightShape {
      bIsDynamicallyLoaded=false
      drawAs=Image
      imageType=NoImage
      mirroring=NoMirror
      resourceName=None
      resourceObject=null
      tiling=NoTile
      imageSize=<32 32>
      margin=<ltrb 0 0 0 0>
      outlineSettings {
       bUseBrushTransparency=false
       roundingType=HalfHeightRadius
       width=0
       color {colorUseRule=UseColor_Specified, specifiedColor=<rgba 0 0 0 0>}
       cornerRadii=<0 0 0 0>
      }
      tintColor {colorUseRule=UseColor_Specified, specifiedColor=<rgba 1 1 1 1>}
      uVRegion {bIsValid=false, max=<0 0>, min=<0 0>}
     }
     selectedBackgroundColor {colorUseRule=UseColor_Foreground, specifiedColor=<rgba 1 0 1 1>}
     shadowColorAndOpacity=<rgba 0 0 0 1>
     shadowOffset=<0 0>
     strikeBrush {
      bIsDynamicallyLoaded=false
      drawAs=Image
      imageType=NoImage
      mirroring=NoMirror
      resourceName=None
      resourceObject=null
      tiling=NoTile
      imageSize=<32 32>
      margin=<ltrb 0 0 0 0>
      outlineSettings {
       bUseBrushTransparency=false
       roundingType=HalfHeightRadius
       width=0
       color {colorUseRule=UseColor_Specified, specifiedColor=<rgba 0 0 0 0>}
       cornerRadii=<0 0 0 0>
      }
      tintColor {colorUseRule=UseColor_Specified, specifiedColor=<rgba 1 1 1 1>}
      uVRegion {bIsValid=false, max=<0 0>, min=<0 0>}
     }
     underlineBrush {
      bIsDynamicallyLoaded=false
      drawAs=Image
      imageType=NoImage
      mirroring=NoMirror
      resourceName=None
      resourceObject=null
      tiling=NoTile
      imageSize=<32 32>
      margin=<ltrb 0 0 0 0>
      outlineSettings {
       bUseBrushTransparency=false
       roundingType=HalfHeightRadius
       width=0
       color {colorUseRule=UseColor_Specified, specifiedColor=<rgba 0 0 0 0>}
       cornerRadii=<0 0 0 0>
      }
      tintColor {colorUseRule=UseColor_Specified, specifiedColor=<rgba 1 1 1 1>}
      uVRegion {bIsValid=false, max=<0 0>, min=<0 0>}
     }
    }
   }
   slot {class=VerticalBoxSlot, properties {horizontalAlignment=HAlign_Left}}
  }
  Overlay_6 {widgetClassPath=&Overlay, parent=HorizontalBox_0, widget=Overlay_6, slot {class=HorizontalBoxSlot}}
  Image_Horizon {
   widgetClassPath=&Image
   bIsVariable=true
   parent=Overlay_6
   widget=Image_Horizon
   properties {
    visibility=HitTestInvisible
    brush {
     bIsDynamicallyLoaded=false
     drawAs=Box
     imageType=NoImage
     mirroring=NoMirror
     resourceName=None
     resourceObject=/Game/Athena/HUD/Vehicles/Materials/M_UI_TargetBracket_Inst
     tiling=NoTile
     imageSize=<300 150>
     margin=<ltrb 0 0 0 0>
     outlineSettings {
      bUseBrushTransparency=false
      roundingType=HalfHeightRadius
      width=0
      color {colorUseRule=UseColor_Specified, specifiedColor=<rgba 0 0 0 0>}
      cornerRadii=<0 0 0 0>
     }
     tintColor {colorUseRule=UseColor_Specified, specifiedColor=<rgba 1 1 1 1>}
     uVRegion {bIsValid=false, max=<0 0>, min=<0 0>}
    }
    colorAndOpacity=<rgba 1 1 1 0>
    renderTransform {angle=0, scale=<1 1>, shear=<0 0>, translation=<-1 2>}
   }
   slot {class=OverlaySlot, properties {horizontalAlignment=HAlign_Center, verticalAlignment=VAlign_Center, padding=<ltrb 250 0 250 0>}}
  }
  SizeBox_Speed {
   widgetClassPath=&SizeBox
   bIsVariable=true
   parent=HorizontalBox_0
   widget=SizeBox_Speed
   properties {bOverride_WidthOverride=true}
   slot {class=HorizontalBoxSlot, properties {verticalAlignment=VAlign_Center}}
  }
  CanvasPanel_3 {
   widgetClassPath=&CanvasPanel
   parent=SizeBox_Speed
   widget=CanvasPanel_3
   slot {class=SizeBoxSlot, properties {horizontalAlignment=HAlign_Right}}
  }
  VerticalBox_12 {
   widgetClassPath=&VerticalBox
   parent=CanvasPanel_3
   widget=VerticalBox_12
   slot {
    class=CanvasPanelSlot
    properties {bAutoSize=true, layoutData {alignment=<1 0.5>, anchors {maximum=<0 0.5>, minimum=<0 0.5>}, offsets=<ltrb 0 0 0 0>}}
   }
  }
  CRichText_Speed {
   widgetClassPath=&CommonRichTextBlock
   bIsVariable=true
   parent=VerticalBox_12
   widget=CRichText_Speed
   properties {
    bOverrideDefaultStyle=true
    justification=Right
    lineHeightPercentage=0.85
    text=xx
    textTransformPolicy=ToUpper
    defaultTextStyleOverride {
     overflowPolicy=Clip
     transformPolicy=None
     colorAndOpacity {colorUseRule=UseColor_Specified, specifiedColor=<rgba 1 1 1 1>}
     font {
      bForceMonospaced=false
      bMaterialIsStencil=false
      fontMaterial=&M_Ui_Additive_White
      fontObject=&BurbankSmall
      letterSpacing=0
      monospacedWidth=1
      size=16
      skewAmount=0
      typefaceFontName=Bold
      outlineSettings {
       bApplyOutlineToDropShadows=false
       bMiteredCorners=false
       bSeparateFillAlpha=false
       outlineMaterial=null
       outlineSize=0
       outlineColor=<rgba 0 0 0 1>
      }
     }
     highlightColor {colorUseRule=UseColor_Specified, specifiedColor=<rgba 0 0 0 1>}
     highlightShape {
      bIsDynamicallyLoaded=false
      drawAs=Image
      imageType=NoImage
      mirroring=NoMirror
      resourceName=None
      resourceObject=null
      tiling=NoTile
      imageSize=<32 32>
      margin=<ltrb 0 0 0 0>
      outlineSettings {
       bUseBrushTransparency=false
       roundingType=HalfHeightRadius
       width=0
       color {colorUseRule=UseColor_Specified, specifiedColor=<rgba 0 0 0 0>}
       cornerRadii=<0 0 0 0>
      }
      tintColor {colorUseRule=UseColor_Specified, specifiedColor=<rgba 1 1 1 1>}
      uVRegion {bIsValid=false, max=<0 0>, min=<0 0>}
     }
     selectedBackgroundColor {colorUseRule=UseColor_Foreground, specifiedColor=<rgba 1 0 1 1>}
     shadowColorAndOpacity=<rgba 0 0 0 1>
     shadowOffset=<0 0>
     strikeBrush {
      bIsDynamicallyLoaded=false
      drawAs=Image
      imageType=NoImage
      mirroring=NoMirror
      resourceName=None
      resourceObject=null
      tiling=NoTile
      imageSize=<32 32>
      margin=<ltrb 0 0 0 0>
      outlineSettings {
       bUseBrushTransparency=false
       roundingType=HalfHeightRadius
       width=0
       color {colorUseRule=UseColor_Specified, specifiedColor=<rgba 0 0 0 0>}
       cornerRadii=<0 0 0 0>
      }
      tintColor {colorUseRule=UseColor_Specified, specifiedColor=<rgba 1 1 1 1>}
      uVRegion {bIsValid=false, max=<0 0>, min=<0 0>}
     }
     underlineBrush {
      bIsDynamicallyLoaded=false
      drawAs=Image
      imageType=NoImage
      mirroring=NoMirror
      resourceName=None
      resourceObject=null
      tiling=NoTile
      imageSize=<32 32>
      margin=<ltrb 0 0 0 0>
      outlineSettings {
       bUseBrushTransparency=false
       roundingType=HalfHeightRadius
       width=0
       color {colorUseRule=UseColor_Specified, specifiedColor=<rgba 0 0 0 0>}
       cornerRadii=<0 0 0 0>
      }
      tintColor {colorUseRule=UseColor_Specified, specifiedColor=<rgba 1 1 1 1>}
      uVRegion {bIsValid=false, max=<0 0>, min=<0 0>}
     }
    }
   }
   slot {class=VerticalBoxSlot, properties {horizontalAlignment=HAlign_Right}}
  }
  CRichText_SpeedKPH {
   widgetClassPath=&CommonRichTextBlock
   bIsVariable=true
   parent=VerticalBox_12
   widget=CRichText_SpeedKPH
   properties {
    bOverrideDefaultStyle=true
    justification=Right
    lineHeightPercentage=0.85
    text=KM/H
    textTransformPolicy=ToUpper
    defaultTextStyleOverride {
     overflowPolicy=Clip
     transformPolicy=None
     colorAndOpacity {colorUseRule=UseColor_Specified, specifiedColor=<rgba 1 0.2521 0 1>}
     font {
      bForceMonospaced=false
      bMaterialIsStencil=false
      fontMaterial=&M_Ui_Additive_White
      fontObject=&BurbankSmall
      letterSpacing=0
      monospacedWidth=1
      size=10
      skewAmount=0
      typefaceFontName=Bold
      outlineSettings {
       bApplyOutlineToDropShadows=false
       bMiteredCorners=false
       bSeparateFillAlpha=false
       outlineMaterial=null
       outlineSize=0
       outlineColor=<rgba 0 0 0 1>
      }
     }
     highlightColor {colorUseRule=UseColor_Specified, specifiedColor=<rgba 0 0 0 1>}
     highlightShape {
      bIsDynamicallyLoaded=false
      drawAs=Image
      imageType=NoImage
      mirroring=NoMirror
      resourceName=None
      resourceObject=null
      tiling=NoTile
      imageSize=<32 32>
      margin=<ltrb 0 0 0 0>
      outlineSettings {
       bUseBrushTransparency=false
       roundingType=HalfHeightRadius
       width=0
       color {colorUseRule=UseColor_Specified, specifiedColor=<rgba 0 0 0 0>}
       cornerRadii=<0 0 0 0>
      }
      tintColor {colorUseRule=UseColor_Specified, specifiedColor=<rgba 1 1 1 1>}
      uVRegion {bIsValid=false, max=<0 0>, min=<0 0>}
     }
     selectedBackgroundColor {colorUseRule=UseColor_Foreground, specifiedColor=<rgba 1 0 1 1>}
     shadowColorAndOpacity=<rgba 0 0 0 1>
     shadowOffset=<0 0>
     strikeBrush {
      bIsDynamicallyLoaded=false
      drawAs=Image
      imageType=NoImage
      mirroring=NoMirror
      resourceName=None
      resourceObject=null
      tiling=NoTile
      imageSize=<32 32>
      margin=<ltrb 0 0 0 0>
      outlineSettings {
       bUseBrushTransparency=false
       roundingType=HalfHeightRadius
       width=0
       color {colorUseRule=UseColor_Specified, specifiedColor=<rgba 0 0 0 0>}
       cornerRadii=<0 0 0 0>
      }
      tintColor {colorUseRule=UseColor_Specified, specifiedColor=<rgba 1 1 1 1>}
      uVRegion {bIsValid=false, max=<0 0>, min=<0 0>}
     }
     underlineBrush {
      bIsDynamicallyLoaded=false
      drawAs=Image
      imageType=NoImage
      mirroring=NoMirror
      resourceName=None
      resourceObject=null
      tiling=NoTile
      imageSize=<32 32>
      margin=<ltrb 0 0 0 0>
      outlineSettings {
       bUseBrushTransparency=false
       roundingType=HalfHeightRadius
       width=0
       color {colorUseRule=UseColor_Specified, specifiedColor=<rgba 0 0 0 0>}
       cornerRadii=<0 0 0 0>
      }
      tintColor {colorUseRule=UseColor_Specified, specifiedColor=<rgba 1 1 1 1>}
      uVRegion {bIsValid=false, max=<0 0>, min=<0 0>}
     }
    }
   }
   slot {class=VerticalBoxSlot, properties {horizontalAlignment=HAlign_Right}}
  }
  VerticalBox_Speed {
   widgetClassPath=&VerticalBox
   parent=HorizontalBox_0
   widget=VerticalBox_Speed
   slot {class=HorizontalBoxSlot, properties {horizontalAlignment=HAlign_Right, verticalAlignment=VAlign_Center}}
  }
  Overlay_1 {
   widgetClassPath=&Overlay
   parent=VerticalBox_Speed
   widget=Overlay_1
   slot {class=VerticalBoxSlot, properties {horizontalAlignment=HAlign_Center}}
  }
  CommonTextBlock_1 {
   widgetClassPath=&CommonTextBlock
   parent=Overlay_1
   widget=CommonTextBlock_1
   properties {
    text=50
    colorAndOpacity {colorUseRule=UseColor_Specified, specifiedColor=<rgba 1 0.3328 0 1>}
    font {
     bForceMonospaced=false
     bMaterialIsStencil=false
     fontMaterial=&M_Ui_Additive_White_025
     fontObject=&BurbankSmall
     letterSpacing=0
     monospacedWidth=1
     size=12
     skewAmount=0
     typefaceFontName=Bold
     outlineSettings {
      bApplyOutlineToDropShadows=false
      bMiteredCorners=false
      bSeparateFillAlpha=false
      outlineMaterial=null
      outlineSize=0
      outlineColor=<rgba 0 0 0 1>
     }
    }
    shadowOffset=<0 0>
   }
   slot {class=OverlaySlot, properties {horizontalAlignment=HAlign_Center, verticalAlignment=VAlign_Center}}
  }
  CommonBorder_SpeedometerContainer {
   widgetClassPath=&CommonBorder
   bIsVariable=true
   parent=VerticalBox_Speed
   widget=CommonBorder_SpeedometerContainer
   properties {padding=<ltrb 0 0 0 0>}
   slot {class=VerticalBoxSlot}
  }
  Border_Speedometer {
   widgetClassPath=&Border
   parent=CommonBorder_SpeedometerContainer
   widget=Border_Speedometer
   properties {
    background {
     bIsDynamicallyLoaded=false
     drawAs=Box
     imageType=NoImage
     mirroring=NoMirror
     resourceName=None
     resourceObject=&T_UI_Tank_ProgressBorder_Mat
     tiling=NoTile
     imageSize=<32 32>
     margin=<ltrb 0.33 0.33 0.33 0.33>
     outlineSettings {
      bUseBrushTransparency=false
      roundingType=HalfHeightRadius
      width=0
      color {colorUseRule=UseColor_Specified, specifiedColor=<rgba 0 0 0 0>}
      cornerRadii=<0 0 0 0>
     }
     tintColor {colorUseRule=UseColor_Specified, specifiedColor=<rgba 1 1 1 1>}
     uVRegion {bIsValid=false, max=<0 0>, min=<0 0>}
    }
    padding=<ltrb 10 10 10 10>
   }
   slot {class=BorderSlot, properties {padding=<ltrb 0 0 0 0>}}
  }
  SizeBox_Speedometer {
   widgetClassPath=&SizeBox
   parent=Border_Speedometer
   widget=SizeBox_Speedometer
   properties {bOverride_HeightOverride=true, heightOverride=383}
   slot {class=BorderSlot, properties {padding=<ltrb 10 10 10 10>}}
  }
  Overlay_5 {widgetClassPath=&Overlay, parent=SizeBox_Speedometer, widget=Overlay_5, slot {class=SizeBoxSlot}}
  Image_Speedometer_Backing {
   widgetClassPath=&Image
   bIsVariable=true
   parent=Overlay_5
   widget=Image_Speedometer_Backing
   properties {
    visibility=HitTestInvisible
    brush {
     bIsDynamicallyLoaded=false
     drawAs=Image
     imageType=NoImage
     mirroring=NoMirror
     resourceName=None
     resourceObject=/Tank/UI/Art/M_UI_TankProgressBar_Backing_Speed
     tiling=NoTile
     imageSize=<17.943 185.3434>
     margin=<ltrb 0 0 0 0>
     outlineSettings {
      bUseBrushTransparency=false
      roundingType=HalfHeightRadius
      width=0
      color {colorUseRule=UseColor_Specified, specifiedColor=<rgba 0 0 0 0>}
      cornerRadii=<0 0 0 0>
     }
     tintColor {colorUseRule=UseColor_Specified, specifiedColor=<rgba 1 1 1 1>}
     uVRegion {bIsValid=false, max=<0 0>, min=<0 0>}
    }
   }
   slot {class=OverlaySlot, properties {verticalAlignment=VAlign_Fill}}
  }
  Image_Speedometer {
   widgetClassPath=&Image
   bInherited=true
   bIsVariable=true
   parent=Overlay_5
   widget=Image_Speedometer
   properties {
    visibility=HitTestInvisible
    brush {
     bIsDynamicallyLoaded=false
     drawAs=Image
     imageType=NoImage
     mirroring=NoMirror
     resourceName=None
     resourceObject=/Tank/UI/Art/M_UI_TankProgressBar_Speed
     tiling=NoTile
     imageSize=<17.943 185.3434>
     margin=<ltrb 0 0 0 0>
     outlineSettings {
      bUseBrushTransparency=false
      roundingType=HalfHeightRadius
      width=0
      color {colorUseRule=UseColor_Specified, specifiedColor=<rgba 0 0 0 0>}
      cornerRadii=<0 0 0 0>
     }
     tintColor {colorUseRule=UseColor_Specified, specifiedColor=<rgba 1 1 1 1>}
     uVRegion {bIsValid=false, max=<0 0>, min=<0 0>}
    }
   }
   slot {class=OverlaySlot, properties {verticalAlignment=VAlign_Fill}}
  }
  CommonTextBlock {
   widgetClassPath=&CommonTextBlock
   parent=VerticalBox_Speed
   widget=CommonTextBlock
   properties {
    text=0
    colorAndOpacity {colorUseRule=UseColor_Specified, specifiedColor=<rgba 1 0.3328 0 1>}
    font {
     bForceMonospaced=false
     bMaterialIsStencil=false
     fontMaterial=&M_Ui_Additive_White_025
     fontObject=&BurbankSmall
     letterSpacing=0
     monospacedWidth=1
     size=12
     skewAmount=0
     typefaceFontName=Bold
     outlineSettings {
      bApplyOutlineToDropShadows=false
      bMiteredCorners=false
      bSeparateFillAlpha=false
      outlineMaterial=null
      outlineSize=0
      outlineColor=<rgba 0 0 0 1>
     }
    }
    shadowOffset=<0 0>
   }
   slot {class=VerticalBoxSlot, properties {horizontalAlignment=HAlign_Center}}
  }
  VerticalBox_48 {
   widgetClassPath=&VerticalBox
   parent=HorizontalBox_0
   widget=VerticalBox_48
   properties {renderTransform {angle=0, scale=<1 1>, shear=<0 0>, translation=<5 0>}}
   slot {class=HorizontalBoxSlot, properties {horizontalAlignment=HAlign_Left, verticalAlignment=VAlign_Center, padding=<ltrb 0 -25 0 -25>}}
  }
  CommonTextBlock_79 {
   widgetClassPath=&CommonTextBlock
   parent=VerticalBox_48
   widget=CommonTextBlock_79
   properties {
    text=F
    colorAndOpacity {colorUseRule=UseColor_Specified, specifiedColor=<rgba 1 0.3328 0 1>}
    font {
     bForceMonospaced=false
     bMaterialIsStencil=false
     fontMaterial=&M_Ui_Additive_White_025
     fontObject=&BurbankSmall
     letterSpacing=0
     monospacedWidth=1
     size=12
     skewAmount=0
     typefaceFontName=Bold
     outlineSettings {
      bApplyOutlineToDropShadows=false
      bMiteredCorners=false
      bSeparateFillAlpha=false
      outlineMaterial=null
      outlineSize=0
      outlineColor=<rgba 0 0 0 1>
     }
    }
    shadowOffset=<0 0>
   }
   slot {class=VerticalBoxSlot, properties {horizontalAlignment=HAlign_Center}}
  }
  CommonBorder_FuelMeterContainer {
   widgetClassPath=&CommonBorder
   bIsVariable=true
   parent=VerticalBox_48
   widget=CommonBorder_FuelMeterContainer
   properties {
    background {
     bIsDynamicallyLoaded=false
     drawAs=Box
     imageType=NoImage
     mirroring=NoMirror
     resourceName=None
     resourceObject=&T_UI_Tank_ProgressBorder_Mat
     tiling=NoTile
     imageSize=<32 32>
     margin=<ltrb 0.33 0.33 0.33 0.33>
     outlineSettings {
      bUseBrushTransparency=false
      roundingType=HalfHeightRadius
      width=0
      color {colorUseRule=UseColor_Specified, specifiedColor=<rgba 0 0 0 0>}
      cornerRadii=<0 0 0 0>
     }
     tintColor {colorUseRule=UseColor_Specified, specifiedColor=<rgba 1 1 1 1>}
     uVRegion {bIsValid=false, max=<0 0>, min=<0 0>}
    }
    padding=<ltrb 10 10 10 10>
   }
   slot {class=VerticalBoxSlot, properties {verticalAlignment=VAlign_Center}}
  }
  SizeBox_Fuel {
   widgetClassPath=&SizeBox
   parent=CommonBorder_FuelMeterContainer
   widget=SizeBox_Fuel
   properties {bOverride_HeightOverride=true, heightOverride=277}
   slot {class=BorderSlot, properties {padding=<ltrb 10 10 10 10>}}
  }
  Overlay_0 {widgetClassPath=&Overlay, parent=SizeBox_Fuel, widget=Overlay_0, slot {class=SizeBoxSlot}}
  Overlay_BoostSparks {
   widgetClassPath=&Overlay
   bInherited=true
   bIsVariable=true
   parent=Overlay_0
   widget=Overlay_BoostSparks
   properties {renderOpacity=0, renderTransformPivot=<71 0.5>}
   slot {class=OverlaySlot, properties {horizontalAlignment=HAlign_Center, verticalAlignment=VAlign_Bottom}}
  }
  Overlay_47 {
   widgetClassPath=&Overlay
   parent=Overlay_BoostSparks
   widget=Overlay_47
   properties {visibility=HitTestInvisible, renderTransform {angle=-180, scale=<1 1>, shear=<0 0>, translation=<0 0>}}
   slot {class=OverlaySlot}
  }
  Spacer_42 {widgetClassPath=&Spacer, parent=Overlay_47, widget=Spacer_42, properties {visibility=HitTestInvisible}, slot {class=OverlaySlot}}
  Image_Pulse {
   widgetClassPath=&Image
   bIsVariable=true
   parent=Overlay_47
   widget=Image_Pulse
   properties {
    visibility=HitTestInvisible
    brush {
     bIsDynamicallyLoaded=false
     drawAs=Image
     imageType=NoImage
     mirroring=NoMirror
     resourceName=None
     resourceObject=/Game/Athena/HUD/LTM/Wax/Materials/M_UI_TextGlow_Pulse
     tiling=NoTile
     imageSize=<44.4227 97.4638>
     margin=<ltrb 0 0 0 0>
     outlineSettings {
      bUseBrushTransparency=false
      roundingType=HalfHeightRadius
      width=0
      color {colorUseRule=UseColor_Specified, specifiedColor=<rgba 0 0 0 0>}
      cornerRadii=<0 0 0 0>
     }
     tintColor {colorUseRule=UseColor_Specified, specifiedColor=<rgba 0.8906 0.1088 0 1>}
     uVRegion {bIsValid=false, max=<0 0>, min=<0 0>}
    }
    renderTransform {angle=0, scale=<1 1>, shear=<0 0>, translation=<0 8>}
   }
   slot {class=OverlaySlot, properties {horizontalAlignment=HAlign_Center, verticalAlignment=VAlign_Center, padding=<ltrb -200 -200 -200 -200>}}
  }
  Image_Glow {
   widgetClassPath=&Image
   bIsVariable=true
   parent=Overlay_47
   widget=Image_Glow
   properties {
    visibility=HitTestInvisible
    brush {
     bIsDynamicallyLoaded=false
     drawAs=Image
     imageType=NoImage
     mirroring=NoMirror
     resourceName=None
     resourceObject=/Game/Athena/HUD/Vehicles/Materials/M_UI_SparksGlow
     tiling=NoTile
     imageSize=<130 109.0002>
     margin=<ltrb 0 0 0 0>
     outlineSettings {
      bUseBrushTransparency=false
      roundingType=HalfHeightRadius
      width=0
      color {colorUseRule=UseColor_Specified, specifiedColor=<rgba 0 0 0 0>}
      cornerRadii=<0 0 0 0>
     }
     tintColor {colorUseRule=UseColor_Specified, specifiedColor=<rgba 1 1 1 0.25>}
     uVRegion {bIsValid=false, max=<0 0>, min=<0 0>}
    }
    renderTransform {angle=0, scale=<1 1>, shear=<0 0>, translation=<0 17>}
   }
   slot {class=OverlaySlot, properties {horizontalAlignment=HAlign_Center, verticalAlignment=VAlign_Center, padding=<ltrb -200 -200 -200 -200>}}
  }
  Image_Sparks {
   widgetClassPath=&Image
   bIsVariable=true
   parent=Overlay_47
   widget=Image_Sparks
   properties {
    visibility=HitTestInvisible
    brush {
     bIsDynamicallyLoaded=false
     drawAs=Image
     imageType=NoImage
     mirroring=NoMirror
     resourceName=None
     resourceObject=/Game/Athena/UI/InGame/Interrogation/Materials/M_UI_Starburst_Sparks
     tiling=NoTile
     imageSize=<295.3884 400>
     margin=<ltrb 0 0 0 0>
     outlineSettings {
      bUseBrushTransparency=false
      roundingType=HalfHeightRadius
      width=0
      color {colorUseRule=UseColor_Specified, specifiedColor=<rgba 0 0 0 0>}
      cornerRadii=<0 0 0 0>
     }
     tintColor {colorUseRule=UseColor_Specified, specifiedColor=<rgba 1 0.3682 0.0743 1>}
     uVRegion {bIsValid=false, max=<0 0>, min=<0 0>}
    }
   }
   slot {class=OverlaySlot, properties {horizontalAlignment=HAlign_Center, verticalAlignment=VAlign_Center, padding=<ltrb -200 -200 -200 -200>}}
  }
  Image_Sparks_1 {
   widgetClassPath=&Image
   bIsVariable=true
   parent=Overlay_47
   widget=Image_Sparks_1
   properties {
    visibility=HitTestInvisible
    brush {
     bIsDynamicallyLoaded=false
     drawAs=Image
     imageType=NoImage
     mirroring=NoMirror
     resourceName=None
     resourceObject=/Game/Athena/UI/InGame/Interrogation/Materials/M_UI_Starburst_Sparks_Inst
     tiling=NoTile
     imageSize=<120 400>
     margin=<ltrb 0 0 0 0>
     outlineSettings {
      bUseBrushTransparency=false
      roundingType=HalfHeightRadius
      width=0
      color {colorUseRule=UseColor_Specified, specifiedColor=<rgba 0 0 0 0>}
      cornerRadii=<0 0 0 0>
     }
     tintColor {colorUseRule=UseColor_Specified, specifiedColor=<rgba 1 0.7183 0.651 1>}
     uVRegion {bIsValid=false, max=<0 0>, min=<0 0>}
    }
   }
   slot {class=OverlaySlot, properties {horizontalAlignment=HAlign_Center, verticalAlignment=VAlign_Center, padding=<ltrb -200 -200 -200 -200>}}
  }
  Image_LargeCone {
   widgetClassPath=&Image
   bIsVariable=true
   parent=Overlay_47
   widget=Image_LargeCone
   properties {
    visibility=HitTestInvisible
    brush {
     bIsDynamicallyLoaded=false
     drawAs=Image
     imageType=NoImage
     mirroring=NoMirror
     resourceName=None
     resourceObject=/Game/Athena/HUD/Vehicles/Materials/M_UI_SparksGlow_Inst
     tiling=NoTile
     imageSize=<150 219.5743>
     margin=<ltrb 0 0 0 0>
     outlineSettings {
      bUseBrushTransparency=false
      roundingType=HalfHeightRadius
      width=0
      color {colorUseRule=UseColor_Specified, specifiedColor=<rgba 0 0 0 0>}
      cornerRadii=<0 0 0 0>
     }
     tintColor {colorUseRule=UseColor_Specified, specifiedColor=<rgba 1 1 1 1>}
     uVRegion {bIsValid=false, max=<0 0>, min=<0 0>}
    }
    renderTransform {angle=0, scale=<1 1>, shear=<0 0>, translation=<0 44>}
   }
   slot {class=OverlaySlot, properties {horizontalAlignment=HAlign_Center, verticalAlignment=VAlign_Center, padding=<ltrb -200 -200 -200 -200>}}
  }
  Image_FuelMeter_Backing {
   widgetClassPath=&Image
   bIsVariable=true
   parent=Overlay_0
   widget=Image_FuelMeter_Backing
   properties {
    visibility=HitTestInvisible
    brush {
     bIsDynamicallyLoaded=false
     drawAs=Image
     imageType=NoImage
     mirroring=NoMirror
     resourceName=None
     resourceObject=&M_UI_TankProgressBar_Backing_Gas
     tiling=NoTile
     imageSize=<14 14>
     margin=<ltrb 0 0 0 0>
     outlineSettings {
      bUseBrushTransparency=false
      roundingType=HalfHeightRadius
      width=0
      color {colorUseRule=UseColor_Specified, specifiedColor=<rgba 0 0 0 0>}
      cornerRadii=<0 0 0 0>
     }
     tintColor {colorUseRule=UseColor_Specified, specifiedColor=<rgba 1 1 1 1>}
     uVRegion {bIsValid=false, max=<0 0>, min=<0 0>}
    }
    renderTransform {angle=0, scale=<-1 1>, shear=<0 0>, translation=<0 0>}
   }
   slot {class=OverlaySlot, properties {verticalAlignment=VAlign_Fill}}
  }
  Image_FuelMeter {
   widgetClassPath=&Image
   bIsVariable=true
   parent=Overlay_0
   widget=Image_FuelMeter
   properties {
    visibility=HitTestInvisible
    brush {
     bIsDynamicallyLoaded=false
     drawAs=Image
     imageType=NoImage
     mirroring=NoMirror
     resourceName=None
     resourceObject=/Tank/UI/Art/M_UI_TankProgressBar_Gas
     tiling=NoTile
     imageSize=<14 14>
     margin=<ltrb 0 0 0 0>
     outlineSettings {
      bUseBrushTransparency=false
      roundingType=HalfHeightRadius
      width=0
      color {colorUseRule=UseColor_Specified, specifiedColor=<rgba 0 0 0 0>}
      cornerRadii=<0 0 0 0>
     }
     tintColor {colorUseRule=UseColor_Specified, specifiedColor=<rgba 1 1 1 1>}
     uVRegion {bIsValid=false, max=<0 0>, min=<0 0>}
    }
    renderTransform {angle=0, scale=<-1 1>, shear=<0 0>, translation=<0 0>}
   }
   slot {class=OverlaySlot, properties {verticalAlignment=VAlign_Fill}}
  }
  Overlay_2 {
   widgetClassPath=&Overlay
   parent=Overlay_0
   widget=Overlay_2
   properties {renderTransform {angle=0, scale=<1 1>, shear=<0 0>, translation=<34 0>}}
   slot {class=OverlaySlot, properties {horizontalAlignment=HAlign_Center, verticalAlignment=VAlign_Center, padding=<ltrb -32 0 -32 0>}}
  }
  FuelMeter_Icon {
   widgetClassPath=&Image
   parent=Overlay_2
   widget=FuelMeter_Icon
   properties {
    visibility=HitTestInvisible
    brush {
     bIsDynamicallyLoaded=false
     drawAs=Image
     imageType=NoImage
     mirroring=NoMirror
     resourceName=None
     resourceObject=/Game/Athena/HUD/Materials/T_UI_LowFuelIcon_Mat
     tiling=NoTile
     imageSize=<16 16>
     margin=<ltrb 0 0 0 0>
     outlineSettings {
      bUseBrushTransparency=false
      roundingType=HalfHeightRadius
      width=0
      color {colorUseRule=UseColor_Specified, specifiedColor=<rgba 0 0 0 0>}
      cornerRadii=<0 0 0 0>
     }
     tintColor {colorUseRule=UseColor_Specified, specifiedColor=<rgba 1 1 1 1>}
     uVRegion {bIsValid=false, max=<0 0>, min=<0 0>}
    }
    renderTransform {angle=0, scale=<2 2>, shear=<0 0>, translation=<0 0>}
   }
   slot {class=OverlaySlot, properties {horizontalAlignment=HAlign_Center, verticalAlignment=VAlign_Center}}
  }
  Fuel_Low {
   widgetClassPath=&Image
   bIsVariable=true
   parent=Overlay_2
   widget=Fuel_Low
   properties {
    brush {
     bIsDynamicallyLoaded=false
     drawAs=Image
     imageType=NoImage
     mirroring=NoMirror
     resourceName=None
     resourceObject=/Game/Athena/HUD/LTM/Wax/Materials/M_UI_TextGlow_Add_Gradient_Orange_Nevada
     tiling=NoTile
     imageSize=<16 16>
     margin=<ltrb 0 0 0 0>
     outlineSettings {
      bUseBrushTransparency=false
      roundingType=HalfHeightRadius
      width=0
      color {colorUseRule=UseColor_Specified, specifiedColor=<rgba 0 0 0 0>}
      cornerRadii=<0 0 0 0>
     }
     tintColor {colorUseRule=UseColor_Specified, specifiedColor=<rgba 1 1 1 1>}
     uVRegion {bIsValid=false, max=<0 0>, min=<0 0>}
    }
    renderTransform {angle=0, scale=<5 5>, shear=<0 0>, translation=<0 0>}
   }
   slot {class=OverlaySlot, properties {horizontalAlignment=HAlign_Center, verticalAlignment=VAlign_Center}}
  }
  CRichText_Infinite {
   widgetClassPath=&CommonRichTextBlock
   bIsVariable=true
   parent=Overlay_2
   widget=CRichText_Infinite
   properties {
    bOverrideDefaultStyle=true
    justification=Right
    lineHeightPercentage=0.85
    text="∞"
    textTransformPolicy=ToUpper
    defaultTextStyleOverride {
     overflowPolicy=Clip
     transformPolicy=None
     colorAndOpacity {colorUseRule=UseColor_Specified, specifiedColor=<rgba 1 1 1 1>}
     font {
      bForceMonospaced=false
      bMaterialIsStencil=false
      fontMaterial=&M_Ui_Additive_White
      fontObject=&BurbankSmall
      letterSpacing=0
      monospacedWidth=1
      size=16
      skewAmount=0
      typefaceFontName=Bold
      outlineSettings {
       bApplyOutlineToDropShadows=false
       bMiteredCorners=false
       bSeparateFillAlpha=false
       outlineMaterial=null
       outlineSize=0
       outlineColor=<rgba 0 0 0 1>
      }
     }
     highlightColor {colorUseRule=UseColor_Specified, specifiedColor=<rgba 0 0 0 1>}
     highlightShape {
      bIsDynamicallyLoaded=false
      drawAs=Image
      imageType=NoImage
      mirroring=NoMirror
      resourceName=None
      resourceObject=null
      tiling=NoTile
      imageSize=<32 32>
      margin=<ltrb 0 0 0 0>
      outlineSettings {
       bUseBrushTransparency=false
       roundingType=HalfHeightRadius
       width=0
       color {colorUseRule=UseColor_Specified, specifiedColor=<rgba 0 0 0 0>}
       cornerRadii=<0 0 0 0>
      }
      tintColor {colorUseRule=UseColor_Specified, specifiedColor=<rgba 1 1 1 1>}
      uVRegion {bIsValid=false, max=<0 0>, min=<0 0>}
     }
     selectedBackgroundColor {colorUseRule=UseColor_Foreground, specifiedColor=<rgba 1 0 1 1>}
     shadowColorAndOpacity=<rgba 0 0 0 1>
     shadowOffset=<0 0>
     strikeBrush {
      bIsDynamicallyLoaded=false
      drawAs=Image
      imageType=NoImage
      mirroring=NoMirror
      resourceName=None
      resourceObject=null
      tiling=NoTile
      imageSize=<32 32>
      margin=<ltrb 0 0 0 0>
      outlineSettings {
       bUseBrushTransparency=false
       roundingType=HalfHeightRadius
       width=0
       color {colorUseRule=UseColor_Specified, specifiedColor=<rgba 0 0 0 0>}
       cornerRadii=<0 0 0 0>
      }
      tintColor {colorUseRule=UseColor_Specified, specifiedColor=<rgba 1 1 1 1>}
      uVRegion {bIsValid=false, max=<0 0>, min=<0 0>}
     }
     underlineBrush {
      bIsDynamicallyLoaded=false
      drawAs=Image
      imageType=NoImage
      mirroring=NoMirror
      resourceName=None
      resourceObject=null
      tiling=NoTile
      imageSize=<32 32>
      margin=<ltrb 0 0 0 0>
      outlineSettings {
       bUseBrushTransparency=false
       roundingType=HalfHeightRadius
       width=0
       color {colorUseRule=UseColor_Specified, specifiedColor=<rgba 0 0 0 0>}
       cornerRadii=<0 0 0 0>
      }
      tintColor {colorUseRule=UseColor_Specified, specifiedColor=<rgba 1 1 1 1>}
      uVRegion {bIsValid=false, max=<0 0>, min=<0 0>}
     }
    }
    renderTransform {angle=0, scale=<1 1>, shear=<0 0>, translation=<-2 20>}
   }
   slot {class=OverlaySlot, properties {horizontalAlignment=HAlign_Center, verticalAlignment=VAlign_Center}}
  }
  CommonTextBlock_205 {
   widgetClassPath=&CommonTextBlock
   parent=VerticalBox_48
   widget=CommonTextBlock_205
   properties {
    text=E
    colorAndOpacity {colorUseRule=UseColor_Specified, specifiedColor=<rgba 1 0.0529 0.003 1>}
    font {
     bForceMonospaced=false
     bMaterialIsStencil=false
     fontMaterial=&M_Ui_Additive_White_025
     fontObject=&BurbankSmall
     letterSpacing=0
     monospacedWidth=1
     size=12
     skewAmount=0
     typefaceFontName=Bold
     outlineSettings {
      bApplyOutlineToDropShadows=false
      bMiteredCorners=false
      bSeparateFillAlpha=false
      outlineMaterial=null
      outlineSize=0
      outlineColor=<rgba 0 0 0 1>
     }
    }
    shadowOffset=<0 0>
   }
   slot {class=VerticalBoxSlot, properties {horizontalAlignment=HAlign_Center}}
  }
  SizeBox_0 {
   widgetClassPath=&SizeBox
   parent=Overlay_8
   widget=SizeBox_0
   properties {bOverride_WidthOverride=true, widthOverride=150, renderTransform {angle=0, scale=<1 1>, shear=<0 0>, translation=<600 0>}}
   slot {class=OverlaySlot, properties {horizontalAlignment=HAlign_Center, verticalAlignment=VAlign_Center, padding=<ltrb 4 4 4 4>}}
  }
  HBox_LowFuel {widgetClassPath=&HorizontalBox, bIsVariable=true, parent=SizeBox_0, widget=HBox_LowFuel, slot {class=SizeBoxSlot}}
  VBox_LowFuel {
   widgetClassPath=&VerticalBox
   parent=HBox_LowFuel
   widget=VBox_LowFuel
   slot {class=HorizontalBoxSlot, properties {horizontalAlignment=HAlign_Center, verticalAlignment=VAlign_Center}}
  }
  Image_Dot_Top {
   widgetClassPath=&Image
   parent=VBox_LowFuel
   widget=Image_Dot_Top
   properties {brush {
     bIsDynamicallyLoaded=false
     drawAs=Image
     imageType=NoImage
     mirroring=NoMirror
     resourceName=None
     resourceObject=/Tank/UI/Art/M_UI_Tank_Entries_Inst
     tiling=NoTile
     imageSize=<8 8>
     margin=<ltrb 0 0 0 0>
     outlineSettings {
      bUseBrushTransparency=false
      roundingType=HalfHeightRadius
      width=0
      color {colorUseRule=UseColor_Specified, specifiedColor=<rgba 0 0 0 0>}
      cornerRadii=<0 0 0 0>
     }
     tintColor {colorUseRule=UseColor_Specified, specifiedColor=<rgba 1 1 1 1>}
     uVRegion {bIsValid=false, max=<0 0>, min=<0 0>}
    }}
   slot {class=VerticalBoxSlot}
  }
  SizeBox_3 {
   widgetClassPath=&SizeBox
   parent=VBox_LowFuel
   widget=SizeBox_3
   properties {bOverride_HeightOverride=true, heightOverride=64}
   slot {class=VerticalBoxSlot, properties {horizontalAlignment=HAlign_Center, size {sizeRule=Fill, value=1}}}
  }
  Image_242 {
   widgetClassPath=&Image
   bIsVariable=true
   parent=SizeBox_3
   widget=Image_242
   properties {
    brush {
     bIsDynamicallyLoaded=false
     drawAs=Image
     imageType=NoImage
     mirroring=NoMirror
     resourceName=None
     resourceObject=&M_Ui_Additive_White_025
     tiling=NoTile
     imageSize=<2 2>
     margin=<ltrb 0 0 0 0>
     outlineSettings {
      bUseBrushTransparency=false
      roundingType=HalfHeightRadius
      width=0
      color {colorUseRule=UseColor_Specified, specifiedColor=<rgba 0 0 0 0>}
      cornerRadii=<0 0 0 0>
     }
     tintColor {colorUseRule=UseColor_Specified, specifiedColor=<rgba 1 1 1 1>}
     uVRegion {bIsValid=false, max=<0 0>, min=<0 0>}
    }
    colorAndOpacity=<rgba 1 1 1 0.1>
   }
   slot {class=SizeBoxSlot}
  }
  Image_Dot_Bottom {
   widgetClassPath=&Image
   parent=VBox_LowFuel
   widget=Image_Dot_Bottom
   properties {brush {
     bIsDynamicallyLoaded=false
     drawAs=Image
     imageType=NoImage
     mirroring=NoMirror
     resourceName=None
     resourceObject=&M_UI_Tank_Entries_Inst
     tiling=NoTile
     imageSize=<8 8>
     margin=<ltrb 0 0 0 0>
     outlineSettings {
      bUseBrushTransparency=false
      roundingType=HalfHeightRadius
      width=0
      color {colorUseRule=UseColor_Specified, specifiedColor=<rgba 0 0 0 0>}
      cornerRadii=<0 0 0 0>
     }
     tintColor {colorUseRule=UseColor_Specified, specifiedColor=<rgba 1 1 1 1>}
     uVRegion {bIsValid=false, max=<0 0>, min=<0 0>}
    }}
   slot {class=VerticalBoxSlot}
  }
  Overlay_Fuel {
   widgetClassPath=&Overlay
   parent=HBox_LowFuel
   widget=Overlay_Fuel
   properties {clipping=ClipToBounds}
   slot {class=HorizontalBoxSlot, properties {padding=<ltrb -3 0 0 0>}}
  }
  Switcher_Fuel {
   widgetClassPath=&CommonWidgetSwitcherLegacy
   bIsVariable=true
   parent=Overlay_Fuel
   widget=Switcher_Fuel
   slot {class=OverlaySlot, properties {verticalAlignment=VAlign_Fill}}
  }
  VerticalBox_LowFuel {
   widgetClassPath=&VerticalBox
   bIsVariable=true
   parent=Switcher_Fuel
   widget=VerticalBox_LowFuel
   slot {class=WidgetSwitcherSlot, properties {horizontalAlignment=HAlign_Left, verticalAlignment=VAlign_Center, padding=<ltrb 8 8 8 8>}}
  }
  Overlay_Warning {
   widgetClassPath=&Overlay
   bIsVariable=true
   parent=VerticalBox_LowFuel
   widget=Overlay_Warning
   slot {class=VerticalBoxSlot, properties {horizontalAlignment=HAlign_Left}}
  }
  CommonTextBlock_6 {
   widgetClassPath=&CommonTextBlock
   parent=Overlay_Warning
   widget=CommonTextBlock_6
   properties {
    lineHeightPercentage=0.85
    text=Warning
    textTransformPolicy=ToUpper
    colorAndOpacity {colorUseRule=UseColor_Specified, specifiedColor=<rgba 0 0 0 0.25>}
    font {
     bForceMonospaced=false
     bMaterialIsStencil=false
     fontMaterial=null
     fontObject=&BurbankSmall
     letterSpacing=100
     monospacedWidth=1
     size=16
     skewAmount=0
     typefaceFontName=Bold
     outlineSettings {
      bApplyOutlineToDropShadows=false
      bMiteredCorners=false
      bSeparateFillAlpha=false
      outlineMaterial=null
      outlineSize=2
      outlineColor=<rgba 0 0 0 0.1>
     }
    }
    margin=<ltrb 0 -4 0 0>
    shadowOffset=<0 0>
   }
   slot {class=OverlaySlot}
  }
  CommonTextBlock_Warning {
   widgetClassPath=&CommonTextBlock
   bIsVariable=true
   parent=Overlay_Warning
   widget=CommonTextBlock_Warning
   properties {
    lineHeightPercentage=0.85
    text=Warning
    textTransformPolicy=ToUpper
    colorAndOpacity {colorUseRule=UseColor_Specified, specifiedColor=<rgba 1 0.05 0 1>}
    font {
     bForceMonospaced=false
     bMaterialIsStencil=false
     fontMaterial=&M_Ui_Additive_White
     fontObject=&BurbankSmall
     letterSpacing=100
     monospacedWidth=1
     size=16
     skewAmount=0
     typefaceFontName=Bold
     outlineSettings {
      bApplyOutlineToDropShadows=false
      bMiteredCorners=false
      bSeparateFillAlpha=false
      outlineMaterial=null
      outlineSize=2
      outlineColor=<rgba 0 0 0 0>
     }
    }
    margin=<ltrb 0 -4 0 0>
    shadowOffset=<0 0>
   }
   slot {class=OverlaySlot}
  }
  warning_Glow {
   widgetClassPath=&Image
   parent=Overlay_Warning
   widget=warning_Glow
   properties {
    renderOpacity=0.25
    brush {
     bIsDynamicallyLoaded=false
     drawAs=Box
     imageType=NoImage
     mirroring=NoMirror
     resourceName=None
     resourceObject=&M_UI_TextGlow_Add_Gradient_Orange_Nevada
     tiling=NoTile
     imageSize=<24 24>
     margin=<ltrb 0.5 0.5 0.5 0.5>
     outlineSettings {
      bUseBrushTransparency=false
      roundingType=HalfHeightRadius
      width=0
      color {colorUseRule=UseColor_Specified, specifiedColor=<rgba 0 0 0 0>}
      cornerRadii=<0 0 0 0>
     }
     tintColor {colorUseRule=UseColor_Specified, specifiedColor=<rgba 1 1 1 1>}
     uVRegion {bIsValid=false, max=<0 0>, min=<0 0>}
    }
   }
   slot {class=OverlaySlot, properties {horizontalAlignment=HAlign_Fill, verticalAlignment=VAlign_Fill, padding=<ltrb -8 -6 -8 -6>}}
  }
  CommonTextBlock_LowFuel {
   widgetClassPath=&CommonTextBlock
   bIsVariable=true
   parent=VerticalBox_LowFuel
   widget=CommonTextBlock_LowFuel
   properties {
    lineHeightPercentage=0.8
    text=LOW FUEL
    textTransformPolicy=ToUpper
    wrapTextAt=300
    font {
     bForceMonospaced=false
     bMaterialIsStencil=false
     fontMaterial=null
     fontObject=&Burbank
     letterSpacing=0
     monospacedWidth=1
     size=24
     skewAmount=0
     typefaceFontName=BurbankRegular-Black
     outlineSettings {
      bApplyOutlineToDropShadows=false
      bMiteredCorners=false
      bSeparateFillAlpha=false
      outlineMaterial=null
      outlineSize=0
      outlineColor=<rgba 0 0 0 0>
     }
    }
    margin=<ltrb 0 -2 0 0>
    shadowOffset=<0 0>
   }
   slot {class=VerticalBoxSlot}
  }
  VerticalBox_NoFuel {
   widgetClassPath=&VerticalBox
   bIsVariable=true
   parent=Switcher_Fuel
   widget=VerticalBox_NoFuel
   slot {class=WidgetSwitcherSlot, properties {horizontalAlignment=HAlign_Left, verticalAlignment=VAlign_Center, padding=<ltrb 8 8 8 8>}}
  }
  Overlay_Warning_3 {
   widgetClassPath=&Overlay
   bIsVariable=true
   parent=VerticalBox_NoFuel
   widget=Overlay_Warning_3
   slot {class=VerticalBoxSlot, properties {horizontalAlignment=HAlign_Left}}
  }
  CommonTextBlock_7 {
   widgetClassPath=&CommonTextBlock
   parent=Overlay_Warning_3
   widget=CommonTextBlock_7
   properties {
    lineHeightPercentage=0.85
    text=Warning
    textTransformPolicy=ToUpper
    colorAndOpacity {colorUseRule=UseColor_Specified, specifiedColor=<rgba 0 0 0 0.25>}
    font {
     bForceMonospaced=false
     bMaterialIsStencil=false
     fontMaterial=null
     fontObject=&BurbankSmall
     letterSpacing=100
     monospacedWidth=1
     size=16
     skewAmount=0
     typefaceFontName=Bold
     outlineSettings {
      bApplyOutlineToDropShadows=false
      bMiteredCorners=false
      bSeparateFillAlpha=false
      outlineMaterial=null
      outlineSize=2
      outlineColor=<rgba 0 0 0 0.1>
     }
    }
    margin=<ltrb 0 -4 0 0>
    shadowOffset=<0 0>
   }
   slot {class=OverlaySlot}
  }
  CommonTextBlock_Warning_3 {
   widgetClassPath=&CommonTextBlock
   bIsVariable=true
   parent=Overlay_Warning_3
   widget=CommonTextBlock_Warning_3
   properties {
    lineHeightPercentage=0.85
    text=Warning
    textTransformPolicy=ToUpper
    colorAndOpacity {colorUseRule=UseColor_Specified, specifiedColor=<rgba 1 0.05 0 1>}
    font {
     bForceMonospaced=false
     bMaterialIsStencil=false
     fontMaterial=&M_Ui_Additive_White
     fontObject=&BurbankSmall
     letterSpacing=100
     monospacedWidth=1
     size=16
     skewAmount=0
     typefaceFontName=Bold
     outlineSettings {
      bApplyOutlineToDropShadows=false
      bMiteredCorners=false
      bSeparateFillAlpha=false
      outlineMaterial=null
      outlineSize=2
      outlineColor=<rgba 0 0 0 0>
     }
    }
    margin=<ltrb 0 -4 0 0>
    shadowOffset=<0 0>
   }
   slot {class=OverlaySlot}
  }
  warning_Glow_5 {
   widgetClassPath=&Image
   parent=Overlay_Warning_3
   widget=warning_Glow_5
   properties {
    renderOpacity=0.25
    brush {
     bIsDynamicallyLoaded=false
     drawAs=Box
     imageType=NoImage
     mirroring=NoMirror
     resourceName=None
     resourceObject=&M_UI_TextGlow_Add_Gradient_Orange_Nevada
     tiling=NoTile
     imageSize=<24 24>
     margin=<ltrb 0.5 0.5 0.5 0.5>
     outlineSettings {
      bUseBrushTransparency=false
      roundingType=HalfHeightRadius
      width=0
      color {colorUseRule=UseColor_Specified, specifiedColor=<rgba 0 0 0 0>}
      cornerRadii=<0 0 0 0>
     }
     tintColor {colorUseRule=UseColor_Specified, specifiedColor=<rgba 1 1 1 1>}
     uVRegion {bIsValid=false, max=<0 0>, min=<0 0>}
    }
   }
   slot {class=OverlaySlot, properties {horizontalAlignment=HAlign_Fill, verticalAlignment=VAlign_Fill, padding=<ltrb -8 -6 -8 -6>}}
  }
  CommonTextBlock_LowFuel_1 {
   widgetClassPath=&CommonTextBlock
   bIsVariable=true
   parent=VerticalBox_NoFuel
   widget=CommonTextBlock_LowFuel_1
   properties {
    lineHeightPercentage=0.8
    text=NO FUEL
    textTransformPolicy=ToUpper
    wrapTextAt=300
    font {
     bForceMonospaced=false
     bMaterialIsStencil=false
     fontMaterial=null
     fontObject=&Burbank
     letterSpacing=0
     monospacedWidth=1
     size=24
     skewAmount=0
     typefaceFontName=BurbankRegular-Black
     outlineSettings {
      bApplyOutlineToDropShadows=false
      bMiteredCorners=false
      bSeparateFillAlpha=false
      outlineMaterial=null
      outlineSize=0
      outlineColor=<rgba 0 0 0 0>
     }
    }
    margin=<ltrb 0 -2 0 0>
    shadowOffset=<0 0>
   }
   slot {class=VerticalBoxSlot}
  }
  HorizontalBox_Danger {
   widgetClassPath=&VerticalBox
   bIsVariable=true
   parent=Overlay_8
   widget=HorizontalBox_Danger
   properties {visibility=Collapsed, renderTransform {angle=0, scale=<1 1>, shear=<0 0>, translation=<0 -128>}}
   slot {class=OverlaySlot, properties {horizontalAlignment=HAlign_Center, verticalAlignment=VAlign_Bottom}}
  }
  VBox_LowFuel_3 {
   widgetClassPath=&HorizontalBox
   parent=HorizontalBox_Danger
   widget=VBox_LowFuel_3
   slot {class=VerticalBoxSlot, properties {horizontalAlignment=HAlign_Center, padding=<ltrb 0 0 0 -2>}}
  }
  Image_Dot_Top_3 {
   widgetClassPath=&Image
   parent=VBox_LowFuel_3
   widget=Image_Dot_Top_3
   properties {brush {
     bIsDynamicallyLoaded=false
     drawAs=Image
     imageType=NoImage
     mirroring=NoMirror
     resourceName=None
     resourceObject=&M_UI_Tank_Entries_Inst
     tiling=NoTile
     imageSize=<8 8>
     margin=<ltrb 0 0 0 0>
     outlineSettings {
      bUseBrushTransparency=false
      roundingType=HalfHeightRadius
      width=0
      color {colorUseRule=UseColor_Specified, specifiedColor=<rgba 0 0 0 0>}
      cornerRadii=<0 0 0 0>
     }
     tintColor {colorUseRule=UseColor_Specified, specifiedColor=<rgba 1 1 1 1>}
     uVRegion {bIsValid=false, max=<0 0>, min=<0 0>}
    }}
   slot {class=HorizontalBoxSlot}
  }
  SizeBox_5 {
   widgetClassPath=&SizeBox
   parent=VBox_LowFuel_3
   widget=SizeBox_5
   properties {bOverride_WidthOverride=true, widthOverride=64}
   slot {class=HorizontalBoxSlot, properties {verticalAlignment=VAlign_Center}}
  }
  Image_44 {
   widgetClassPath=&Image
   bIsVariable=true
   parent=SizeBox_5
   widget=Image_44
   properties {
    brush {
     bIsDynamicallyLoaded=false
     drawAs=Image
     imageType=NoImage
     mirroring=NoMirror
     resourceName=None
     resourceObject=&M_Ui_Additive_White_025
     tiling=NoTile
     imageSize=<2 2>
     margin=<ltrb 0 0 0 0>
     outlineSettings {
      bUseBrushTransparency=false
      roundingType=HalfHeightRadius
      width=0
      color {colorUseRule=UseColor_Specified, specifiedColor=<rgba 0 0 0 0>}
      cornerRadii=<0 0 0 0>
     }
     tintColor {colorUseRule=UseColor_Specified, specifiedColor=<rgba 1 1 1 1>}
     uVRegion {bIsValid=false, max=<0 0>, min=<0 0>}
    }
    colorAndOpacity=<rgba 1 1 1 0.1>
   }
   slot {class=SizeBoxSlot}
  }
  Image_Dot_Bottom_3 {
   widgetClassPath=&Image
   parent=VBox_LowFuel_3
   widget=Image_Dot_Bottom_3
   properties {brush {
     bIsDynamicallyLoaded=false
     drawAs=Image
     imageType=NoImage
     mirroring=NoMirror
     resourceName=None
     resourceObject=&M_UI_Tank_Entries_Inst
     tiling=NoTile
     imageSize=<8 8>
     margin=<ltrb 0 0 0 0>
     outlineSettings {
      bUseBrushTransparency=false
      roundingType=HalfHeightRadius
      width=0
      color {colorUseRule=UseColor_Specified, specifiedColor=<rgba 0 0 0 0>}
      cornerRadii=<0 0 0 0>
     }
     tintColor {colorUseRule=UseColor_Specified, specifiedColor=<rgba 1 1 1 1>}
     uVRegion {bIsValid=false, max=<0 0>, min=<0 0>}
    }}
   slot {class=HorizontalBoxSlot}
  }
  SizeBox_6 {
   widgetClassPath=&SizeBox
   parent=HorizontalBox_Danger
   widget=SizeBox_6
   properties {bOverride_HeightOverride=true, clipping=ClipToBounds, heightOverride=80}
   slot {class=VerticalBoxSlot, properties {padding=<ltrb 0 -1 0 -1>}}
  }
  VerticalBox_7 {
   widgetClassPath=&VerticalBox
   parent=SizeBox_6
   widget=VerticalBox_7
   slot {class=SizeBoxSlot, properties {verticalAlignment=VAlign_Center}}
  }
  Overlay_Warning_2 {
   widgetClassPath=&Overlay
   bIsVariable=true
   parent=VerticalBox_7
   widget=Overlay_Warning_2
   slot {class=VerticalBoxSlot, properties {horizontalAlignment=HAlign_Center}}
  }
  CommonTextBlock_5 {
   widgetClassPath=&CommonTextBlock
   parent=Overlay_Warning_2
   widget=CommonTextBlock_5
   properties {
    lineHeightPercentage=0.85
    text="Danger!"
    textTransformPolicy=ToUpper
    colorAndOpacity {colorUseRule=UseColor_Specified, specifiedColor=<rgba 0 0 0 0.5>}
    font {
     bForceMonospaced=false
     bMaterialIsStencil=false
     fontMaterial=null
     fontObject=&BurbankSmall
     letterSpacing=100
     monospacedWidth=1
     size=22
     skewAmount=0
     typefaceFontName=Bold
     outlineSettings {
      bApplyOutlineToDropShadows=false
      bMiteredCorners=false
      bSeparateFillAlpha=false
      outlineMaterial=null
      outlineSize=2
      outlineColor=<rgba 0 0 0 0.1>
     }
    }
    margin=<ltrb 0 -4 0 0>
    shadowOffset=<0 0>
   }
   slot {class=OverlaySlot}
  }
  CommonTextBlock_Warning_2 {
   widgetClassPath=&CommonTextBlock
   bIsVariable=true
   parent=Overlay_Warning_2
   widget=CommonTextBlock_Warning_2
   properties {
    lineHeightPercentage=0.85
    text="Danger!"
    textTransformPolicy=ToUpper
    colorAndOpacity {colorUseRule=UseColor_Specified, specifiedColor=<rgba 1 0.05 0 1>}
    font {
     bForceMonospaced=false
     bMaterialIsStencil=false
     fontMaterial=&M_Ui_Additive_White
     fontObject=&BurbankSmall
     letterSpacing=100
     monospacedWidth=1
     size=22
     skewAmount=0
     typefaceFontName=Bold
     outlineSettings {
      bApplyOutlineToDropShadows=false
      bMiteredCorners=false
      bSeparateFillAlpha=false
      outlineMaterial=null
      outlineSize=2
      outlineColor=<rgba 0 0 0 0>
     }
    }
    margin=<ltrb 0 -4 0 0>
    shadowOffset=<0 0>
   }
   slot {class=OverlaySlot}
  }
  warning_Glow_3 {
   widgetClassPath=&Image
   parent=Overlay_Warning_2
   widget=warning_Glow_3
   properties {
    renderOpacity=0.25
    brush {
     bIsDynamicallyLoaded=false
     drawAs=Box
     imageType=NoImage
     mirroring=NoMirror
     resourceName=None
     resourceObject=&M_UI_TextGlow_Add_Gradient_Orange_Nevada
     tiling=NoTile
     imageSize=<32 32>
     margin=<ltrb 0.5 0.5 0.5 0.5>
     outlineSettings {
      bUseBrushTransparency=false
      roundingType=HalfHeightRadius
      width=0
      color {colorUseRule=UseColor_Specified, specifiedColor=<rgba 0 0 0 0>}
      cornerRadii=<0 0 0 0>
     }
     tintColor {colorUseRule=UseColor_Specified, specifiedColor=<rgba 1 1 1 1>}
     uVRegion {bIsValid=false, max=<0 0>, min=<0 0>}
    }
   }
   slot {class=OverlaySlot, properties {horizontalAlignment=HAlign_Fill, verticalAlignment=VAlign_Fill, padding=<ltrb -8 -6 -8 -6>}}
  }
  Overlay_11 {widgetClassPath=&Overlay, parent=VerticalBox_7, widget=Overlay_11, slot {class=VerticalBoxSlot}}
  Text_WarningText_Backing_1 {
   widgetClassPath=&CommonTextBlock
   bIsVariable=true
   parent=Overlay_11
   widget=Text_WarningText_Backing_1
   properties {
    justification=Center
    lineHeightPercentage=0.8
    text="Engine Temp Critical!"
    textTransformPolicy=ToUpper
    colorAndOpacity {colorUseRule=UseColor_Specified, specifiedColor=<rgba 0 0 0 0.5>}
    font {
     bForceMonospaced=false
     bMaterialIsStencil=false
     fontMaterial=null
     fontObject=&Burbank
     letterSpacing=0
     monospacedWidth=1
     size=24
     skewAmount=0
     typefaceFontName=BurbankRegular-Black
     outlineSettings {
      bApplyOutlineToDropShadows=false
      bMiteredCorners=false
      bSeparateFillAlpha=false
      outlineMaterial=null
      outlineSize=2
      outlineColor=<rgba 0 0 0 0.1>
     }
    }
    margin=<ltrb 0 -2 0 0>
    shadowOffset=<0 0>
   }
   slot {class=OverlaySlot}
  }
  Text_Warningtext_1 {
   widgetClassPath=&CommonTextBlock
   bIsVariable=true
   parent=Overlay_11
   widget=Text_Warningtext_1
   properties {
    justification=Center
    lineHeightPercentage=0.8
    text="Engine Temp Critical!"
    textTransformPolicy=ToUpper
    font {
     bForceMonospaced=false
     bMaterialIsStencil=false
     fontMaterial=/Tank/UI/Art/M_Ui_Additive_Orange
     fontObject=&Burbank
     letterSpacing=0
     monospacedWidth=1
     size=24
     skewAmount=0
     typefaceFontName=BurbankRegular-Black
     outlineSettings {
      bApplyOutlineToDropShadows=false
      bMiteredCorners=false
      bSeparateFillAlpha=false
      outlineMaterial=null
      outlineSize=2
      outlineColor=<rgba 0 0 0 0>
     }
    }
    margin=<ltrb 0 -2 0 0>
    shadowOffset=<0 0>
   }
   slot {class=OverlaySlot}
  }
  warning_Glow_4 {
   widgetClassPath=&Image
   parent=Overlay_11
   widget=warning_Glow_4
   properties {
    renderOpacity=0.25
    brush {
     bIsDynamicallyLoaded=false
     drawAs=Box
     imageType=NoImage
     mirroring=NoMirror
     resourceName=None
     resourceObject=&M_UI_TextGlow_Add_Gradient_Orange_Nevada
     tiling=NoTile
     imageSize=<40 40>
     margin=<ltrb 0.5 0.5 0.5 0.5>
     outlineSettings {
      bUseBrushTransparency=false
      roundingType=HalfHeightRadius
      width=0
      color {colorUseRule=UseColor_Specified, specifiedColor=<rgba 0 0 0 0>}
      cornerRadii=<0 0 0 0>
     }
     tintColor {colorUseRule=UseColor_Specified, specifiedColor=<rgba 1 1 1 1>}
     uVRegion {bIsValid=false, max=<0 0>, min=<0 0>}
    }
   }
   slot {class=OverlaySlot, properties {horizontalAlignment=HAlign_Fill, verticalAlignment=VAlign_Fill, padding=<ltrb -8 -6 -8 -10>}}
  }
  VBox_LowFuel_2 {
   widgetClassPath=&HorizontalBox
   parent=HorizontalBox_Danger
   widget=VBox_LowFuel_2
   slot {class=VerticalBoxSlot, properties {horizontalAlignment=HAlign_Center, padding=<ltrb 0 -2 0 0>}}
  }
  Image_Dot_Top_2 {
   widgetClassPath=&Image
   parent=VBox_LowFuel_2
   widget=Image_Dot_Top_2
   properties {brush {
     bIsDynamicallyLoaded=false
     drawAs=Image
     imageType=NoImage
     mirroring=NoMirror
     resourceName=None
     resourceObject=&M_UI_Tank_Entries_Inst
     tiling=NoTile
     imageSize=<8 8>
     margin=<ltrb 0 0 0 0>
     outlineSettings {
      bUseBrushTransparency=false
      roundingType=HalfHeightRadius
      width=0
      color {colorUseRule=UseColor_Specified, specifiedColor=<rgba 0 0 0 0>}
      cornerRadii=<0 0 0 0>
     }
     tintColor {colorUseRule=UseColor_Specified, specifiedColor=<rgba 1 1 1 1>}
     uVRegion {bIsValid=false, max=<0 0>, min=<0 0>}
    }}
   slot {class=HorizontalBoxSlot}
  }
  SizeBox_4 {
   widgetClassPath=&SizeBox
   parent=VBox_LowFuel_2
   widget=SizeBox_4
   properties {bOverride_WidthOverride=true, widthOverride=64}
   slot {class=HorizontalBoxSlot, properties {verticalAlignment=VAlign_Center}}
  }
  Image_43 {
   widgetClassPath=&Image
   bIsVariable=true
   parent=SizeBox_4
   widget=Image_43
   properties {
    brush {
     bIsDynamicallyLoaded=false
     drawAs=Image
     imageType=NoImage
     mirroring=NoMirror
     resourceName=None
     resourceObject=&M_Ui_Additive_White_025
     tiling=NoTile
     imageSize=<2 2>
     margin=<ltrb 0 0 0 0>
     outlineSettings {
      bUseBrushTransparency=false
      roundingType=HalfHeightRadius
      width=0
      color {colorUseRule=UseColor_Specified, specifiedColor=<rgba 0 0 0 0>}
      cornerRadii=<0 0 0 0>
     }
     tintColor {colorUseRule=UseColor_Specified, specifiedColor=<rgba 1 1 1 1>}
     uVRegion {bIsValid=false, max=<0 0>, min=<0 0>}
    }
    colorAndOpacity=<rgba 1 1 1 0.1>
   }
   slot {class=SizeBoxSlot}
  }
  Image_Dot_Bottom_2 {
   widgetClassPath=&Image
   parent=VBox_LowFuel_2
   widget=Image_Dot_Bottom_2
   properties {brush {
     bIsDynamicallyLoaded=false
     drawAs=Image
     imageType=NoImage
     mirroring=NoMirror
     resourceName=None
     resourceObject=&M_UI_Tank_Entries_Inst
     tiling=NoTile
     imageSize=<8 8>
     margin=<ltrb 0 0 0 0>
     outlineSettings {
      bUseBrushTransparency=false
      roundingType=HalfHeightRadius
      width=0
      color {colorUseRule=UseColor_Specified, specifiedColor=<rgba 0 0 0 0>}
      cornerRadii=<0 0 0 0>
     }
     tintColor {colorUseRule=UseColor_Specified, specifiedColor=<rgba 1 1 1 1>}
     uVRegion {bIsValid=false, max=<0 0>, min=<0 0>}
    }}
   slot {class=HorizontalBoxSlot}
  }
  SafeZone_Danger {
   widgetClassPath=&SafeZone
   bIsVariable=true
   parent=HorizontalBox_Danger
   widget=SafeZone_Danger
   properties {padLeft=false, padRight=false, padTop=false}
   slot {class=VerticalBoxSlot}
  }
  SafeZone_TankIcon {
   widgetClassPath=&SafeZone
   bIsVariable=true
   parent=Overlay_8
   widget=SafeZone_TankIcon
   properties {padBottom=false, padLeft=false, padTop=false}
   slot {class=OverlaySlot, properties {horizontalAlignment=HAlign_Fill, verticalAlignment=VAlign_Fill}}
  }
  OV_SafezoneMobile {widgetClassPath=&Overlay, parent=SafeZone_TankIcon, widget=OV_SafezoneMobile, slot {class=SafeZoneSlot}}
  Overlay_TankSilhouette {
   widgetClassPath=&Overlay
   bIsVariable=true
   parent=OV_SafezoneMobile
   widget=Overlay_TankSilhouette
   properties {renderTransform {angle=0, scale=<1 1>, shear=<0 0>, translation=<-70 -70>}}
   slot {class=OverlaySlot, properties {horizontalAlignment=HAlign_Right, verticalAlignment=VAlign_Bottom}}
  }
  Image_Tank_1 {
   widgetClassPath=&Image
   bIsVariable=true
   parent=Overlay_TankSilhouette
   widget=Image_Tank_1
   properties {brush {
     bIsDynamicallyLoaded=false
     drawAs=Image
     imageType=NoImage
     mirroring=NoMirror
     resourceName=None
     resourceObject=/Tank/UI/Art/M_UI_TankIcon_Master_Body1
     tiling=NoTile
     imageSize=<128 128>
     margin=<ltrb 0 0 0 0>
     outlineSettings {
      bUseBrushTransparency=false
      roundingType=HalfHeightRadius
      width=0
      color {colorUseRule=UseColor_Specified, specifiedColor=<rgba 0 0 0 0>}
      cornerRadii=<0 0 0 0>
     }
     tintColor {colorUseRule=UseColor_Specified, specifiedColor=<rgba 0 0 0 0.75>}
     uVRegion {bIsValid=false, max=<0 0>, min=<0 0>}
    }}
   slot {class=OverlaySlot}
  }
  Image_Tank {
   widgetClassPath=&Image
   bIsVariable=true
   parent=Overlay_TankSilhouette
   widget=Image_Tank
   properties {brush {
     bIsDynamicallyLoaded=false
     drawAs=Image
     imageType=NoImage
     mirroring=NoMirror
     resourceName=None
     resourceObject=/Tank/UI/Art/M_UI_TankIcon_Master_Body
     tiling=NoTile
     imageSize=<128 128>
     margin=<ltrb 0 0 0 0>
     outlineSettings {
      bUseBrushTransparency=false
      roundingType=HalfHeightRadius
      width=0
      color {colorUseRule=UseColor_Specified, specifiedColor=<rgba 0 0 0 0>}
      cornerRadii=<0 0 0 0>
     }
     tintColor {colorUseRule=UseColor_Specified, specifiedColor=<rgba 1 1 1 1>}
     uVRegion {bIsValid=false, max=<0 0>, min=<0 0>}
    }}
   slot {class=OverlaySlot}
  }
  Image_RightTread_1 {
   widgetClassPath=&Image
   bIsVariable=true
   parent=Overlay_TankSilhouette
   widget=Image_RightTread_1
   properties {
    brush {
     bIsDynamicallyLoaded=false
     drawAs=Image
     imageType=NoImage
     mirroring=NoMirror
     resourceName=None
     resourceObject=/Tank/UI/Art/M_UI_TankIcon_Master_Thread_Red
     tiling=NoTile
     imageSize=<32 128>
     margin=<ltrb 0 0 0 0>
     outlineSettings {
      bUseBrushTransparency=false
      roundingType=HalfHeightRadius
      width=0
      color {colorUseRule=UseColor_Specified, specifiedColor=<rgba 0 0 0 0>}
      cornerRadii=<0 0 0 0>
     }
     tintColor {colorUseRule=UseColor_Specified, specifiedColor=<rgba 0 0 0 0.75>}
     uVRegion {bIsValid=false, max=<0 0>, min=<0 0>}
    }
    renderTransform {angle=0, scale=<-1 1>, shear=<0 0>, translation=<-14 -5>}
   }
   slot {class=OverlaySlot, properties {horizontalAlignment=HAlign_Right}}
  }
  Image_LeftTread_1 {
   widgetClassPath=&Image
   bIsVariable=true
   parent=Overlay_TankSilhouette
   widget=Image_LeftTread_1
   properties {
    brush {
     bIsDynamicallyLoaded=false
     drawAs=Image
     imageType=NoImage
     mirroring=NoMirror
     resourceName=None
     resourceObject=&M_UI_TankIcon_Master_Thread_Red
     tiling=NoTile
     imageSize=<32 128>
     margin=<ltrb 0 0 0 0>
     outlineSettings {
      bUseBrushTransparency=false
      roundingType=HalfHeightRadius
      width=0
      color {colorUseRule=UseColor_Specified, specifiedColor=<rgba 0 0 0 0>}
      cornerRadii=<0 0 0 0>
     }
     tintColor {colorUseRule=UseColor_Specified, specifiedColor=<rgba 0 0 0 0.75>}
     uVRegion {bIsValid=false, max=<0 0>, min=<0 0>}
    }
    renderTransform {angle=0, scale=<1 1>, shear=<0 0>, translation=<14 -5>}
   }
   slot {class=OverlaySlot}
  }
  HorizontalBox {
   widgetClassPath=&HorizontalBox
   parent=Overlay_TankSilhouette
   widget=HorizontalBox
   properties {renderTransform {angle=0, scale=<1.1 1>, shear=<0 0>, translation=<0 13>}}
   slot {class=OverlaySlot, properties {horizontalAlignment=HAlign_Center, verticalAlignment=VAlign_Bottom}}
  }
  Image_LeftEngine_1 {
   widgetClassPath=&Image
   bIsVariable=true
   parent=HorizontalBox
   widget=Image_LeftEngine_1
   properties {brush {
     bIsDynamicallyLoaded=false
     drawAs=Image
     imageType=NoImage
     mirroring=NoMirror
     resourceName=None
     resourceObject=/Tank/UI/Art/M_UI_TankIcon_Master_Engine1
     tiling=NoTile
     imageSize=<32 32>
     margin=<ltrb 0 0 0 0>
     outlineSettings {
      bUseBrushTransparency=false
      roundingType=HalfHeightRadius
      width=0
      color {colorUseRule=UseColor_Specified, specifiedColor=<rgba 0 0 0 0>}
      cornerRadii=<0 0 0 0>
     }
     tintColor {colorUseRule=UseColor_Specified, specifiedColor=<rgba 0 0 0 0.75>}
     uVRegion {bIsValid=false, max=<0 0>, min=<0 0>}
    }}
   slot {class=HorizontalBoxSlot, properties {padding=<ltrb -3 0 -3 0>}}
  }
  Image_RightEngine_1 {
   widgetClassPath=&Image
   bIsVariable=true
   parent=HorizontalBox
   widget=Image_RightEngine_1
   properties {
    brush {
     bIsDynamicallyLoaded=false
     drawAs=Image
     imageType=NoImage
     mirroring=NoMirror
     resourceName=None
     resourceObject=&M_UI_TankIcon_Master_Engine1
     tiling=NoTile
     imageSize=<32 32>
     margin=<ltrb 0 0 0 0>
     outlineSettings {
      bUseBrushTransparency=false
      roundingType=HalfHeightRadius
      width=0
      color {colorUseRule=UseColor_Specified, specifiedColor=<rgba 0 0 0 0>}
      cornerRadii=<0 0 0 0>
     }
     tintColor {colorUseRule=UseColor_Specified, specifiedColor=<rgba 0 0 0 0.75>}
     uVRegion {bIsValid=false, max=<0 0>, min=<0 0>}
    }
    renderTransform {angle=0, scale=<-1 1>, shear=<0 0>, translation=<0 0>}
   }
   slot {class=HorizontalBoxSlot, properties {padding=<ltrb -3 0 -3 0>}}
  }
  Image_Hatch_backing {
   widgetClassPath=&Image
   bIsVariable=true
   parent=Overlay_TankSilhouette
   widget=Image_Hatch_backing
   properties {
    brush {
     bIsDynamicallyLoaded=false
     drawAs=Image
     imageType=NoImage
     mirroring=NoMirror
     resourceName=None
     resourceObject=/Tank/UI/Art/M_UI_TankIcon_Hartch_Red
     tiling=NoTile
     imageSize=<32 32>
     margin=<ltrb 0 0 0 0>
     outlineSettings {
      bUseBrushTransparency=false
      roundingType=HalfHeightRadius
      width=0
      color {colorUseRule=UseColor_Specified, specifiedColor=<rgba 0 0 0 0>}
      cornerRadii=<0 0 0 0>
     }
     tintColor {colorUseRule=UseColor_Specified, specifiedColor=<rgba 0 0 0 0.75>}
     uVRegion {bIsValid=false, max=<0 0>, min=<0 0>}
    }
    renderTransform {angle=0, scale=<1 1>, shear=<0 0>, translation=<0 -31>}
   }
   slot {class=OverlaySlot, properties {horizontalAlignment=HAlign_Center, verticalAlignment=VAlign_Center}}
  }
  Overlay_Tank_Turret {
   widgetClassPath=&Overlay
   bIsVariable=true
   parent=Overlay_TankSilhouette
   widget=Overlay_Tank_Turret
   properties {renderTransformPivot=<0.5 0.66>}
   slot {class=OverlaySlot}
  }
  Image_BAcking {
   widgetClassPath=&Image
   bIsVariable=true
   parent=Overlay_Tank_Turret
   widget=Image_BAcking
   properties {
    brush {
     bIsDynamicallyLoaded=false
     drawAs=Image
     imageType=NoImage
     mirroring=NoMirror
     resourceName=None
     resourceObject=/Tank/UI/Art/M_UI_TankIcon_Master_transluscent
     tiling=NoTile
     imageSize=<128 128>
     margin=<ltrb 0 0 0 0>
     outlineSettings {
      bUseBrushTransparency=false
      roundingType=HalfHeightRadius
      width=0
      color {colorUseRule=UseColor_Specified, specifiedColor=<rgba 0 0 0 0>}
      cornerRadii=<0 0 0 0>
     }
     tintColor {colorUseRule=UseColor_Specified, specifiedColor=<rgba 0 0 0 0.75>}
     uVRegion {bIsValid=false, max=<0 0>, min=<0 0>}
    }
    renderTransform {angle=-45, scale=<1 1>, shear=<0 0>, translation=<0 -16>}
   }
   slot {class=OverlaySlot}
  }
  Image_220 {
   widgetClassPath=&Image
   bIsVariable=true
   parent=Overlay_Tank_Turret
   widget=Image_220
   properties {
    brush {
     bIsDynamicallyLoaded=false
     drawAs=Image
     imageType=NoImage
     mirroring=NoMirror
     resourceName=None
     resourceObject=/Tank/UI/Art/M_UI_TankIcon_Master
     tiling=NoTile
     imageSize=<128 128>
     margin=<ltrb 0 0 0 0>
     outlineSettings {
      bUseBrushTransparency=false
      roundingType=HalfHeightRadius
      width=0
      color {colorUseRule=UseColor_Specified, specifiedColor=<rgba 0 0 0 0>}
      cornerRadii=<0 0 0 0>
     }
     tintColor {colorUseRule=UseColor_Specified, specifiedColor=<rgba 1 1 1 1>}
     uVRegion {bIsValid=false, max=<0 0>, min=<0 0>}
    }
    renderTransform {angle=-45, scale=<1 1>, shear=<0 0>, translation=<0 -16>}
   }
   slot {class=OverlaySlot}
  }
  Image_LeftTread {
   widgetClassPath=&Image
   bIsVariable=true
   parent=Overlay_TankSilhouette
   widget=Image_LeftTread
   properties {
    brush {
     bIsDynamicallyLoaded=false
     drawAs=Image
     imageType=NoImage
     mirroring=NoMirror
     resourceName=None
     resourceObject=/Tank/UI/Art/M_UI_TankIcon_Master_Thread
     tiling=NoTile
     imageSize=<32 128>
     margin=<ltrb 0 0 0 0>
     outlineSettings {
      bUseBrushTransparency=false
      roundingType=HalfHeightRadius
      width=0
      color {colorUseRule=UseColor_Specified, specifiedColor=<rgba 0 0 0 0>}
      cornerRadii=<0 0 0 0>
     }
     tintColor {colorUseRule=UseColor_Specified, specifiedColor=<rgba 1 1 1 1>}
     uVRegion {bIsValid=false, max=<0 0>, min=<0 0>}
    }
    renderTransform {angle=0, scale=<1 1>, shear=<0 0>, translation=<14 -5>}
   }
   slot {class=OverlaySlot}
  }
  Image_RightTread {
   widgetClassPath=&Image
   bIsVariable=true
   parent=Overlay_TankSilhouette
   widget=Image_RightTread
   properties {
    brush {
     bIsDynamicallyLoaded=false
     drawAs=Image
     imageType=NoImage
     mirroring=NoMirror
     resourceName=None
     resourceObject=&M_UI_TankIcon_Master_Thread
     tiling=NoTile
     imageSize=<32 128>
     margin=<ltrb 0 0 0 0>
     outlineSettings {
      bUseBrushTransparency=false
      roundingType=HalfHeightRadius
      width=0
      color {colorUseRule=UseColor_Specified, specifiedColor=<rgba 0 0 0 0>}
      cornerRadii=<0 0 0 0>
     }
     tintColor {colorUseRule=UseColor_Specified, specifiedColor=<rgba 1 1 1 1>}
     uVRegion {bIsValid=false, max=<0 0>, min=<0 0>}
    }
    renderTransform {angle=0, scale=<-1 1>, shear=<0 0>, translation=<-14 -5>}
   }
   slot {class=OverlaySlot, properties {horizontalAlignment=HAlign_Right}}
  }
  Image_Hatch {
   widgetClassPath=&Image
   bIsVariable=true
   parent=Overlay_TankSilhouette
   widget=Image_Hatch
   properties {
    brush {
     bIsDynamicallyLoaded=false
     drawAs=Image
     imageType=NoImage
     mirroring=NoMirror
     resourceName=None
     resourceObject=/Tank/UI/Art/M_UI_TankIcon_Master_Hatch
     tiling=NoTile
     imageSize=<32 32>
     margin=<ltrb 0 0 0 0>
     outlineSettings {
      bUseBrushTransparency=false
      roundingType=HalfHeightRadius
      width=0
      color {colorUseRule=UseColor_Specified, specifiedColor=<rgba 0 0 0 0>}
      cornerRadii=<0 0 0 0>
     }
     tintColor {colorUseRule=UseColor_Specified, specifiedColor=<rgba 1 1 1 1>}
     uVRegion {bIsValid=false, max=<0 0>, min=<0 0>}
    }
    renderTransform {angle=0, scale=<1 1>, shear=<0 0>, translation=<0 -31>}
   }
   slot {class=OverlaySlot, properties {horizontalAlignment=HAlign_Center, verticalAlignment=VAlign_Center}}
  }
  HorizontalBox_1 {
   widgetClassPath=&HorizontalBox
   parent=Overlay_TankSilhouette
   widget=HorizontalBox_1
   properties {renderTransform {angle=0, scale=<1.1 1>, shear=<0 0>, translation=<0 13>}}
   slot {class=OverlaySlot, properties {horizontalAlignment=HAlign_Center, verticalAlignment=VAlign_Bottom}}
  }
  Image_LeftEngine {
   widgetClassPath=&Image
   bIsVariable=true
   parent=HorizontalBox_1
   widget=Image_LeftEngine
   properties {brush {
     bIsDynamicallyLoaded=false
     drawAs=Image
     imageType=NoImage
     mirroring=NoMirror
     resourceName=None
     resourceObject=/Tank/UI/Art/M_UI_TankIcon_Master_Engine
     tiling=NoTile
     imageSize=<32 32>
     margin=<ltrb 0 0 0 0>
     outlineSettings {
      bUseBrushTransparency=false
      roundingType=HalfHeightRadius
      width=0
      color {colorUseRule=UseColor_Specified, specifiedColor=<rgba 0 0 0 0>}
      cornerRadii=<0 0 0 0>
     }
     tintColor {colorUseRule=UseColor_Specified, specifiedColor=<rgba 1 1 1 1>}
     uVRegion {bIsValid=false, max=<0 0>, min=<0 0>}
    }}
   slot {class=HorizontalBoxSlot, properties {padding=<ltrb -3 0 -3 0>}}
  }
  Image_RightEngine {
   widgetClassPath=&Image
   bIsVariable=true
   parent=HorizontalBox_1
   widget=Image_RightEngine
   properties {
    brush {
     bIsDynamicallyLoaded=false
     drawAs=Image
     imageType=NoImage
     mirroring=NoMirror
     resourceName=None
     resourceObject=&M_UI_TankIcon_Master_Engine
     tiling=NoTile
     imageSize=<32 32>
     margin=<ltrb 0 0 0 0>
     outlineSettings {
      bUseBrushTransparency=false
      roundingType=HalfHeightRadius
      width=0
      color {colorUseRule=UseColor_Specified, specifiedColor=<rgba 0 0 0 0>}
      cornerRadii=<0 0 0 0>
     }
     tintColor {colorUseRule=UseColor_Specified, specifiedColor=<rgba 1 1 1 1>}
     uVRegion {bIsValid=false, max=<0 0>, min=<0 0>}
    }
    renderTransform {angle=0, scale=<-1 1>, shear=<0 0>, translation=<0 0>}
   }
   slot {class=HorizontalBoxSlot, properties {padding=<ltrb -3 0 -3 0>}}
  }
  HBox_SpeedLines {
   widgetClassPath=&HorizontalBox
   bIsVariable=true
   parent=Overlay_TankSilhouette
   widget=HBox_SpeedLines
   properties {renderOpacity=0, renderTransform {angle=0, scale=<1.1 0.95>, shear=<0 0>, translation=<0 40>}, renderTransformPivot=<0.5 0>}
   slot {class=OverlaySlot, properties {horizontalAlignment=HAlign_Center, verticalAlignment=VAlign_Bottom}}
  }
  Image_SpeedRight {
   widgetClassPath=&Image
   parent=HBox_SpeedLines
   widget=Image_SpeedRight
   properties {brush {
     bIsDynamicallyLoaded=false
     drawAs=Image
     imageType=NoImage
     mirroring=NoMirror
     resourceName=None
     resourceObject=/Tank/UI/Art/M_UI_TankIcon_SpeedLines
     tiling=NoTile
     imageSize=<32 32>
     margin=<ltrb 0 0 0 0>
     outlineSettings {
      bUseBrushTransparency=false
      roundingType=HalfHeightRadius
      width=0
      color {colorUseRule=UseColor_Specified, specifiedColor=<rgba 0 0 0 0>}
      cornerRadii=<0 0 0 0>
     }
     tintColor {colorUseRule=UseColor_Specified, specifiedColor=<rgba 1 1 1 1>}
     uVRegion {bIsValid=false, max=<0 0>, min=<0 0>}
    }}
   slot {class=HorizontalBoxSlot, properties {padding=<ltrb 4 0 4 0>}}
  }
  Image_Speed_Left {
   widgetClassPath=&Image
   parent=HBox_SpeedLines
   widget=Image_Speed_Left
   properties {
    brush {
     bIsDynamicallyLoaded=false
     drawAs=Image
     imageType=NoImage
     mirroring=NoMirror
     resourceName=None
     resourceObject=&M_UI_TankIcon_SpeedLines
     tiling=NoTile
     imageSize=<32 32>
     margin=<ltrb 0 0 0 0>
     outlineSettings {
      bUseBrushTransparency=false
      roundingType=HalfHeightRadius
      width=0
      color {colorUseRule=UseColor_Specified, specifiedColor=<rgba 0 0 0 0>}
      cornerRadii=<0 0 0 0>
     }
     tintColor {colorUseRule=UseColor_Specified, specifiedColor=<rgba 1 1 1 1>}
     uVRegion {bIsValid=false, max=<0 0>, min=<0 0>}
    }
    renderTransform {angle=0, scale=<-1 1>, shear=<0 0>, translation=<0 0>}
   }
   slot {class=HorizontalBoxSlot, properties {padding=<ltrb 4 0 4 0>}}
  }
  HorizontalBox_Warning {
   widgetClassPath=&HorizontalBox
   bIsVariable=true
   parent=OV_SafezoneMobile
   widget=HorizontalBox_Warning
   properties {visibility=Collapsed, renderTransform {angle=0, scale=<1 1>, shear=<0 0>, translation=<-240 -87>}}
   slot {class=OverlaySlot, properties {horizontalAlignment=HAlign_Right, verticalAlignment=VAlign_Bottom}}
  }
  Overlay_10 {
   widgetClassPath=&Overlay
   parent=HorizontalBox_Warning
   widget=Overlay_10
   properties {clipping=ClipToBounds}
   slot {class=HorizontalBoxSlot, properties {padding=<ltrb 0 0 -3 0>}}
  }
  VerticalBox_8 {
   widgetClassPath=&VerticalBox
   parent=Overlay_10
   widget=VerticalBox_8
   slot {class=OverlaySlot, properties {verticalAlignment=VAlign_Center, padding=<ltrb 6 0 6 0>}}
  }
  Overlay_Warning_1 {
   widgetClassPath=&Overlay
   bIsVariable=true
   parent=VerticalBox_8
   widget=Overlay_Warning_1
   slot {class=VerticalBoxSlot, properties {horizontalAlignment=HAlign_Right}}
  }
  CommonTextBlock_4 {
   widgetClassPath=&CommonTextBlock
   parent=Overlay_Warning_1
   widget=CommonTextBlock_4
   properties {
    lineHeightPercentage=0.85
    text=Warning
    textTransformPolicy=ToUpper
    colorAndOpacity {colorUseRule=UseColor_Specified, specifiedColor=<rgba 0 0 0 0.5>}
    font {
     bForceMonospaced=false
     bMaterialIsStencil=false
     fontMaterial=null
     fontObject=&BurbankSmall
     letterSpacing=100
     monospacedWidth=1
     size=16
     skewAmount=0
     typefaceFontName=Bold
     outlineSettings {
      bApplyOutlineToDropShadows=false
      bMiteredCorners=false
      bSeparateFillAlpha=false
      outlineMaterial=null
      outlineSize=2
      outlineColor=<rgba 0 0 0 0.1>
     }
    }
    margin=<ltrb 0 -4 0 0>
    shadowOffset=<0 0>
   }
   slot {class=OverlaySlot}
  }
  CommonTextBlock_Warning_1 {
   widgetClassPath=&CommonTextBlock
   bIsVariable=true
   parent=Overlay_Warning_1
   widget=CommonTextBlock_Warning_1
   properties {
    lineHeightPercentage=0.85
    text=Warning
    textTransformPolicy=ToUpper
    colorAndOpacity {colorUseRule=UseColor_Specified, specifiedColor=<rgba 1 0.05 0 1>}
    font {
     bForceMonospaced=false
     bMaterialIsStencil=false
     fontMaterial=&M_Ui_Additive_White
     fontObject=&BurbankSmall
     letterSpacing=100
     monospacedWidth=1
     size=16
     skewAmount=0
     typefaceFontName=Bold
     outlineSettings {
      bApplyOutlineToDropShadows=false
      bMiteredCorners=false
      bSeparateFillAlpha=false
      outlineMaterial=null
      outlineSize=2
      outlineColor=<rgba 0 0 0 0>
     }
    }
    margin=<ltrb 0 -4 0 0>
    shadowOffset=<0 0>
   }
   slot {class=OverlaySlot}
  }
  warning_Glow_1 {
   widgetClassPath=&Image
   parent=Overlay_Warning_1
   widget=warning_Glow_1
   properties {
    renderOpacity=0.25
    brush {
     bIsDynamicallyLoaded=false
     drawAs=Box
     imageType=NoImage
     mirroring=NoMirror
     resourceName=None
     resourceObject=&M_UI_TextGlow_Add_Gradient_Orange_Nevada
     tiling=NoTile
     imageSize=<24 24>
     margin=<ltrb 0.5 0.5 0.5 0.5>
     outlineSettings {
      bUseBrushTransparency=false
      roundingType=HalfHeightRadius
      width=0
      color {colorUseRule=UseColor_Specified, specifiedColor=<rgba 0 0 0 0>}
      cornerRadii=<0 0 0 0>
     }
     tintColor {colorUseRule=UseColor_Specified, specifiedColor=<rgba 1 1 1 1>}
     uVRegion {bIsValid=false, max=<0 0>, min=<0 0>}
    }
   }
   slot {class=OverlaySlot, properties {horizontalAlignment=HAlign_Fill, verticalAlignment=VAlign_Fill, padding=<ltrb -8 -6 -8 -6>}}
  }
  Overlay_TreadDestroyed {widgetClassPath=&Overlay, bIsVariable=true, parent=VerticalBox_8, widget=Overlay_TreadDestroyed, slot {class=VerticalBoxSlot}}
  Text_WarningText_Backing {
   widgetClassPath=&CommonTextBlock
   bIsVariable=true
   parent=Overlay_TreadDestroyed
   widget=Text_WarningText_Backing
   properties {
    justification=Center
    lineHeightPercentage=0.8
    text=Tread Destroyed
    textTransformPolicy=ToUpper
    colorAndOpacity {colorUseRule=UseColor_Specified, specifiedColor=<rgba 0 0 0 0.5>}
    font {
     bForceMonospaced=false
     bMaterialIsStencil=false
     fontMaterial=null
     fontObject=&Burbank
     letterSpacing=0
     monospacedWidth=1
     size=24
     skewAmount=0
     typefaceFontName=BurbankRegular-Black
     outlineSettings {
      bApplyOutlineToDropShadows=false
      bMiteredCorners=false
      bSeparateFillAlpha=false
      outlineMaterial=null
      outlineSize=2
      outlineColor=<rgba 0 0 0 0.1>
     }
    }
    margin=<ltrb 0 -2 0 0>
    shadowOffset=<0 0>
   }
   slot {class=OverlaySlot}
  }
  Text_Warningtext {
   widgetClassPath=&CommonTextBlock
   bIsVariable=true
   parent=Overlay_TreadDestroyed
   widget=Text_Warningtext
   properties {
    justification=Center
    lineHeightPercentage=0.8
    text=Tread Destroyed
    textTransformPolicy=ToUpper
    font {
     bForceMonospaced=false
     bMaterialIsStencil=false
     fontMaterial=&M_Ui_Additive_Orange
     fontObject=&Burbank
     letterSpacing=0
     monospacedWidth=1
     size=24
     skewAmount=0
     typefaceFontName=BurbankRegular-Black
     outlineSettings {
      bApplyOutlineToDropShadows=false
      bMiteredCorners=false
      bSeparateFillAlpha=false
      outlineMaterial=null
      outlineSize=2
      outlineColor=<rgba 0 0 0 0>
     }
    }
    margin=<ltrb 0 -2 0 0>
    shadowOffset=<0 0>
   }
   slot {class=OverlaySlot}
  }
  warning_Glow_2 {
   widgetClassPath=&Image
   parent=Overlay_TreadDestroyed
   widget=warning_Glow_2
   properties {
    renderOpacity=0.25
    brush {
     bIsDynamicallyLoaded=false
     drawAs=Box
     imageType=NoImage
     mirroring=NoMirror
     resourceName=None
     resourceObject=&M_UI_TextGlow_Add_Gradient_Orange_Nevada
     tiling=NoTile
     imageSize=<32 32>
     margin=<ltrb 0.5 0.5 0.5 0.5>
     outlineSettings {
      bUseBrushTransparency=false
      roundingType=HalfHeightRadius
      width=0
      color {colorUseRule=UseColor_Specified, specifiedColor=<rgba 0 0 0 0>}
      cornerRadii=<0 0 0 0>
     }
     tintColor {colorUseRule=UseColor_Specified, specifiedColor=<rgba 1 1 1 1>}
     uVRegion {bIsValid=false, max=<0 0>, min=<0 0>}
    }
   }
   slot {class=OverlaySlot, properties {horizontalAlignment=HAlign_Fill, verticalAlignment=VAlign_Fill, padding=<ltrb -8 -6 -8 -10>}}
  }
  VBox_LowFuel_1 {
   widgetClassPath=&VerticalBox
   parent=HorizontalBox_Warning
   widget=VBox_LowFuel_1
   slot {class=HorizontalBoxSlot, properties {horizontalAlignment=HAlign_Center, verticalAlignment=VAlign_Center}}
  }
  Image_Dot_Top_1 {
   widgetClassPath=&Image
   parent=VBox_LowFuel_1
   widget=Image_Dot_Top_1
   properties {brush {
     bIsDynamicallyLoaded=false
     drawAs=Image
     imageType=NoImage
     mirroring=NoMirror
     resourceName=None
     resourceObject=&M_UI_Tank_Entries_Inst
     tiling=NoTile
     imageSize=<8 8>
     margin=<ltrb 0 0 0 0>
     outlineSettings {
      bUseBrushTransparency=false
      roundingType=HalfHeightRadius
      width=0
      color {colorUseRule=UseColor_Specified, specifiedColor=<rgba 0 0 0 0>}
      cornerRadii=<0 0 0 0>
     }
     tintColor {colorUseRule=UseColor_Specified, specifiedColor=<rgba 1 1 1 1>}
     uVRegion {bIsValid=false, max=<0 0>, min=<0 0>}
    }}
   slot {class=VerticalBoxSlot}
  }
  SizeBox {
   widgetClassPath=&SizeBox
   parent=VBox_LowFuel_1
   widget=SizeBox
   properties {bOverride_HeightOverride=true, heightOverride=64}
   slot {class=VerticalBoxSlot, properties {horizontalAlignment=HAlign_Center, size {sizeRule=Fill, value=1}}}
  }
  Image_33 {
   widgetClassPath=&Image
   bIsVariable=true
   parent=SizeBox
   widget=Image_33
   properties {
    brush {
     bIsDynamicallyLoaded=false
     drawAs=Image
     imageType=NoImage
     mirroring=NoMirror
     resourceName=None
     resourceObject=&M_Ui_Additive_White_025
     tiling=NoTile
     imageSize=<2 2>
     margin=<ltrb 0 0 0 0>
     outlineSettings {
      bUseBrushTransparency=false
      roundingType=HalfHeightRadius
      width=0
      color {colorUseRule=UseColor_Specified, specifiedColor=<rgba 0 0 0 0>}
      cornerRadii=<0 0 0 0>
     }
     tintColor {colorUseRule=UseColor_Specified, specifiedColor=<rgba 1 1 1 1>}
     uVRegion {bIsValid=false, max=<0 0>, min=<0 0>}
    }
    colorAndOpacity=<rgba 1 1 1 0.1>
   }
   slot {class=SizeBoxSlot}
  }
  Image_Dot_Bottom_1 {
   widgetClassPath=&Image
   parent=VBox_LowFuel_1
   widget=Image_Dot_Bottom_1
   properties {brush {
     bIsDynamicallyLoaded=false
     drawAs=Image
     imageType=NoImage
     mirroring=NoMirror
     resourceName=None
     resourceObject=&M_UI_Tank_Entries_Inst
     tiling=NoTile
     imageSize=<8 8>
     margin=<ltrb 0 0 0 0>
     outlineSettings {
      bUseBrushTransparency=false
      roundingType=HalfHeightRadius
      width=0
      color {colorUseRule=UseColor_Specified, specifiedColor=<rgba 0 0 0 0>}
      cornerRadii=<0 0 0 0>
     }
     tintColor {colorUseRule=UseColor_Specified, specifiedColor=<rgba 1 1 1 1>}
     uVRegion {bIsValid=false, max=<0 0>, min=<0 0>}
    }}
   slot {class=VerticalBoxSlot}
  }
  WBP_Vehicle_Warning {
   widgetClassPath=/Game/Athena/HUD/Vehicles/Dashboard/WBP_Vehicle_Warning
   bIsVariable=true
   parent=Overlay_8
   widget=WBP_Vehicle_Warning
   slot {class=OverlaySlot, properties {horizontalAlignment=HAlign_Center, verticalAlignment=VAlign_Center}}
  }
  Widget_FuelGauge {widgetClassPath=/Script/FortniteUI.FortVehicleFuelGauge, bInherited=true, widget=null}
  HorizontalBox_FuelCluster {widgetClassPath=&HorizontalBox, bInherited=true, widget=null}
  Image_BoostGlow {widgetClassPath=&Image, bInherited=true, widget=null}
  Image_BoostMeter {widgetClassPath=&Image, bInherited=true, widget=null}
  NumericText_Speed {widgetClassPath=/Script/CommonUI.CommonNumericTextBlock, bInherited=true, widget=null}
  Overlay_Gauge {widgetClassPath=&Overlay, bInherited=true, widget=null}
  Text_ItemName {widgetClassPath=&CommonTextBlock, bInherited=true, widget=null}
  TutorialHighlight {widgetClassPath=&TutorialHighlight}
 ]
}
